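# Selectable CommonButtonBase checkbox, named for acknowledging a price change in the Crew UI, with an inherited input action prompt beside it.
# An overlay stacks empty and checked box images, which the Check animation swaps by scale and opacity on select, reversing on deselect.
# Hover and unhover play the Hover animation forward and back, enlarging the overlay and animating a border color material parameter.
# The Burst animation has no bindings and is never played.
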WidgetBlueprint /Crew/UI/BP_PriceChangeAcknowledgeCheckBox
animations [
 Burst {binding_count=0, length_seconds=1.5, object=&self:Burst, sequence {bindings []}}
 Check {
  binding_count=2
  length_seconds=0.25
  object=&self:Check
  sequence {bindings [
    Image_Check {
     id=#1
     possessed_class=Image
     tracks [{
       type=MovieScene2DTransformTrack
       sections [{
         type=MovieScene2DTransformSection
         channels [
          Scale.X {type=MovieSceneScriptingFloatChannel, keys [[0,0],[5,1]]}
          Scale.Y {type=MovieSceneScriptingFloatChannel, keys [[0,0],[5,1]]}
          Shear.X {type=MovieSceneScriptingFloatChannel}
          Shear.Y {type=MovieSceneScriptingFloatChannel}
         ]
        }]
      }]
    }
    Image_Checkbox {
     id=#2
     possessed_class=Image
     tracks [
      {
       type=MovieScene2DTransformTrack
       sections [{
         type=MovieScene2DTransformSection
         channels [
          Scale.X {type=MovieSceneScriptingFloatChannel, keys [[0,1],[5,0]]}
          Scale.Y {type=MovieSceneScriptingFloatChannel, keys [[0,1],[5,0]]}
         ]
        }]
      }
      {
       type=MovieSceneFloatTrack
       display_name=Render Opacity
       sections [{type=MovieSceneFloatSection, channels [None {type=MovieSceneScriptingFloatChannel, keys [[0,1],[3,0]]}]}]
      }
     ]
    }
   ]}
 }
 Hover {
  binding_count=2
  length_seconds=0.15
  object=&self:Hover
  sequence {bindings [
    Overlay_1 {
     id=#3
     possessed_class=Overlay
     tracks [{
       type=MovieScene2DTransformTrack
       sections [{
         type=MovieScene2DTransformSection
         channels [
          Translation.X {type=MovieSceneScriptingFloatChannel}
          Translation.Y {type=MovieSceneScriptingFloatChannel}
          Angle {type=MovieSceneScriptingFloatChannel}
          Scale.X {type=MovieSceneScriptingFloatChannel, keys [[0,1],[3,1.1]]}
          Scale.Y {type=MovieSceneScriptingFloatChannel, keys [[0,1],[3,1.1]]}
          Shear.X {type=MovieSceneScriptingFloatChannel}
          Shear.Y {type=MovieSceneScriptingFloatChannel}
         ]
        }]
      }]
    }
    Image_Checkbox {
     id=#4
     possessed_class=Image
     tracks [{
       type=MovieSceneWidgetMaterialTrack
       sections [{type=MovieSceneParameterSection, channels [Animate_BorderColor_Hover {type=MovieSceneScriptingFloatChannel, keys [[0,0],[3,1]]}]}]
      }]
    }
   ]}
 }
]
graphs [EventGraph {nodes [
   Event On Hovered()
   > PlayAnimationForward(InAnimation=Get Hover() -> (Hover:Widget Animation Object Reference))
   Event On Unhovered()
   > PlayAnimationReverse(InAnimation=Get Hover())
   Event On Selected()
   > PlayAnimationForward(InAnimation=Get Check() -> (Check:Widget Animation Object Reference))
   Event On Deselected()
   > PlayAnimationReverse(InAnimation=Get Check())
  ]}]
variables []
widgets {
 info {inheritedWidgetCount=1, parentClass=&CommonButtonBase, rootWidgetClass=&HorizontalBox, widgetCount=6}
 widgets [
  HB_BundleText {widgetClassPath=&HorizontalBox, bIsVariable=true, widget=HB_BundleText, properties {visibility=Visible}}
  HorizontalBox_0 {widgetClassPath=&HorizontalBox, parent=HB_BundleText, widget=HorizontalBox_0, slot {class=HorizontalBoxSlot}}
  InputActionWidget {
   widgetClassPath=&CommonActionWidget
   bInherited=true
   bIsVariable=true
   parent=HorizontalBox_0
   widget=InputActionWidget
   properties {renderTransform {angle=0, scale=<1.25 1.25>, shear=<0 0>, translation=<0 0>}}
   slot {class=HorizontalBoxSlot, properties {padding=<ltrb 10 0 10 0>}}
  }
  Overlay_1 {widgetClassPath=&Overlay, parent=HorizontalBox_0, widget=Overlay_1, slot {class=HorizontalBoxSlot}}
  Image_Check {
   widgetClassPath=&Image
   bIsVariable=true
   parent=Overlay_1
   widget=Image_Check
   properties {
    brush {
     bIsDynamicallyLoaded=false
     drawAs=Image
     imageType=NoImage
     mirroring=NoMirror
     resourceName=None
     resourceObject=/FortUILibrary/Foundation/Art_ToolKit/AngledBoxes/M_UI_AngledCheckBox_Crafting
     tiling=NoTile
     imageSize=<64 64>
     margin=<ltrb 0 0 0 0>
     outlineSettings {
      bUseBrushTransparency=false
      roundingType=HalfHeightRadius
      width=0
      color {colorUseRule=UseColor_Specified, specifiedColor=<rgba 0 0 0 0>}
      cornerRadii=<0 0 0 0>
     }
     tintColor {colorUseRule=UseColor_Specified, specifiedColor=<rgba 1 1 1 1>}
     uVRegion {bIsValid=false, max=<0 0>, min=<0 0>}
    }
    renderTransform {angle=0, scale=<0 0>, shear=<0 0>, translation=<4 -11>}
   }
   slot {class=OverlaySlot, properties {padding=<ltrb -4 11 0 0>}}
  }
  Image_Checkbox {
   widgetClassPath=&Image
   bIsVariable=true
   parent=Overlay_1
   widget=Image_Checkbox
   properties {brush {
     bIsDynamicallyLoaded=false
     drawAs=Image
     imageType=NoImage
     mirroring=NoMirror
     resourceName=None
     resourceObject=/FortUILibrary/Foundation/Art_ToolKit/AngledBoxes/M_UI_AngledCheckBox_Empty
     tiling=NoTile
     imageSize=<64 64>
     margin=<ltrb 0 0 0 0>
     outlineSettings {
      bUseBrushTransparency=false
      roundingType=HalfHeightRadius
      width=0
      color {colorUseRule=UseColor_Specified, specifiedColor=<rgba 0 0 0 0>}
      cornerRadii=<0 0 0 0>
     }
     tintColor {colorUseRule=UseColor_Specified, specifiedColor=<rgba 1 1 1 1>}
     uVRegion {bIsValid=false, max=<0 0>, min=<0 0>}
    }}
   slot {class=OverlaySlot}
  }
 ]
}
