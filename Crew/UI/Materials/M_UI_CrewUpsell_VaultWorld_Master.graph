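# Master material for the Crew upsell background in the front-end Vault World, set to masked, default lit.
# The Crew upsell background function supplies the base color, which passes through Vault World post-process calibration.
# The calibrated color then feeds the shared Vault World transition function.
# All visual logic lives in these three material functions, leaving the master as a thin wrapper that chains them.
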Material /Crew/UI/Materials/M_UI_CrewUpsell_VaultWorld_Master
graph {
 properties {bAutomaticallySetUsageInEditor=false, blendMode=BLEND_Masked, usedShadingModels=Default Lit}
 nodes [
  n1 = MaterialFunctionCall!(Vault World Color=MaterialFunctionCall!(CMSTexture) {materialFunction=/Game/Athena/UI/Subscriptions/Art/Materials/MF_BG_CrewUpsell}."Switch Param (False)") {materialFunction=/Game/Environments/FrontEnd/Blueprints/ItemPreview/VaultWorldTransitions/MaterialFunctions/MF_UI_VaultWorld_PostProcessCalibration}
  MaterialFunctionCall!(Vault World Color=@n1.Calibrated Color) {materialFunction=/Game/Environments/FrontEnd/Blueprints/ItemPreview/VaultWorldTransitions/MaterialFunctions/MF_UI_VaultWorld_Transition}
 ]
}
