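# Translucent UI master material that procedurally draws a Crew subscription badge with animatable color and shine.
# Rotated UV edge masks form the outline by subtracting an inset copy, with an optional notch and radial bottom-edge distortion.
# Four sine waves on screen-aligned UVs create metallic bands that tint the color and, in bloom mode, mask opacity.
# Color can fade into a screen-space three-color radial gradient, or a static switch substitutes a radial gradient fill.
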Material /Crew/UI/Materials/M_UI_Subs_Badge_Master
graph {
 properties {
  bUsedWithStaticMesh=false
  blendMode=BLEND_Translucent
  materialDomain=MD_UI
  numCustomizedUVs=2
  pixelDepthOffsetMode=PDOM_Legacy
  usedShadingModels=Default Lit
 }
 nodes [
  n1 = ScalarParameter() {defaultValue=0.55, parameterName=BG_RadiusX}
  AppendVector(A=@n1, B=ScalarParameter() {defaultValue=0.55, parameterName=BG_RadiusY})
  n2 = ScalarParameter() {group=Pattern, parameterName=Persp_Y}
  n3 = Multiply(A=Time(), B=ScalarParameter() {defaultValue=0.125, parameterName=TimeControl})
  n4 = LinearInterpolate(Alpha=Add(A=Sine(None=@n3)))
  n5 = ScalarParameter() {group=Pattern, parameterName=Persp_X}
  n6 = ScalarParameter() {group=Pattern, parameterName=Slide_X}
  n7 = ScalarParameter() {group=Pattern, parameterName=Slide_Y}
  MaterialFunctionCall!(
   UV=TextureCoordinate()
   Texture
   FixedScaleFactor_X
   FixedScaleFactor_Y
   Persp_Y=LinearInterpolate(A=@n2, B=Multiply(A=@n2) {constB=-1}, Alpha=@n4)
   Persp_X=LinearInterpolate(A=@n5, B=Multiply(A=@n5) {constB=-1}, Alpha=@n4)
   Slide_X=LinearInterpolate(A=@n6, B=Multiply(A=@n6) {constB=-1}, Alpha=@n4)
   Slide_Y=LinearInterpolate(A=@n7, B=Multiply(A=@n7) {constB=-1}, Alpha=@n4)
   Flip_Y
   Flip_X
  ) {materialFunction=/Game/UI/Foundation/Materials/MF_UI_BasicPerspective}
  n8 = VectorParameter() {group=Color, parameterName=Color_A, defaultValue=<rgba 0 1 0 1>}
  n9 = VectorParameter() {group=Color, parameterName=Color_B, defaultValue=<rgba 1 0 0 1>}
  n10 = VectorParameter() {group=Color, parameterName=Color_C, defaultValue=<rgba 0 0 1 1>}
  MaterialFunctionCall(A=@n8.RGB, B=@n9.RGB, C=@n10.RGB) {materialFunction=&Lerp_3Color}
  n11 = ScalarParameter() {group=Color, parameterName=Rotation}
  n12 = MaterialFunctionCall() {materialFunction=&ScreenAlignedUVs}
  n13 = ScalarParameter() {group=Color, parameterName=OffsetGradient}
  n14 = Multiply(A=@n13) {constB=-1}
  n15 = ScalarParameter() {group=Color, parameterName=FlipDireciton}
  n16 = ComponentMask(None=Add(
    A=MaterialFunctionCall!(Rotation_Fraction=@n11, TexCoord=@n12."X 100%, Y 100%") {materialFunction=/FortUILibrary/Foundation/Art_ToolKit/AngledBoxes/MF_UI_Rotator_Fractions}.Result
    B=LinearInterpolate(A=@n13, B=@n14, Alpha=@n15)
   )) {r=true}
  n17 = Power(Base=LinearInterpolate(A=@n16, B=OneMinus(None=@n16), Alpha=ScalarParameter() {group=Color, parameterName=FlipDireciton})) {constExponent=1.22}
  Saturate(None=OneMinus(None=Saturate(None=@n17)))
  n18 = ScalarParameter() {group=RadialGradFill, parameterName=FillAmount}
  n19 = MaterialFunctionCall(
   UVs=TextureCoordinate()
   Texture Scale=AppendVector(
    A=ScalarParameter() {defaultValue=1, group=RadialGradFill, parameterName=Scale_X}
    B=ScalarParameter() {defaultValue=0.25, group=RadialGradFill, parameterName=Scale_Y}
   )
  ) {materialFunction=&ScaleUVsByCenter}
  n20 = AppendVector(A=Constant() {r=0.5}, B=ScalarParameter() {defaultValue=0.5, group=RadialGradFill, parameterName=Position_Y})
  n21 = ScalarParameter() {defaultValue=0.5, group=RadialGradFill, parameterName=Radius}
  n22 = ScalarParameter() {defaultValue=4, group=RadialGradFill, parameterName=Density}
  n23 = OneMinus(None=Saturate(None=MaterialFunctionCall(UVs=@n19.UVs, CenterPosition=@n20, Radius=@n21, Density=@n22) {materialFunction=&RadialGradientExponential}.RadialGradientExponential))
  n24 = ScalarParameter() {defaultValue=0.125, parameterName=LeftBadge_Top_Angle_2}
  n25 = TextureCoordinate()
  n26 = ScalarParameter() {defaultValue=-0.125, parameterName=LeftBadge_Top_Offset_2}
  n27 = Add(
   A=MaterialFunctionCall(Rotation_Fraction=Multiply(A=@n24) {constB=-1}, TexCoord=@n25) {materialFunction=&MF_UI_Rotator_Fractions}.Result
   B=@n26
  )
  n28 = ScalarParameter() {defaultValue=40, parameterName=BadgeEdgeHardness}
  n29 = Saturate(None=Multiply(A=ComponentMask(None=@n27) {g=true}, B=@n28))
  n30 = Add(
   A=MaterialFunctionCall(Rotation_Fraction=ScalarParameter() {defaultValue=0.125, parameterName=LeftBadge_Top_Angle_2}, TexCoord=@n25) {materialFunction=&MF_UI_Rotator_Fractions}.Result
   B=@n26
  )
  n31 = Saturate(None=Multiply(A=ComponentMask(None=@n30) {g=true}, B=@n28))
  n32 = ScalarParameter() {defaultValue=-0.05, parameterName=LeftBadge_Top_Angle}
  n33 = TextureCoordinate()
  n34 = ScalarParameter() {defaultValue=-0.25, parameterName=LeftBadge_Top_Offset}
  n35 = Add(
   A=MaterialFunctionCall(Rotation_Fraction=Multiply(A=@n32) {constB=-1}, TexCoord=@n33) {materialFunction=&MF_UI_Rotator_Fractions}.Result
   B=@n34
  )
  n36 = ScalarParameter() {defaultValue=40, parameterName=BadgeEdgeHardness}
  n37 = Saturate(None=Multiply(A=ComponentMask(None=@n35) {g=true}, B=@n36))
  n38 = Add(
   A=MaterialFunctionCall(Rotation_Fraction=ScalarParameter() {defaultValue=-0.05, parameterName=LeftBadge_Top_Angle}, TexCoord=@n33) {materialFunction=&MF_UI_Rotator_Fractions}.Result
   B=@n34
  )
  n39 = Saturate(None=Multiply(A=ComponentMask(None=@n38) {g=true}, B=@n36))
  n40 = DotProduct(A=OneMinus(None=DotProduct(A=@n29, B=@n31)), B=OneMinus(None=Add(A=@n37, B=@n39)))
  n41 = ScalarParameter() {defaultValue=0.025, parameterName=LeftBadge_Side_Angle}
  n42 = TextureCoordinate()
  n43 = ScalarParameter() {defaultValue=-0.1, parameterName=LeftBadge_Side_Offset}
  n44 = Add(
   A=MaterialFunctionCall(Rotation_Fraction=Multiply(A=@n41) {constB=-1}, TexCoord=OneMinus(None=@n42)) {materialFunction=&MF_UI_Rotator_Fractions}.Result
   B=@n43
  )
  n45 = ScalarParameter() {defaultValue=40, parameterName=BadgeEdgeHardness}
  n46 = Multiply(A=ComponentMask(None=@n44) {r=true}, B=@n45)
  n47 = Add(
   A=MaterialFunctionCall(Rotation_Fraction=ScalarParameter() {defaultValue=0.025, parameterName=LeftBadge_Side_Angle}, TexCoord=@n42) {materialFunction=&MF_UI_Rotator_Fractions}.Result
   B=@n43
  )
  n48 = Multiply(A=ComponentMask(None=@n47) {r=true}, B=@n45)
  n49 = DotProduct(A=Saturate(None=OneMinus(None=@n40)), B=DotProduct(A=Saturate(None=@n46), B=Saturate(None=@n48)))
  n50 = ScalarParameter() {defaultValue=0.025, parameterName=LeftBadge_Bottom_Angle}
  n51 = OneMinus(None=TextureCoordinate())
  n52 = ScalarParameter() {defaultValue=-0.1, parameterName=LeftBadge_Bottom_Offset}
  n53 = Add(
   A=MaterialFunctionCall(Rotation_Fraction=Multiply(A=@n50) {constB=-1}, TexCoord=@n51) {materialFunction=&MF_UI_Rotator_Fractions}.Result
   B=@n52
  )
  n54 = ScalarParameter() {defaultValue=0.5, parameterName=RadialDistortion_X_Position}
  n55 = ScalarParameter() {defaultValue=0.5, parameterName=RadialDistortion_y_Position}
  n56 = Multiply(
   A=MaterialFunctionCall(
    UVs=TextureCoordinate()
    CenterPosition=AppendVector(A=@n54, B=@n55)
    Radius=ScalarParameter() {parameterName=RadialDistortion_Size}
    Density=ScalarParameter() {defaultValue=2, parameterName=RadialDistortion_Density}
   ) {materialFunction=&RadialGradientExponential}.RadialGradientExponential
   B=ScalarParameter() {parameterName=RadialDistortion_Amount}
  )
  n57 = ComponentMask(None=Add(A=@n53, B=@n56)) {g=true}
  n58 = ScalarParameter() {defaultValue=40, parameterName=BadgeEdgeHardness}
  n59 = Add(
   A=MaterialFunctionCall(Rotation_Fraction=ScalarParameter() {defaultValue=0.025, parameterName=LeftBadge_Bottom_Angle}, TexCoord=@n51) {materialFunction=&MF_UI_Rotator_Fractions}.Result
   B=@n52
  )
  n60 = ComponentMask(None=Add(A=@n59, B=@n56)) {g=true}
  n61 = DotProduct(A=Saturate(None=Multiply(A=@n57, B=@n58)), B=Saturate(None=Multiply(A=@n60, B=@n58)))
  n62 = Add(A=AppendVector(A=@n27, B=Constant()), B=Multiply(A=ScalarParameter() {defaultValue=0.125, parameterName=BadgeThiccness}) {constB=-1}) {constB=-0.125}
  n63 = ScalarParameter() {defaultValue=40, parameterName=BadgeEdgeHardness}
  n64 = Saturate(None=Multiply(A=ComponentMask(None=@n62) {g=true}, B=@n63))
  n65 = Add(A=AppendVector(A=@n30, B=Constant()), B=Multiply(A=ScalarParameter() {defaultValue=0.125, parameterName=BadgeThiccness}) {constB=-1}) {constB=-0.125}
  n66 = Saturate(None=Multiply(A=ComponentMask(None=@n65) {g=true}, B=@n63))
  n67 = Add(A=AppendVector(A=@n35, B=Constant()), B=Multiply(A=ScalarParameter() {defaultValue=0.125, parameterName=BadgeThiccness}) {constB=-1}) {constB=-0.125}
  n68 = ScalarParameter() {defaultValue=40, parameterName=BadgeEdgeHardness}
  n69 = Saturate(None=Multiply(A=ComponentMask(None=@n67) {g=true}, B=@n68))
  n70 = Add(A=AppendVector(A=@n38, B=Constant()), B=Multiply(A=ScalarParameter() {defaultValue=0.125, parameterName=BadgeThiccness}) {constB=-1}) {constB=-0.125}
  n71 = Saturate(None=Multiply(A=ComponentMask(None=@n70) {g=true}, B=@n68))
  n72 = DotProduct(A=OneMinus(None=DotProduct(A=@n64, B=@n66)), B=OneMinus(None=Add(A=@n69, B=@n71)))
  n73 = Add(A=AppendVector(A=@n44, B=Constant()), B=Multiply(A=ScalarParameter() {defaultValue=0.125, parameterName=BadgeThiccness}) {constB=-1}) {constB=-0.125}
  n74 = ScalarParameter() {defaultValue=40, parameterName=BadgeEdgeHardness}
  n75 = Saturate(None=Multiply(A=ComponentMask(None=@n73) {r=true}, B=@n74))
  n76 = Add(A=AppendVector(A=@n47, B=Constant()), B=Multiply(A=ScalarParameter() {defaultValue=0.125, parameterName=BadgeThiccness}) {constB=-1}) {constB=-0.125}
  n77 = Saturate(None=Multiply(A=ComponentMask(None=@n76) {r=true}, B=@n74))
  n78 = Constant()
  n79 = ScalarParameter() {defaultValue=0.125, parameterName=BadgeThiccness}
  n80 = Add(
   A=Add(A=AppendVector(A=@n53, B=@n78), B=Multiply(A=@n79) {constB=-1}) {constB=-0.125}
   B=Multiply(A=ScalarParameter() {parameterName=BadgeThiccness_BottomAdjustment}) {constB=-1}
  ) {constB=-0.125}
  n81 = ScalarParameter() {defaultValue=40, parameterName=BadgeEdgeHardness}
  n82 = Multiply(A=ComponentMask(None=Add(A=@n80, B=@n56) {constB=-0.125}) {g=true}, B=@n81)
  n83 = Constant()
  n84 = ScalarParameter() {defaultValue=0.125, parameterName=BadgeThiccness}
  n85 = Add(
   A=Add(A=AppendVector(A=@n59, B=@n83), B=Multiply(A=@n84) {constB=-1}) {constB=-0.125}
   B=Multiply(A=ScalarParameter() {parameterName=BadgeThiccness_BottomAdjustment}) {constB=-1}
  ) {constB=-0.125}
  n86 = Multiply(A=ComponentMask(None=Add(A=@n85, B=@n56) {constB=-0.125}) {g=true}, B=@n81)
  n87 = DotProduct(A=DotProduct(A=OneMinus(None=@n72), B=DotProduct(A=@n75, B=@n77)), B=DotProduct(A=Saturate(None=@n82), B=Saturate(None=@n86)))
  n88 = ScalarParameter() {defaultValue=40, parameterName=BadgeEdgeHardness}
  n89 = Multiply(A=Add(A=TextureCoordinate()) {constB=-0.5}, B=@n88)
  n90 = ComponentMask(None=LinearInterpolate(A=@n89, B=OneMinus(None=@n89), Alpha=ScalarParameter() {defaultValue=1, group=Notch, parameterName=FlipNotchSide})) {r=true}
  n91 = TextureCoordinate()
  n92 = Add(
   A=MaterialFunctionCall(Rotation_Fraction=ScalarParameter() {defaultValue=0.025, group=Notch, parameterName=LeftBadge_Notch_Angle}, TexCoord=OneMinus(None=@n91)) {materialFunction=&MF_UI_Rotator_Fractions}.Result
   B=ScalarParameter() {defaultValue=-0.5, group=Notch, parameterName=LeftBadge_Notch_Offset}
  )
  n93 = Saturate(None=Multiply(A=ComponentMask(None=@n92) {g=true}, B=@n88))
  n94 = Constant()
  n95 = ScalarParameter() {defaultValue=0.125, group=Notch, parameterName=NotchThiccness}
  n96 = ComponentMask(None=Add(A=AppendVector(A=@n92, B=@n94), B=Multiply(A=@n95) {constB=-1}) {constB=-0.125}) {g=true}
  n97 = OneMinus(None=Saturate(None=Multiply(A=@n96, B=@n88)))
  n98 = OneMinus(None=DotProduct(A=Saturate(None=@n90), B=DotProduct(A=@n93, B=@n97)))
  n99 = LinearInterpolate(
   B=DotProduct(A=DotProduct(A=@n49, B=@n61), B=OneMinus(None=@n87))
   Alpha=StaticSwitchParameter(True=Saturate(None=@n98), False=Constant() {r=1}) {group=Notch, parameterName=UseNotch}
  )
  n100 = Add(
   A=MaterialFunctionCall(
    Rotation_Fraction=ScalarParameter() {defaultValue=0.125, parameterName=SilverRotation}
    TexCoord=MaterialFunctionCall() {materialFunction=&ScreenAlignedUVs}."X 100%, Y 100%"
   ) {materialFunction=&MF_UI_Rotator_Fractions}.Result
   B=AppendVector(A=ScalarParameter() {parameterName=MetalOffset}, B=Constant())
  )
  n101 = Sine(None=Multiply(A=@n100, B=ScalarParameter() {defaultValue=1, group=Color, parameterName=Sine_1})) {period=0.125}
  n102 = Sine(None=Multiply(A=@n100, B=ScalarParameter() {defaultValue=1, group=Color, parameterName=Sine_2})) {period=0.3}
  n103 = Sine(None=Multiply(A=@n100, B=ScalarParameter() {defaultValue=1, group=Color, parameterName=Sine_3})) {period=0.25}
  n104 = Sine(None=Multiply(A=@n100, B=ScalarParameter() {defaultValue=1, group=Color, parameterName=Sine_4})) {period=0.325}
  n105 = DotProduct(
   A=DotProduct(A=ComponentMask(None=@n101) {r=true}, B=ComponentMask(None=@n102) {r=true})
   B=DotProduct(A=ComponentMask(None=@n103) {r=true}, B=ComponentMask(None=@n104) {r=true})
  )
  n106 = Saturate(None=Multiply(A=Saturate(None=@n105), B=ScalarParameter() {defaultValue=1, group=Color, parameterName=ColorMultiply}))
  n107 = LinearInterpolate(A=@n99, B=DotProduct(A=@n106, B=@n99), Alpha=ScalarParameter() {parameterName=IsBloom})
  StaticSwitchParameter(True=LinearInterpolate(B=LinearInterpolate(A=@n18, Alpha=@n23), Alpha=@n107), False=@n107) {group=RadialGradFill, parameterName=UseRadialGradFill}
  n108 = LinearInterpolate(A=@n9.RGB, B=Constant() {r=1}, Alpha=ScalarParameter() {parameterName=AnimToWhite})
  n109 = Multiply(
   A=MaterialFunctionCall() {materialFunction=/FortUILibrary/Foundation/Art_ToolKit/Tools/MF_UI_SquareToScreenSpace}.Y_Priority
   B=AppendVector(A=ScalarParameter() {defaultValue=1, parameterName=BG_MultiplyUV_X}, B=ScalarParameter() {defaultValue=1, parameterName=BG_MultiplyUV_Y})
  )
  n110 = AppendVector(
   A=ScalarParameter() {defaultValue=0.5, parameterName=BG_X_Position, sliderMin=0.5}
   B=ScalarParameter() {defaultValue=0.5, parameterName=BG_Y_Position}
  )
  n111 = ScalarParameter() {defaultValue=2, parameterName=BG_Density}
  n112 = MaterialFunctionCall!(
   X=ComponentMask(None=MaterialFunctionCall(UVs=@n109, CenterPosition=@n110, Radius=@n1, Density=@n111) {materialFunction=&RadialGradientExponential}.RadialGradientExponential) {r=true}
   Tangent 0=ScalarParameter() {group=Color, parameterName=BottomTangent}
   Tangent 1=ScalarParameter() {defaultValue=3, group=Color, parameterName=TopTangent}
  ) {materialFunction=/Engine/Functions/Engine_MaterialFunctions02/Gradients/SmoothCurve}
  n113 = Saturate(None=Saturate(None=ComponentMask(None=@n112.Result) {r=true}))
  n114 = LinearInterpolate(A=@n113, B=OneMinus(None=ComponentMask(None=@n113) {r=true}), Alpha=ScalarParameter() {group=Color, parameterName=FlipDireciton})
  n115 = ScalarParameter() {defaultValue=1, group=Color, parameterName=AdjustRate}
  n116 = Saturate(None=MaterialFunctionCall(A=@n10.RGB, B=@n9.RGB, C=@n8.RGB, Alpha=Power(Base=@n114, Exp=@n115) {constExponent=1.22}) {materialFunction=&Lerp_3Color}.Result)
  n117 = LinearInterpolate(
   A=Saturate(None=LinearInterpolate(A=@n8.RGB, B=@n108, Alpha=@n106))
   B=@n116
   Alpha=ScalarParameter() {group=Color, parameterName=FadeIntoScreenSpace}
  )
  n118 = ComponentMask(None=Add(A=TextureCoordinate()) {constB=-0.333}) {g=true}
  StaticSwitchParameter(
   True=LinearInterpolate(
    A=VectorParameter() {group=Color, parameterName=Color_A, defaultValue=<rgba 0 1 0 1>}.RGB
    B=VectorParameter() {group=Color, parameterName=Color_B, defaultValue=<rgba 1 0 0 1>}.RGB
    Alpha=@n23
   )
   False=LinearInterpolate(A=@n117, B=LinearInterpolate(A=@n117, B=@n116, Alpha=@n118), Alpha=ScalarParameter() {group=Color, parameterName=FadeTipIntoScreenSpace})
  ) {group=RadialGradFill, parameterName=UseRadialGradFill}
 ]
}
