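# RichTextButton is a reusable CommonButtonLegacy widget showing a rich text label inside a material-driven Border.
# PreConstruct applies the configurable text, color, justification, auto-wrap and brush material.
# Hover animates the material's opacity and reflection parameters and Hover2 scales the border up slightly, but this graph plays neither.
# The font size, shear, right image and SimpleHover variables are declared but unused in this graph.
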WidgetBlueprint /Crew/UI/RichTextButton
animations [
 Hover {
  binding_count=1
  length_seconds=0.1
  object=&self:Hover
  sequence {bindings [Border_Material {
     id=#1
     possessed_class=Border
     tracks [{
       type=MovieSceneWidgetMaterialTrack
       sections [{
         type=MovieSceneParameterSection
         channels [
          Buttonn_Opcaity {type=MovieSceneScriptingFloatChannel, keys [[0,0.8],[2,1]]}
          Reflection_Opcaity {type=MovieSceneScriptingFloatChannel, keys [[0,0.125],[2,0.2]]}
          Reflection anim {type=MovieSceneScriptingFloatChannel, keys [[0,0],[2,0.5]]}
         ]
        }]
      }]
    }]}
 }
 Hover2 {
  binding_count=1
  length_seconds=0.1
  object=&self:Hover2
  sequence {bindings [Border_Material {
     id=#2
     possessed_class=Border
     tracks [{
       type=MovieScene2DTransformTrack
       sections [{
         type=MovieScene2DTransformSection
         channels [
          Scale.X {type=MovieSceneScriptingFloatChannel, keys [[0,1],[2,1.05]]}
          Scale.Y {type=MovieSceneScriptingFloatChannel, keys [[0,1],[2,1.05]]}
         ]
        }]
      }]
    }]}
 }
]
graphs [EventGraph {nodes [
   Event PreConstruct()
   n1 = Get Rich_text() -> (Rich_text:Common Rich Text Block Object Reference)
   > SetText(self=@n1, InText=Get Button text() -> (Button text:Text))
   > SetDefaultColorAndOpacity!(
    execute
    self:Rich Text Block Object Reference=@n1
    InColorAndOpacity:Slate Color Structure=Get Text Color() -> (Text Color:Slate Color Structure)
   ) -> (then)
   > Set Justification!(
    execute
    Justification:ETextJustify Enum=Get Justifiction() -> (Justifiction:ETextJustify Enum=Left)
    self:Text Layout Widget Object Reference=@n1
   ) -> (then, Output_Get:ETextJustify Enum=Left)
   > SetAutoWrapText!(execute, self:Rich Text Block Object Reference=Get Rich_text(), InAutoTextWrap:bool=Get Auto wrap text() -> (Auto wrap text:bool)) -> (then)
   > SetBrushFromMaterial(
    self:Border Object Reference=Get Border_Material() -> (Border_Material:Border Object Reference)
    Material=Get Material Brush() -> (Material Brush:Material Interface Object Reference)
   )
  ]}]
variables [
 Auto wrap text {type=boolean, default=true}
 Button text {type=string, default=Button text}
 FontSize {type=integer}
 FontSizeOverride {type=boolean}
 Justifiction {type=ETextJustify, default=Left}
 Material Brush {type=MaterialInterface, default=/Game/Athena/UI/Frontend/Spatial/Art/M_Ui_Skewed_SpatialButton_Default}
 Right Image {type=Texture2D, default=/Game/UI/Foundation/Textures/Icons/Items/T-FNBR-BattlePass}
 ShearText {type=boolean}
 SimpleHover {type=boolean}
 Text Color {type=SlateColor, default {colorUseRule=UseColor_Specified, specifiedColor=<rgba 1 1 1 1>}}
]
widgets {
 info {inheritedWidgetCount=1, parentClass=&CommonButtonLegacy, rootWidgetClass=&Border, widgetCount=3}
 widgets [
  Border_Material {
   widgetClassPath=&Border
   bIsVariable=true
   widget=Border_Material
   properties {
    verticalAlignment=VAlign_Center
    background {
     bIsDynamicallyLoaded=false
     drawAs=Box
     imageType=NoImage
     mirroring=NoMirror
     resourceName=None
     resourceObject=null
     tiling=NoTile
     imageSize=<128 128>
     margin=<ltrb 0.45 0 0.45 0>
     outlineSettings {
      bUseBrushTransparency=false
      roundingType=HalfHeightRadius
      width=0
      color {colorUseRule=UseColor_Specified, specifiedColor=<rgba 0 0 0 0>}
      cornerRadii=<0 0 0 0>
     }
     tintColor {colorUseRule=UseColor_Specified, specifiedColor=<rgba 1 1 1 0>}
     uVRegion {bIsValid=false, max=<0 0>, min=<0 0>}
    }
    padding=<ltrb 0 0 0 0>
   }
  }
  Rich_text {
   widgetClassPath=&CommonRichTextBlock
   bIsVariable=true
   parent=Border_Material
   widget=Rich_text
   properties {
    bOverrideDefaultStyle=true
    bTintInlineIcon=true
    justification=Center
    lineHeightPercentage=0.8
    text="Purchase\r\n<text color=\"FFFFFFFF\">Battle Pass!</>"
    defaultTextStyleOverride {
     overflowPolicy=Clip
     transformPolicy=None
     colorAndOpacity {colorUseRule=UseColor_Specified, specifiedColor=<rgba 1 1 1 1>}
     font {
      bForceMonospaced=false
      bMaterialIsStencil=false
      fontMaterial=null
      fontObject=&HeadingNow
      letterSpacing=0
      monospacedWidth=1
      size=16.5
      skewAmount=0
      typefaceFontName=76 Bold
      outlineSettings {
       bApplyOutlineToDropShadows=false
       bMiteredCorners=false
       bSeparateFillAlpha=false
       outlineMaterial=null
       outlineSize=0
       outlineColor=<rgba 0 0 0 1>
      }
     }
     highlightColor {colorUseRule=UseColor_Specified, specifiedColor=<rgba 0 0 0 1>}
     highlightShape {
      bIsDynamicallyLoaded=false
      drawAs=Image
      imageType=NoImage
      mirroring=NoMirror
      resourceName=None
      resourceObject=null
      tiling=NoTile
      imageSize=<32 32>
      margin=<ltrb 0 0 0 0>
      outlineSettings {
       bUseBrushTransparency=false
       roundingType=HalfHeightRadius
       width=0
       color {colorUseRule=UseColor_Specified, specifiedColor=<rgba 0 0 0 0>}
       cornerRadii=<0 0 0 0>
      }
      tintColor {colorUseRule=UseColor_Specified, specifiedColor=<rgba 1 1 1 1>}
      uVRegion {bIsValid=false, max=<0 0>, min=<0 0>}
     }
     selectedBackgroundColor {colorUseRule=UseColor_Foreground, specifiedColor=<rgba 1 0 1 1>}
     shadowColorAndOpacity=<rgba 0 0 0 1>
     shadowOffset=<0 0>
     strikeBrush {
      bIsDynamicallyLoaded=false
      drawAs=Image
      imageType=NoImage
      mirroring=NoMirror
      resourceName=None
      resourceObject=null
      tiling=NoTile
      imageSize=<32 32>
      margin=<ltrb 0 0 0 0>
      outlineSettings {
       bUseBrushTransparency=false
       roundingType=HalfHeightRadius
       width=0
       color {colorUseRule=UseColor_Specified, specifiedColor=<rgba 0 0 0 0>}
       cornerRadii=<0 0 0 0>
      }
      tintColor {colorUseRule=UseColor_Specified, specifiedColor=<rgba 1 1 1 1>}
      uVRegion {bIsValid=false, max=<0 0>, min=<0 0>}
     }
     underlineBrush {
      bIsDynamicallyLoaded=false
      drawAs=Image
      imageType=NoImage
      mirroring=NoMirror
      resourceName=None
      resourceObject=null
      tiling=NoTile
      imageSize=<32 32>
      margin=<ltrb 0 0 0 0>
      outlineSettings {
       bUseBrushTransparency=false
       roundingType=HalfHeightRadius
       width=0
       color {colorUseRule=UseColor_Specified, specifiedColor=<rgba 0 0 0 0>}
       cornerRadii=<0 0 0 0>
      }
      tintColor {colorUseRule=UseColor_Specified, specifiedColor=<rgba 1 1 1 1>}
      uVRegion {bIsValid=false, max=<0 0>, min=<0 0>}
     }
    }
    margin=<ltrb 0 0 0 10>
   }
   slot {class=BorderSlot, properties {verticalAlignment=VAlign_Center, padding=<ltrb 0 0 0 0>}}
  }
  InputActionWidget {widgetClassPath=&CommonActionWidget, bInherited=true, widget=null}
 ]
}
