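# Fortnite Crew modal, built on SubscriptionModalBase and the UIKit dialog, that presents a subscription change notice for acknowledgment.
# Dialog buttons swap a visibility switcher between explainer, sign-up steps and EULA, relabeling via RefreshButtons, or report non-acceptance to the notification view model and close.
# Config flags hide Terms and Cancel; subscription platform gates the first button, footer and link.
# The rich-text link launches the view model's URL.
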WidgetBlueprint /Crew/UI/PriceChangeAcknowledgeModal
graphs [
 EventGraph {nodes [
   "On Clicked (RichTextButton)"!() -> (OutputDelegate, then, Button:Common Button Base Object Reference) {is=title K2Node_ComponentBoundEvent}
   n1 = "To String (Text)"(InText=Get MoreInfoUrl!(self:Crew Platform And Countries Notification VM Object Reference=Get CrewPlatformAndCountriesNotificationVM() -> (CrewPlatformAndCountriesNotificationVM:Crew Platform And Countries Notification VM Object Reference)) -> (MoreInfoUrl:Text))
   > n2 = CanLaunchURL!(execute, URL:String=@n1) -> (then, ReturnValue:bool)
   > Branch(Condition=@n2)
   > LaunchURL!(execute, URL:String=@n1) -> (then)
   n3 = ButtonClicked!() -> (OutputDelegate, then, ButtonIndex:int) {is=title K2Node_CustomEvent}
   > n4 = Switch on Int!(execute:Exec, Selection:int=@n3.ButtonIndex) -> (Default:Exec, 0:Exec, 1:Exec, 2:Exec)
   n5 = SetActiveWidget(
    execute=@n4.1
    self=Get LeftSideTextSwitcher() -> (LeftSideTextSwitcher:Common Visibility Switcher Object Reference)
    Widget=Select(
     Option 0:Widget Object Reference=Get SubscriptionEulaWidget() -> (SubscriptionEulaWidget:Subscription Eula Widget Object Reference)
     Option 1:Widget Object Reference=Get ExplainerScrollBox() -> (ExplainerScrollBox:Athena Scroll Box Object Reference)
     Index=Get IsTermsVisible() -> (IsTermsVisible:bool)
    ) -> (ReturnValue:Widget Object Reference)
   )
   RefreshButtons(execute=[@n5.then, @n6.then])
   ProcessAcknowledgment!(
    execute=@n4.2
    self:Crew Platform And Countries Notification VM Object Reference=Get CrewPlatformAndCountriesNotificationVM()
    Accepted:bool=false
   ) -> (then)
   > DeactivateWidget()
   n6 = SetActiveWidget(
    execute=@n4.0
    self=Get LeftSideTextSwitcher()
    Widget=Select(
     Option 0=Get SignUpStepsScrollBox() -> (SignUpStepsScrollBox:Athena Scroll Box Object Reference)
     Option 1=Get ExplainerScrollBox()
     Index=Get IsSignUpStepsVisible() -> (IsSignUpStepsVisible:bool)
    )
   )
   Event Construct()
   > n7 = Construct Object from Class(Class=/Crew/UI/Subscription/Data/DialogVM_NotificationModal.DialogVM_NotificationModal_C) -> (ReturnValue:Dialog VM Notification Modal Object Reference)
   n8 = Get Dialog_Internal() -> (Dialog_Internal:WBP UIKit Dialog Internal Object Reference)
   > SetDialogData!(execute, self:WBP UIKit Dialog Internal Object Reference=@n8, Data:UI Kit Dialog Viewmodel Object Reference=@n7) -> (then)
   > Bind Event to Dialog Button Clicked Dispatcher(self:WBP UIKit Dialog Internal Object Reference=@n8, Delegate=Create Event()) {is=title K2Node_AddDelegate}
   > SetScrollWidget!(
    execute
    ToScroll:Widget Object Reference=Get ScrollBox_Text!(self:WBP UIKit Dialog Internal Object Reference=@n8) -> (ScrollBox_Text:Scroll Box Object Reference)
   ) -> (then)
   n9 = Get ButtonWidgets!(self:WBP UIKit Dialog Internal Object Reference=Get Dialog_Internal()) -> (ButtonWidgets:Array of WBP UIKit Button CTA Base Object References)
   > Set AcceptButton(AcceptButton:WBP UIKit Button CTA Base Object Reference="Get (a copy)"(Array:Array of WBP UIKit Button CTA Base Object References=@n9) -> (Output:WBP UIKit Button CTA Base Object Reference)) -> (Output_Get:WBP UIKit Button CTA Base Object Reference)
   > Set TermsButton(TermsButton:WBP UIKit Button CTA Base Object Reference="Get (a copy)"(Array=@n9, Dimension 1=1)) -> (Output_Get:WBP UIKit Button CTA Base Object Reference)
   Event On Activated()
   n10 = Get ButtonWidgets(self=Get Dialog_Internal())
   > SetVisibility(
    self="Get (a copy)"(Array=@n10, Dimension 1=1)
    InVisibility=Select(
     Option 0:ESlateVisibility Enum=SelfHitTestInvisible
     Option 1:ESlateVisibility Enum=Collapsed
     Index=GetConfigDataBoolWithDefault!(EntryName:String=IsTermsButtonHidden, bDefaultValue:bool=false) -> (ReturnValue:bool)
    ) -> (ReturnValue:ESlateVisibility Enum)
   )
   > SetVisibility(
    self="Get (a copy)"(Array=@n10, Dimension 1=2)
    InVisibility=Select(Option 0=SelfHitTestInvisible, Option 1=Collapsed, Index=GetConfigDataBoolWithDefault(EntryName=IsCancelButtonHidden, bDefaultValue=false))
   )
   > ActivateWidget(self=Get SubscriptionEulaWidget())
   n11 = Get CrewVM() -> (CrewVM:Crew VM Object Reference)
   n12 = IsOnActiveSubscriptionPlatform!(self:Crew VM Object Reference=@n11) -> (ReturnValue:bool)
   n13 = "Not Equal (Enum)"!(
    A:EAppStore Enum=ActiveSubscriptionPlatform!(self:Crew VM Object Reference=@n11) -> (ReturnValue:EAppStore Enum=DebugStore)
    B:EAppStore Enum=EpicPurchasingService
   ) -> (ReturnValue:bool)
   > SetVisibility(
    self=Get AcceptButton() -> (AcceptButton:WBP UIKit Button CTA Base Object Reference)
    InVisibility=Select(Option 0=Collapsed, Option 1=SelfHitTestInvisible, Index=AND Boolean(A=@n13, B=@n12))
   )
   Event PreConstruct()
   > n14 = AddChildToGrid!(
    execute
    self:Grid Panel Object Reference=Get Content_FirstColumn!(self:WBP UIKit Dialog Internal Object Reference=Get Dialog_Internal()) -> (Content_FirstColumn:Grid Panel Object Reference)
    Content:Widget Object Reference=Get RichTextButton() -> (RichTextButton:Rich Text Button Object Reference)
    InRow:int=3
    InColumn:int=1
   ) -> (then, ReturnValue:Grid Slot Object Reference)
   > SetPadding!(
    execute
    self:Grid Slot Object Reference=@n14
    InPadding_Left:float=80
    InPadding_Top:float=20
    InPadding_Right:float=80
    InPadding_Bottom:float=20
   ) -> (then)
  ]}
 RefreshButtons {nodes [
   RefreshButtons() {is=title K2Node_FunctionEntry}
   n1 = Get LeftSideTextSwitcher()
   > Set IsTermsVisible(IsTermsVisible:bool="Equal (Object)"(
     A=GetActiveWidget!(self:Common Visibility Switcher Object Reference=@n1) -> (ReturnValue:Widget Object Reference)
     B=Get SubscriptionEulaWidget()
    )) -> (Output_Get:bool)
   n2 = Get LeftSideTextSwitcher()
   > Set IsSignUpStepsVisible(IsSignUpStepsVisible:bool="Equal (Object)"(A=GetActiveWidget(self=@n2), B=Get SignUpStepsScrollBox())) -> (Output_Get:bool)
   n3 = Get CrewPlatformAndCountriesNotificationVM()
   > SetText!(
    execute
    self:Fort CTAButton Object Reference=Get AcceptButton()
    ButtonText:"Text (by ref)"=SelectText(
     A=Get ExplainerTextLocalized() -> (ExplainerTextLocalized:Text)
     B=Get ConfirmButtonText!(self:Crew Platform And Countries Notification VM Object Reference=@n3) -> (ConfirmButtonText:Text)
     bPickA=Get IsSignUpStepsVisible()
    )
   ) -> (then)
   > SetText(
    self=Get TermsButton() -> (TermsButton:WBP UIKit Button CTA Base Object Reference)
    ButtonText=SelectText(A=Get ExplainerTextLocalized(), B="LOCTEXT(\"Terms\")", bPickA=Get IsTermsVisible())
   )
  ]}
 SetCancelButtonText {nodes [
   n1 = SetCancelButtonText!() -> (then, ButtonText:Text) {is=title K2Node_FunctionEntry}
   n2 = Get Dialog_Internal()
   > SetText(self="Get (a copy)"(Array=Get ButtonWidgets(self=@n2), Dimension 1=2), ButtonText=@n1)
  ]}
 SetConfirmButtonText {nodes [
   n1 = SetConfirmButtonText!() -> (then, ButtonText:Text) {is=title K2Node_FunctionEntry}
   n2 = Get Dialog_Internal()
   > SetText(self="Get (a copy)"(Array=Get ButtonWidgets(self=@n2)), ButtonText=@n1)
  ]}
 SetFooterText {nodes [
   n1 = SetFooterText!() -> (then, FooterText:Text) {is=title K2Node_FunctionEntry}
   n2 = TextIsEmpty(InText=@n1)
   n3 = IsOnActiveSubscriptionPlatform(self=Get CrewVM())
   n4 = Get Dialog_Internal()
   > SetVisibility(
    self=Get FooterText!(self:WBP UIKit Dialog Internal Object Reference=@n4) -> (FooterText:Common Text Object Reference)
    InVisibility=Select(Option 0=SelfHitTestInvisible, Option 1=Collapsed, Index=OR Boolean(A=@n2, B=@n3))
   )
   > SetDialogFooterText!(execute, self:WBP UIKit Dialog Internal Object Reference=@n4, FooterText:Text=Get FooterText!() -> (FooterText:Text)) -> (then)
  ]}
 SetPriceChangeAcknowledgeBodyTable {nodes [SetPriceChangeAcknowledgeBodyTable!() -> (then, NewParam:Array of Crew Table Row Structures) {is=title K2Node_FunctionEntry}]}
 SetUrlText {nodes [
   n1 = SetUrlText!() -> (then, MoreInfoURL:Text) {is=title K2Node_FunctionEntry}
   n2 = TextIsEmpty(InText=@n1)
   n3 = NOT Boolean(A=IsOnActiveSubscriptionPlatform(self=Get CrewVM()))
   n4 = Get RichTextButton()
   > SetVisibility(self=@n4, InVisibility=Select(Option 0=SelfHitTestInvisible, Option 1=Collapsed, Index=OR Boolean(A=@n2, B=@n3)))
   > SetText!(
    execute
    self:Rich Text Block Object Reference=Get Rich_text!(self:Rich Text Button Object Reference=@n4) -> (Rich_text:Common Rich Text Block Object Reference)
    InText:"Text (by ref)"=Get MoreInfoURL() -> (MoreInfoURL:Text)
   ) -> (then)
  ]}
 ToggleAcceptButton {nodes [
   ToggleAcceptButton() {is=title K2Node_FunctionEntry}
   > Set ShouldShowSignUpSteps(ShouldShowSignUpSteps:bool=NOT Boolean(A=Get IsSignUpStepsVisible())) -> (Output_Get:bool)
   n1 = Get IsSignUpStepsVisible()
   > SetActiveWidget(
    self=Get LeftSideTextSwitcher()
    Widget=Select(Option 0:Widget Object Reference=Get ExplainerScrollBox(), Option 1:Widget Object Reference=Get SignUpStepsScrollBox(), Index=@n1) -> (ReturnValue:Widget Object Reference)
   )
   n2 = Get CrewPlatformAndCountriesNotificationVM()
   n3 = Get Dialog_Internal()
   > SetText!(
    execute
    self:Fort CTAButton Object Reference="Get (a copy)"(Array=Get ButtonWidgets(self=@n3))
    ButtonText:"Text (by ref)"=SelectText(A="LOCTEXT(\"Explainer\")", B=Get ConfirmButtonText(self=@n2), bPickA=@n1)
   ) -> (then)
  ]}
 ToggleTermsText {nodes [
   ToggleTermsText() {is=title K2Node_FunctionEntry}
   > n1 = Set IsTermsVisible(IsTermsVisible=NOT Boolean(A=Get IsTermsVisible()))
   > SetActiveWidget(self=Get LeftSideTextSwitcher(), Widget=Select(Option 0=Get ExplainerScrollBox(), Option 1=Get SubscriptionEulaWidget(), Index=@n1))
   n2 = Get Dialog_Internal()
   > SetText(
    self="Get (a copy)"(Array=Get ButtonWidgets(self=@n2), Dimension 1=1)
    ButtonText=SelectText(A="LOCTEXT(\"Explainer\")", B="LOCTEXT(\"Terms\")", bPickA=Get IsTermsVisible())
   )
   SetVisibility(
    self=Get ContentRichText!(self:WBP UIKit Dialog Internal Object Reference=Get Dialog_Internal()) -> (ContentRichText:Common Rich Text Block Object Reference)
    InVisibility=Select(Option 0:ESlateVisibility Enum=SelfHitTestInvisible, Option 1:ESlateVisibility Enum=Collapsed, Index=Get IsTermsVisible()) -> (ReturnValue:ESlateVisibility Enum)
   )
   SetVisibility(self=Get SubscriptionEulaWidget(), InVisibility=Visible)
  ]}
]
variables [
 AcceptButton {type=WBP_UIKit_ButtonCTA_Base_C, default=null}
 CancellationModalInfoClass {type=Class, default=/Crew/UI/Subscription/SubscriptionCancellationInfoModal.SubscriptionCancellationInfoModal_C}
 ExplainerTextLocalized {type=string, default=Fortine Crew Changes}
 IsExplainerVisible {type=boolean}
 IsSignUpStepsVisible {type=boolean}
 IsTermsVisible {type=boolean}
 ShouldShowSignUpSteps {type=boolean, default=true}
 TermsButton {type=WBP_UIKit_ButtonCTA_Base_C, default=null}
]
widgets {
 info {
  inheritedWidgetCount=0
  namedSlotCount=3
  parentClass=/Crew/UI/Subscription/SubscriptionModalBase.SubscriptionModalBase_C
  rootWidgetClass=&Overlay
  widgetCount=12
 }
 widgets [
  Overlay_31 {widgetClassPath=&Overlay, widget=Overlay_31}
  Dialog_Internal {
   widgetClassPath=/FortUILibrary/Systems/Dialog/WBP_UIKit_Dialog_Internal
   bIsVariable=true
   parent=Overlay_31
   widget=Dialog_Internal
   properties {
    designTimePreview_DialogVM=&DialogVM_NotificationModal_C
    useFooter=true
    useText=None
    buttons {
     each Button Horizontal Fill=false
     each Button Min Width=157
     each Button Size=Medium
     focuse Button=0
     space Between Buttons=24
     cTA Buttons Types [Loud,Regular]
     group Placement {
      horizontal Alignment=HAlign_Fill
      vertical Alignment=VAlign_Bottom
      grid {column=1, column Span=0, layer=0, row=4, row Span=0, nudge=<0 0>}
      padding=<ltrb 80 20 80 48>
     }
    }
    dialogSIze {
     height Override {enable=true, value=895}
     max Aspect Ratio {enable=false, value=0}
     max Desired Height {enable=true, value=800}
     max Desired Width {enable=false, value=0}
     min Aspect Ratio {enable=false, value=0}
     min Desired Height {enable=false, value=0}
     min Desired Width {enable=false, value=0}
     width Override {enable=true, value=1490}
    }
    footerTextProperties {
     auto Wrap Text=true
     justification=Left
     line Height Percentage=1.35
     mobile Font Size Multiplier=1.3
     transform Policy=None
     wrap Text At=785
     color {colorUseRule=UseColor_Specified, specifiedColor=<rgba 0.9216 0.9911 0.0103 1>}
     font {
      bForceMonospaced=false
      bMaterialIsStencil=false
      fontMaterial=null
      fontObject=&HeadingNow
      letterSpacing=1
      monospacedWidth=1
      size=11.25
      skewAmount=0
      typefaceFontName=74 Regular
      outlineSettings {
       bApplyOutlineToDropShadows=false
       bMiteredCorners=false
       bSeparateFillAlpha=false
       outlineMaterial=null
       outlineSize=0
       outlineColor=<rgba 0 0 0 1>
      }
     }
     placement {
      horizontal Alignment=HAlign_Left
      vertical Alignment=VAlign_Center
      grid {column=1, column Span=0, layer=0, row=2, row Span=0, nudge=<0 0>}
      padding=<ltrb 80 12 80 0>
     }
    }
    titleProperties {
     auto Wrap Text=true
     justification=Left
     line Height Percentage=1
     mobile Font Size Multiplier=1.3
     transform Policy=ToUpper
     wrap Text At=808
     color {colorUseRule=UseColor_Specified, specifiedColor=<rgba 1 1 1 1>}
     font {
      bForceMonospaced=false
      bMaterialIsStencil=false
      fontMaterial=null
      fontObject=&HeadingNow
      letterSpacing=0
      monospacedWidth=1
      size=25.5
      skewAmount=0
      typefaceFontName=86 Bold Italic
      outlineSettings {
       bApplyOutlineToDropShadows=false
       bMiteredCorners=false
       bSeparateFillAlpha=false
       outlineMaterial=null
       outlineSize=0
       outlineColor=<rgba 0 0 0 1>
      }
     }
     placement {
      horizontal Alignment=HAlign_Fill
      vertical Alignment=VAlign_Top
      grid {column=1, column Span=0, layer=0, row=0, row Span=0, nudge=<0 0>}
      padding=<ltrb 80 80 80 32>
     }
    }
   }
   slot {class=OverlaySlot, properties {horizontalAlignment=HAlign_Fill, verticalAlignment=VAlign_Fill}}
  }
  Overlay_0 {widgetClassPath=&Overlay, namedSlotHost=Dialog_Internal, widget=Overlay_0}
  LeftSideTextSwitcher {
   widgetClassPath=&CommonVisibilitySwitcher
   bIsVariable=true
   parent=Overlay_0
   widget=LeftSideTextSwitcher
   properties {bOverride_Cursor=true}
   slot {class=OverlaySlot, properties {horizontalAlignment=HAlign_Fill, verticalAlignment=VAlign_Fill, padding=<ltrb 80 0 80 0>}}
  }
  ExplainerScrollBox {
   widgetClassPath=/Script/FortniteUI.AthenaScrollBox
   bIsVariable=true
   parent=LeftSideTextSwitcher
   widget=ExplainerScrollBox
   slot {class=CommonVisibilitySwitcherSlot}
  }
  ExplainerText {
   widgetClassPath=&CommonRichTextBlock
   bIsVariable=true
   parent=ExplainerScrollBox
   widget=ExplainerText
   properties {
    autoWrapText=true
    bOverrideDefaultStyle=true
    lineHeightPercentage=0.875
    mobileTextBlockScale=1.2
    text="This is the <text color=\"FFFFFFFF\" fontface=\"black\">description</>, this is the wrapping policy when the string is long."
    defaultTextStyleOverride {
     overflowPolicy=Clip
     transformPolicy=None
     colorAndOpacity {colorUseRule=UseColor_Specified, specifiedColor=<rgba 0.5776 0.5776 0.5776 1>}
     font {
      bForceMonospaced=false
      bMaterialIsStencil=false
      fontMaterial=null
      fontObject=&HeadingNow
      letterSpacing=0
      monospacedWidth=1
      size=15.75
      skewAmount=0
      typefaceFontName=74 Regular
      outlineSettings {
       bApplyOutlineToDropShadows=false
       bMiteredCorners=false
       bSeparateFillAlpha=false
       outlineMaterial=null
       outlineSize=0
       outlineColor=<rgba 0 0 0 1>
      }
     }
     highlightColor {colorUseRule=UseColor_Specified, specifiedColor=<rgba 0 0 0 1>}
     highlightShape {
      bIsDynamicallyLoaded=false
      drawAs=Image
      imageType=NoImage
      mirroring=NoMirror
      resourceName=None
      resourceObject=null
      tiling=NoTile
      imageSize=<32 32>
      margin=<ltrb 0 0 0 0>
      outlineSettings {
       bUseBrushTransparency=false
       roundingType=HalfHeightRadius
       width=0
       color {colorUseRule=UseColor_Specified, specifiedColor=<rgba 0 0 0 0>}
       cornerRadii=<0 0 0 0>
      }
      tintColor {colorUseRule=UseColor_Specified, specifiedColor=<rgba 1 1 1 1>}
      uVRegion {bIsValid=false, max=<0 0>, min=<0 0>}
     }
     selectedBackgroundColor {colorUseRule=UseColor_Foreground, specifiedColor=<rgba 1 0 1 1>}
     shadowColorAndOpacity=<rgba 0 0 0 1>
     shadowOffset=<0 0>
     strikeBrush {
      bIsDynamicallyLoaded=false
      drawAs=Image
      imageType=NoImage
      mirroring=NoMirror
      resourceName=None
      resourceObject=null
      tiling=NoTile
      imageSize=<32 32>
      margin=<ltrb 0 0 0 0>
      outlineSettings {
       bUseBrushTransparency=false
       roundingType=HalfHeightRadius
       width=0
       color {colorUseRule=UseColor_Specified, specifiedColor=<rgba 0 0 0 0>}
       cornerRadii=<0 0 0 0>
      }
      tintColor {colorUseRule=UseColor_Specified, specifiedColor=<rgba 1 1 1 1>}
      uVRegion {bIsValid=false, max=<0 0>, min=<0 0>}
     }
     underlineBrush {
      bIsDynamicallyLoaded=false
      drawAs=Image
      imageType=NoImage
      mirroring=NoMirror
      resourceName=None
      resourceObject=null
      tiling=NoTile
      imageSize=<32 32>
      margin=<ltrb 0 0 0 0>
      outlineSettings {
       bUseBrushTransparency=false
       roundingType=HalfHeightRadius
       width=0
       color {colorUseRule=UseColor_Specified, specifiedColor=<rgba 0 0 0 0>}
       cornerRadii=<0 0 0 0>
      }
      tintColor {colorUseRule=UseColor_Specified, specifiedColor=<rgba 1 1 1 1>}
      uVRegion {bIsValid=false, max=<0 0>, min=<0 0>}
     }
    }
   }
   slot {class=ScrollBoxSlot, properties {verticalAlignment=VAlign_Center, size {sizeRule=Fill, value=1}}}
  }
  SubscriptionEulaWidget {
   widgetClassPath=/Crew/UI/Subscription/SubscriptionEulaWidget
   bIsVariable=true
   parent=LeftSideTextSwitcher
   widget=SubscriptionEulaWidget
   properties {bSupportsActivationFocus=false, eULAKeyToShow=fn-crew-terms}
   slot {class=CommonVisibilitySwitcherSlot}
  }
  SignUpStepsScrollBox {
   widgetClassPath=&AthenaScrollBox
   bIsVariable=true
   parent=LeftSideTextSwitcher
   widget=SignUpStepsScrollBox
   slot {class=CommonVisibilitySwitcherSlot}
  }
  SignUpStepsText {
   widgetClassPath=&CommonRichTextBlock
   bIsVariable=true
   parent=SignUpStepsScrollBox
   widget=SignUpStepsText
   properties {
    autoWrapText=true
    bOverrideDefaultStyle=true
    lineHeightPercentage=0.875
    mobileTextBlockScale=1.2
    text="This is the <text color=\"FFFFFFFF\" fontface=\"black\">description</>, this is the wrapping policy when the string is long."
    defaultTextStyleOverride {
     overflowPolicy=Clip
     transformPolicy=None
     colorAndOpacity {colorUseRule=UseColor_Specified, specifiedColor=<rgba 0.5776 0.5776 0.5776 1>}
     font {
      bForceMonospaced=false
      bMaterialIsStencil=false
      fontMaterial=null
      fontObject=&HeadingNow
      letterSpacing=0
      monospacedWidth=1
      size=15.75
      skewAmount=0
      typefaceFontName=74 Regular
      outlineSettings {
       bApplyOutlineToDropShadows=false
       bMiteredCorners=false
       bSeparateFillAlpha=false
       outlineMaterial=null
       outlineSize=0
       outlineColor=<rgba 0 0 0 1>
      }
     }
     highlightColor {colorUseRule=UseColor_Specified, specifiedColor=<rgba 0 0 0 1>}
     highlightShape {
      bIsDynamicallyLoaded=false
      drawAs=Image
      imageType=NoImage
      mirroring=NoMirror
      resourceName=None
      resourceObject=null
      tiling=NoTile
      imageSize=<32 32>
      margin=<ltrb 0 0 0 0>
      outlineSettings {
       bUseBrushTransparency=false
       roundingType=HalfHeightRadius
       width=0
       color {colorUseRule=UseColor_Specified, specifiedColor=<rgba 0 0 0 0>}
       cornerRadii=<0 0 0 0>
      }
      tintColor {colorUseRule=UseColor_Specified, specifiedColor=<rgba 1 1 1 1>}
      uVRegion {bIsValid=false, max=<0 0>, min=<0 0>}
     }
     selectedBackgroundColor {colorUseRule=UseColor_Foreground, specifiedColor=<rgba 1 0 1 1>}
     shadowColorAndOpacity=<rgba 0 0 0 1>
     shadowOffset=<0 0>
     strikeBrush {
      bIsDynamicallyLoaded=false
      drawAs=Image
      imageType=NoImage
      mirroring=NoMirror
      resourceName=None
      resourceObject=null
      tiling=NoTile
      imageSize=<32 32>
      margin=<ltrb 0 0 0 0>
      outlineSettings {
       bUseBrushTransparency=false
       roundingType=HalfHeightRadius
       width=0
       color {colorUseRule=UseColor_Specified, specifiedColor=<rgba 0 0 0 0>}
       cornerRadii=<0 0 0 0>
      }
      tintColor {colorUseRule=UseColor_Specified, specifiedColor=<rgba 1 1 1 1>}
      uVRegion {bIsValid=false, max=<0 0>, min=<0 0>}
     }
     underlineBrush {
      bIsDynamicallyLoaded=false
      drawAs=Image
      imageType=NoImage
      mirroring=NoMirror
      resourceName=None
      resourceObject=null
      tiling=NoTile
      imageSize=<32 32>
      margin=<ltrb 0 0 0 0>
      outlineSettings {
       bUseBrushTransparency=false
       roundingType=HalfHeightRadius
       width=0
       color {colorUseRule=UseColor_Specified, specifiedColor=<rgba 0 0 0 0>}
       cornerRadii=<0 0 0 0>
      }
      tintColor {colorUseRule=UseColor_Specified, specifiedColor=<rgba 1 1 1 1>}
      uVRegion {bIsValid=false, max=<0 0>, min=<0 0>}
     }
    }
   }
   slot {class=ScrollBoxSlot, properties {verticalAlignment=VAlign_Center, size {sizeRule=Fill, value=1}}}
  }
  RichTextButton {
   widgetClassPath=&/RichTextButton
   bIsVariable=true
   namedSlotHost=Dialog_Internal
   widget=RichTextButton
   properties {
    button text="[PH] URL Text"
    holdData=&CommonUI_HoldData_Base_C
    onButtonBaseClicked=(&self:WidgetTree.RichTextButton.HandleOnButtonClicked)
    onButtonBaseDoubleClicked=(&self:WidgetTree.RichTextButton.HandleOnButtonDoubleClicked)
    onButtonBaseHovered=(&self:WidgetTree.RichTextButton.HandleOnButtonHovered)
    onButtonBaseUnhovered=(&self:WidgetTree.RichTextButton.HandleOnButtonUnhovered)
    onSelectedChangedBase=(&self:WidgetTree.RichTextButton.HandleOnSelectedChanged)
    style=/FortUILibrary/UIKit/Buttons/Blocks/ButtonStyle_UIKit_Empty.ButtonStyle_UIKit_Empty_C
    text Color {colorUseRule=UseColor_Specified, specifiedColor=<rgba 0.9301 1 0.0103 1>}
   }
  }
  ImageContent {widgetClassPath=&VerticalBox, namedSlotHost=Dialog_Internal, widget=ImageContent}
  WBP_SubscriberSince {
   widgetClassPath=/Crew/UI/Subscription/WBP_SubscriberSince
   bIsVariable=true
   parent=ImageContent
   widget=WBP_SubscriberSince
   slot {class=VerticalBoxSlot, properties {verticalAlignment=VAlign_Center, padding=<ltrb 0 0 80 0>, size {sizeRule=Fill, value=1}}}
  }
 ]
}
