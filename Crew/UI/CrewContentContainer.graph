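# Crew subscription UI container widget that shows a top row of skewed tab buttons over a switcher of content pages.
# The native CrewSubscriptionContentContainer parent supplies the switcher, tab container and tab entry box, flanked here by left and right shoulder action prompts.
# On activation, the graph sets every tab button's text to white.
# The OnTabSelected event sets the switcher's active index to the selected tab, with zero transition duration.
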WidgetBlueprint /Crew/UI/CrewContentContainer
graphs [EventGraph {nodes [
   Event On Activated()
   > n1 = For Each Loop(Array:Array of User Widget Object References=GetAllEntries!(self:Dynamic Entry Box Base Object Reference=Get EntryBox_Tabs() -> (EntryBox_Tabs:Dynamic Entry Box Object Reference)) -> (ReturnValue:Array of User Widget Object References)) -> (Array Element:User Widget Object Reference)
   > n2 = Cast To BP_SkewedTabButton!(execute, Object=@n1) -> (then, CastFailed, AsBP Skewed Tab Button:BP Skewed Tab Button Object Reference) {is=title K2Node_DynamicCast}
   > SetTextColor!(execute, self:BP Skewed Tab Button Object Reference=@n2, Color:Linear Color Structure=(R=1,G=1,B=1,A=1)) -> (then)
   n3 = Event OnTabSelected!() -> (OutputDelegate, then, TabIndex:int=0) {is=title K2Node_Event}
   > SetActiveWidgetIndex(self=Get Switcher_Tabs() -> (Switcher_Tabs:Common Activatable Widget Switcher Object Reference), Index=@n3.TabIndex)
  ]}]
variables []
widgets {
 info {inheritedWidgetCount=3, parentClass=/Script/CrewUI.CrewSubscriptionContentContainer, rootWidgetClass=&Overlay, widgetCount=9}
 widgets [
  Overlay_0 {widgetClassPath=&Overlay, widget=Overlay_0}
  Switcher_Tabs {
   widgetClassPath=&CommonActivatableWidgetSwitcher
   bInherited=true
   parent=Overlay_0
   widget=Switcher_Tabs
   properties {clipping=ClipToBounds, transitionDuration=0}
   slot {class=OverlaySlot, properties {horizontalAlignment=HAlign_Fill, verticalAlignment=VAlign_Fill}}
  }
  Widget_TabsContainer {
   widgetClassPath=&Border
   bInherited=true
   parent=Overlay_0
   widget=Widget_TabsContainer
   properties {brushColor=<rgba 0 0 0 0>, padding=<ltrb 0 0 0 0>}
   slot {class=OverlaySlot, properties {horizontalAlignment=HAlign_Fill}}
  }
  Border_0 {
   widgetClassPath=&Border
   parent=Widget_TabsContainer
   widget=Border_0
   properties {brushColor=<rgba 0.0052 0.0052 0.0056 0.15>, padding=<ltrb 0 0 0 0>}
   slot {class=BorderSlot, properties {padding=<ltrb 0 0 0 0>}}
  }
  SafeZone_0 {
   widgetClassPath=&SafeZone
   bIsVariable=true
   parent=Border_0
   widget=SafeZone_0
   properties {padBottom=false, padRight=false, padTop=false}
   slot {class=BorderSlot, properties {padding=<ltrb 0 0 0 0>}}
  }
  HorizontalBox_136 {widgetClassPath=&HorizontalBox, parent=SafeZone_0, widget=HorizontalBox_136, slot {class=SafeZoneSlot}}
  PreviousTabAction {
   widgetClassPath=&CommonActionWidget
   bIsVariable=true
   parent=HorizontalBox_136
   widget=PreviousTabAction
   properties {inputActions [{dataTable=&InputActionDataTable, rowName=GenericLeftShoulder}]}
   slot {class=HorizontalBoxSlot, properties {padding=<ltrb 16 0 16 0>}}
  }
  EntryBox_Tabs {
   widgetClassPath=&DynamicEntryBox
   bInherited=true
   bIsVariable=true
   parent=HorizontalBox_136
   widget=EntryBox_Tabs
   properties {entryWidgetClass=/Game/Athena/HUD/BP_SkewedTabButton.BP_SkewedTabButton_C}
   slot {class=HorizontalBoxSlot}
  }
  NextTabAction {
   widgetClassPath=&CommonActionWidget
   bIsVariable=true
   parent=HorizontalBox_136
   widget=NextTabAction
   properties {inputActions [{dataTable=&InputActionDataTable, rowName=GenericRightShoulder}]}
   slot {class=HorizontalBoxSlot, properties {padding=<ltrb 16 0 0 0>}}
  }
 ]
}
