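# Widget Blueprint for the standalone Crew subscription screen, which holds a loading view and the Crew content container in a switcher.
# The loading view shows a load guard over a fake locker background.
# When the camera level loads, it sets the front-end camera to Home without blending, then switches to the content, with zero-second delays before each step.
# A safe-zoned, top-right close button appears only for touch input.
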WidgetBlueprint /Crew/UI/StandaloneCrewContentContainer
graphs [EventGraph {nodes [
   Event OnCameraLevelLoaded() {is=title K2Node_Event}
   > Delay(Duration=0)
   > SetFrontEndCamera!(
    execute
    self:Fort Front End Context Object Reference=Get FortFrontEndContext() -> (ReturnValue:Fort Front End Context Object Reference) {is=title K2Node_GetSubsystem}
    NewState:EFrontEndCamera Enum=Home
    TransitionParams:View Target Transition Params Structure=(BlendTime=0,BlendFunction=VTBlend_Cubic,BlendExp=2,bLockOutgoing=False)
   ) -> (then)
   > Delay(Duration=0)
   > SetActiveWidget(
    self:Widget Switcher Object Reference=Get Switcher_TabContent() -> (Switcher_TabContent:Common Widget Switcher Legacy Object Reference)
    Widget=Get Widget_CrewContentContainer() -> (Widget_CrewContentContainer:Crew Subscription Content Container Object Reference)
   )
  ]}]
variables []
widgets {
 info {
  inheritedWidgetCount=4
  parentClass=/Script/CrewUI.CrewStandaloneSubscriptionContentContainer
  rootWidgetClass=&Overlay
  widgetCount=9
 }
 widgets [
  Overlay_0 {widgetClassPath=&Overlay, widget=Overlay_0}
  Switcher_TabContent {
   widgetClassPath=&CommonWidgetSwitcherLegacy
   bInherited=true
   bIsVariable=true
   parent=Overlay_0
   widget=Switcher_TabContent
   slot {class=OverlaySlot, properties {horizontalAlignment=HAlign_Fill, verticalAlignment=VAlign_Fill}}
  }
  Content_LoadingIndicator {
   widgetClassPath=&Overlay
   bInherited=true
   parent=Switcher_TabContent
   widget=Content_LoadingIndicator
   slot {class=WidgetSwitcherSlot}
  }
  Border_0 {
   widgetClassPath=&Border
   parent=Content_LoadingIndicator
   widget=Border_0
   properties {
    background {
     bIsDynamicallyLoaded=false
     drawAs=Image
     imageType=NoImage
     mirroring=NoMirror
     resourceName=None
     resourceObject=/FortniteCoreUI/Frontend/Locker/FakeLockerBackground
     tiling=NoTile
     imageSize=<1024 512>
     margin=<ltrb 0 0 0 0>
     outlineSettings {
      bUseBrushTransparency=false
      roundingType=HalfHeightRadius
      width=0
      color {colorUseRule=UseColor_Specified, specifiedColor=<rgba 0 0 0 0>}
      cornerRadii=<0 0 0 0>
     }
     tintColor {colorUseRule=UseColor_Specified, specifiedColor=<rgba 1 1 1 1>}
     uVRegion {bIsValid=false, max=<0 0>, min=<0 0>}
    }
    padding=<ltrb 0 0 0 0>
   }
   slot {class=OverlaySlot, properties {horizontalAlignment=HAlign_Fill, verticalAlignment=VAlign_Fill}}
  }
  LoadGuard {
   widgetClassPath=/Script/CommonUI.CommonLoadGuard
   parent=Border_0
   widget=LoadGuard
   properties {bShowLoading=true, textStyle=&TextStyle-Base_C}
   slot {class=BorderSlot, properties {padding=<ltrb 0 0 0 0>}}
  }
  Widget_CrewContentContainer {
   widgetClassPath=&/CrewContentContainer
   bInherited=true
   bIsVariable=true
   parent=Switcher_TabContent
   widget=Widget_CrewContentContainer
   slot {class=WidgetSwitcherSlot}
  }
  SafeZone_1 {
   widgetClassPath=&SafeZone
   bIsVariable=true
   parent=Overlay_0
   widget=SafeZone_1
   properties {padBottom=false, padTop=false}
   slot {class=OverlaySlot, properties {horizontalAlignment=HAlign_Fill, verticalAlignment=VAlign_Fill}}
  }
  VisibilityBorder_Touch {
   widgetClassPath=&CommonHardwareVisibilityBorder
   parent=SafeZone_1
   widget=VisibilityBorder_Touch
   properties {
    padding=<ltrb 0 30 30 0>
    visibilityQuery {
     autoDescription=" ANY( Input.Touch )"
     tokenStreamVersion=0
     userDescription=if using touch
     queryTokenStream [0,1,1,1,0]
     tagDictionary [{tagName=Input.Touch}]
    }
   }
   slot {class=SafeZoneSlot, properties {hAlign=HAlign_Right, vAlign=VAlign_Top}}
  }
  Button_CloseTouch {
   widgetClassPath=/FortUILibrary/Foundation/Buttons/CloseButton
   bInherited=true
   bIsVariable=true
   parent=VisibilityBorder_Touch
   widget=Button_CloseTouch
   properties {
    onButtonBaseClicked=(&self:WidgetTree.Button_CloseTouch.HandleOnButtonClicked)
    onButtonBaseDoubleClicked=(&self:WidgetTree.Button_CloseTouch.HandleOnButtonDoubleClicked)
    onButtonBaseHovered=(&self:WidgetTree.Button_CloseTouch.HandleOnButtonHovered)
    onButtonBaseUnhovered=(&self:WidgetTree.Button_CloseTouch.HandleOnButtonUnhovered)
    onSelectedChangedBase=(&self:WidgetTree.Button_CloseTouch.HandleOnSelectedChanged)
   }
   slot {class=BorderSlot, properties {padding=<ltrb 0 30 30 0>}}
  }
 ]
}
