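# Crew subscription widget showing the selected benefit's details, or a general Crew summary when none is set, above subscription buttons.
# HandleSubState switches between join/rejoin, manage and unavailable views, adjusting labels, billing note and a platform store icon injected into SubButton.
# SubButton broadcasts PurchaseClicked with its hold time, requiring hold except on mobile.
# Cancellation and resolve-issue buttons async-queue modals fed by CrewVM view models; battle pass benefits show a season pass badge.
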WidgetBlueprint /Crew/UI/CrewBenefitInfo
graphs [
 EventGraph {nodes [
   n1 = "On Clicked (SubButton)"!() -> (OutputDelegate, then, Button:Common Button Base Object Reference) {is=title K2Node_ComponentBoundEvent}
   > n2 = GetPurchaseTraceId!(execute) -> (then, ReturnValue:Flow Analytics Additional Info Structure)
   > Fire Event Element Interacted(Element=@n1.Button, AdditionalInfo=@n2)
   > Call Purchase Clicked!(
    execute:Exec
    self:Crew Benefit Info Object Reference
    HoldTime:double=GetRequiredHoldTime!(self:Common Button Base Object Reference=@n1.Button) -> (ReturnValue:float)
   ) -> (then:Exec) {is=title K2Node_CallDelegate}
   n3 = "On Clicked (CancelButton)"!() -> (OutputDelegate, then, Button:Common Button Base Object Reference) {is=title K2Node_ComponentBoundEvent}
   > n4 = GetPurchaseTraceId()
   > Fire Event Element Interacted(Element=@n3.Button, AdditionalInfo=@n4)
   > n5 = QueueModalWidgetSoftClassAsync!(
    execute:Exec
    SoftWidgetClass:Common Activatable Widget Soft Class Reference=/Crew/UI/Subscription/SubscriptionCancellationInfoModal.SubscriptionCancellationInfoModal_C
   ) -> (then:Exec, OnModalWidgetQueued:Exec, ActivatableWidget:Common Activatable Widget Object Reference) {is=title K2Node_AsyncAction}
   n6 = Cast To SubscriptionCancellationInfoModal!(execute=@n5.OnModalWidgetQueued, Object=@n5) -> (then, CastFailed, AsSubscription Cancellation Info Modal:Subscription Cancellation Info Modal Object Reference) {is=title K2Node_DynamicCast}
   > Set CrewSubscriptionModalInfoVM!(
    execute
    CrewSubscriptionModalInfoVM:Crew Subscription Modal Info VM Object Reference=GetModalViewModelById!(
     self:Crew VM Object Reference=Get CrewVM() -> (CrewVM:Crew VM Object Reference)
     ModalId:String=Get ModalId!(self:Subscription Modal Base Object Reference=@n6) -> (ModalId:String)
    ) -> (ReturnValue:Crew Subscription Modal Info VM Object Reference)
    self:Subscription Cancellation Info Modal Object Reference=@n6
   ) -> (then, Output_Get:Crew Subscription Modal Info VM Object Reference)
   > Set CrewCancellationInfoVM!(
    execute
    CrewCancellationInfoVM:Crew Cancellation Info VM Object Reference=Get CancellationInfo!(self:Crew VM Object Reference=Get CrewVM()) -> (CancellationInfo:Crew Cancellation Info VM Object Reference)
    self:Subscription Cancellation Info Modal Object Reference=@n6
   ) -> (then, Output_Get:Crew Cancellation Info VM Object Reference)
   n7 = "On Clicked (ResolveIssue_Button)"!() -> (OutputDelegate, then, Button:Common Button Base Object Reference) {is=title K2Node_ComponentBoundEvent}
   > n8 = GetPurchaseTraceId()
   > Fire Event Element Interacted(Element=@n7.Button, AdditionalInfo=@n8)
   > n9 = Async Load Asset(Asset=/Crew/UI/Subscription/SubscriptionResolveIssueModal)
   n10 = QueueModalWidgetClassAsync!(
    execute:Exec=@n9.Completed
    WidgetClass:Common Activatable Widget Class Reference=/Crew/UI/Subscription/SubscriptionResolveIssueModal.SubscriptionResolveIssueModal_C
   ) -> (then:Exec, OnModalWidgetQueued:Exec, ActivatableWidget:Common Activatable Widget Object Reference) {is=title K2Node_AsyncAction}
   n11 = Cast To SubscriptionResolveIssueModal!(execute=@n10.OnModalWidgetQueued, Object=@n10) -> (then, CastFailed, AsSubscription Resolve Issue Modal:Subscription Resolve Issue Modal Object Reference) {is=title K2Node_DynamicCast}
   > Set CrewSubscriptionModalInfoVM(
    CrewSubscriptionModalInfoVM=GetModalViewModelById(self=Get CrewVM(), ModalId=Get ModalId(self=@n11))
    self:Subscription Resolve Issue Modal Object Reference=@n11
   )
   "On Clicked (Button_OpenParentalControls)"!() -> (OutputDelegate, then, Button:Common Button Base Object Reference) {is=title K2Node_ComponentBoundEvent}
   > RequestToOpenParentalControls!(execute) -> (then)
   Event PreConstruct()
   > n12 = Get!(execute:Exec) -> (then:Exec, else:Exec, PlatformIconWidget:WBP Price Platform Store Icon Object Reference)
   n13 = Create Widget!(
    execute:Exec=@n12.else
    Class:User Widget Class Reference=/FortUILibrary/Systems/ItemShopTile/WBP_Price_PlatformStoreIcon.WBP_Price_PlatformStoreIcon_C
    OwningPlayer:Player Controller Object Reference=GetOwningPlayer()
    PreviewPlayStationPlatform:bool=Get PreviewPlatformIcon() -> (PreviewPlatformIcon:bool)
    ImageSize:Vector 2D Structure=Get PlatformIconSize() -> (PlatformIconSize:Vector 2D Structure)
   ) -> (then:Exec, ReturnValue:WBP Price Platform Store Icon Object Reference)
   > n14 = Set PlatformIconWidget(PlatformIconWidget:WBP Price Platform Store Icon Object Reference=@n13) -> (Output_Get:WBP Price Platform Store Icon Object Reference)
   > n15 = AddChildToGrid!(
    execute
    self:Grid Panel Object Reference=Get Root!(self:WBP UIKit Button CTA Base Object Reference=Get SubButton() -> (SubButton:WBP UIKit Button Loud Object Reference)) -> (Root:Grid Panel Object Reference)
    Content:Widget Object Reference=@n14
    InRow:int=1
    InColumn:int=4
   ) -> (then, ReturnValue:Grid Slot Object Reference)
   > Set VerticalAlignment!(execute, VerticalAlignment:EVerticalAlignment Enum=VAlign_Center, self:Grid Slot Object Reference=@n15) -> (then, Output_Get:EVerticalAlignment Enum=VAlign_Fill)
   > Set HorizontalAlignment!(execute, HorizontalAlignment:EHorizontalAlignment Enum=HAlign_Center, self:Grid Slot Object Reference=@n15) -> (then, Output_Get:EHorizontalAlignment Enum=HAlign_Fill)
   > SetGridSlotPadding!(
    execute
    Widget:Widget Object Reference=Get PlatformIconWidget() -> (PlatformIconWidget:WBP Price Platform Store Icon Object Reference)
    Padding:Margin Structure=Make Margin(Left=4, Top=0, Right=8, Bottom=0)
   ) -> (then)
   Event Construct()
   > SetRequiresHold!(execute, self:Common Button Base Object Reference=Get SubButton(), bInRequiresHold:bool=NOT Boolean(A=IsMobileGame())) -> (then)
  ]}
 GetDesiredFocusTarget {nodes [
   GetDesiredFocusTarget() {is=title K2Node_FunctionEntry}
   n1 = Get ButtonSwitcher() -> (ButtonSwitcher:Common Visibility Switcher Object Reference)
   n2 = GetActiveWidgetIndex!(self:Common Visibility Switcher Object Reference=Get PurchaseAvailabilitySwitcher() -> (PurchaseAvailabilitySwitcher:Common Visibility Switcher Object Reference)) -> (ReturnValue:int=0)
   n3 = Get Button_OpenParentalControls() -> (Button_OpenParentalControls:WBP UIKit Button Loud Object Reference)
   n4 = Get SubButton()
   n5 = Get ResolveIssue_Button() -> (ResolveIssue_Button:WBP UIKit Button Loud Object Reference)
   n6 = IsVisible(self=@n5)
   n7 = Get CancelButton() -> (CancelButton:WBP UIKit Button Regular Object Reference)
   > Return Node!(
    execute
    Desired:Widget Object Reference=Select(
     Option 0:Widget Object Reference=Select(Option 0=@n4, Option 1=@n3, Index=@n2)
     Option 1:Widget Object Reference=Select(Option 0=@n7, Option 1=@n5, Index:bool=@n6)
     Index:int=GetActiveWidgetIndex(self=@n1) -> (ReturnValue=0)
    ) -> (ReturnValue:Widget Object Reference)
   )
  ]}
 HandleSubState {nodes [
   HandleSubState!() -> (then, SubState:EMcpSubscriptionState Enum=Inactive) {is=title K2Node_FunctionEntry}
   > UpdatePlatformStoreIcon!(execute, self, ShouldShow:bool=false) -> (then)
   > n1 = Switch on EMcpSubscriptionState!(execute, Selection:EMcpSubscriptionState Enum=Get SubState() -> (SubState:EMcpSubscriptionState Enum=Inactive)) -> (
    Inactive:Exec
    Active:Exec
    ActiveWithExternalBilling:Exec
    NotAutoRenewable:Exec
    Canceled:Exec
    PaymentProcessError:Exec
    PlatformAuthError:Exec
    RewardsDelayed:Exec
    Unavailable:Exec
    Unknown:Exec
   ) {is=title K2Node_SwitchEnum}
   n2 = SetActiveWidgetIndex(
    execute=[@n1.Active, @n1.ActiveWithExternalBilling, @n1.PaymentProcessError, @n1.PlatformAuthError, @n1.RewardsDelayed]
    self:Common Visibility Switcher Object Reference=Get ButtonSwitcher()
    Index=1
   )
   n3 = Switch on EMcpSubscriptionState(execute=[@n2.then, @n11.then, @n12.then], Selection=Get SubState() -> (SubState=Inactive))
   n4 = SetVisibility(
    execute=[@n3.Active, @n3.Canceled, @n3.NotAutoRenewable, @n3.PlatformAuthError, @n3.RewardsDelayed, @n3.Unavailable, @n3.Unknown]
    self=[Get PriceNextBill() -> (PriceNextBill:Common Text Object Reference), Get ResolveIssue_Button()]
    InVisibility=Collapsed
   )
   n5 = Switch on EMcpSubscriptionState(execute=[@n9.then, @n3.Inactive, @n10.then, @n4.then], Selection=Get SubState() -> (SubState=Inactive))
   Set Text(execute=@n5.Canceled, Text="LOCTEXT(\"Rejoin Crew\")", self:Fort CTAButton Object Reference=Get SubButton())
   > n6 = Set SecondaryText!(execute, SecondaryText:Text, self:Fort CTAButton Object Reference=Get SubButton()) -> (then, Output_Get:Text)
   n7 = Get CrewVM()
   SetVisibility(
    execute=[@n8.then, @n6.then]
    self=Get SubButton()
    InVisibility=Select(
     Option 0:ESlateVisibility Enum=Hidden
     Option 1:ESlateVisibility Enum=SelfHitTestInvisible
     Index=IsOnActiveSubscriptionPlatform!(self:Crew VM Object Reference=@n7) -> (ReturnValue:bool)
    ) -> (ReturnValue:ESlateVisibility Enum)
   )
   Set Text(execute=@n5.Inactive, Text="LOCTEXT(\"Join Crew Now\")", self=Get SubButton())
   > Set SecondaryText(SecondaryText=Get PriceText!(self:Crew VM Object Reference=Get CrewVM()) -> (PriceText:Text), self=Get SubButton())
   > UpdatePlatformStoreIcon(ShouldShow=true)
   > SetVisibility(self=Get SubButton(), InVisibility=SelfHitTestInvisible)
   Set Text(execute=@n5.NotAutoRenewable, Text="LOCTEXT(\"Continue Membership\")", self=Get SubButton())
   > Set SecondaryText(SecondaryText=Get PriceText(self=Get CrewVM()), self=Get SubButton())
   > n8 = UpdatePlatformStoreIcon(ShouldShow=true)
   "SetText (Text)"(execute=@n3.ActiveWithExternalBilling, self=Get PriceNextBill(), InText="LOCTEXT(\"Check your subscription platform for details.\")")
   > SetVisibility(self=Get ResolveIssue_Button(), InVisibility=Collapsed)
   > n9 = SetVisibility(self=Get PriceNextBill(), InVisibility=SelfHitTestInvisible)
   SetVisibility(execute=@n3.PaymentProcessError, self=Get ResolveIssue_Button(), InVisibility=SelfHitTestInvisible)
   > n10 = SetVisibility(self=Get PriceNextBill(), InVisibility=Collapsed)
   n11 = SetActiveWidgetIndex(execute=[@n1.Canceled, @n1.Inactive, @n1.NotAutoRenewable, @n1.Unknown], self=Get ButtonSwitcher(), Index=0)
   n12 = SetActiveWidgetIndex(execute=@n1.Unavailable, self=Get ButtonSwitcher(), Index=2)
  ]}
 InitializeSeasonPassSystemVM {nodes [
   InitializeSeasonPassSystemVM!() -> (then, Passes:Array of Season Pass VM Object References) {is=title K2Node_FunctionEntry}
   > SetSeasonPassVMFromBenefit!(execute, self, Benefit:Crew Benefit VM Object Reference=Get CrewBenefitVM() -> (CrewBenefitVM:Crew Benefit VM Object Reference)) -> (then)
  ]}
 PurchaseClicked {nodes [PurchaseClicked!() -> (then, HoldTime:double=~) {is=title K2Node_FunctionEntry}]}
 RefreshBadgeVisibility {nodes [
   RefreshBadgeVisibility() {is=title K2Node_FunctionEntry}
   n1 = Get CrewBenefitVM()
   > Branch(Condition=IsValid(Object=@n1))
   > SwitchOnType!(execute, self, Type:ECrewTileType Enum=GetType!(self:Crew Benefit VM Object Reference=@n1) -> (ReturnValue:ECrewTileType Enum=None)) -> (then)
  ]}
 SetBenefit {nodes [
   n1 = SetBenefit!() -> (then, Benefit:Crew Benefit VM Object Reference) {is=title K2Node_FunctionEntry}
   > Set CrewBenefitVM(CrewBenefitVM:Crew Benefit VM Object Reference=@n1) -> (Output_Get:Crew Benefit VM Object Reference)
   > n2 = Branch(Condition=IsValid(Object=@n1))
   n3 = Get CrewLogo() -> (CrewLogo:Image Object Reference)
   > SetVisibility(self=@n3, InVisibility=Collapsed)
   > SetVisibility(
    self=[Get BenefitType() -> (BenefitType:Common Text Object Reference), Get BenefitName() -> (BenefitName:Common Text Object Reference)]
    InVisibility=SelfHitTestInvisible
   )
   > SetSeasonPassVMFromBenefit(Benefit=Get Benefit() -> (Benefit:Crew Benefit VM Object Reference))
   > RefreshBadgeVisibility() {is=title K2Node_CallFunction}
   n4 = Get CrewVM()
   n5 = IsSubscribed!(self:Crew VM Object Reference=@n4) -> (ReturnValue:bool)
   Set Text(
    execute=@n2.else
    Text=SelectText(B=Get ShortDescription!(self:Crew VM Object Reference=@n4) -> (ShortDescription:Text), bPickA=@n5)
    self:Rich Text Block Object Reference=Get BenefitDescription() -> (BenefitDescription:Common Rich Text Block Object Reference)
   )
   > SetVisibility(self=@n3, InVisibility=SelfHitTestInvisible)
   > SetVisibility(self=Get BenefitType(), InVisibility=Select(Option 0=SelfHitTestInvisible, Option 1=Collapsed, Index=@n5))
   > SetVisibility(self=Get BenefitName(), InVisibility=Collapsed)
   > Set Text(Text="LOCTEXT(\"Monthly Subscription\")", self:Text Object Reference=Get BenefitType())
   > SwitchOnType(Type=None)
  ]}
 SetPrice {nodes [
   n1 = SetPrice!() -> (then, CrewPrice:Text) {is=title K2Node_FunctionEntry}
   n2 = Get SubscriptionState!(self:Crew VM Object Reference=Get CrewVM()) -> (SubscriptionState:EMcpSubscriptionState Enum=Inactive)
   n3 = TextToUpper!(InText:"Text (by ref)"=Get PriceTemplate() -> (PriceTemplate:Text)) -> (ReturnValue:Text)
   > SetSecondaryText!(
    execute
    self:Fort CTAButton Object Reference=Get SubButton()
    ButtonSecondaryText:"Text (by ref)"=SelectText(
     B=Format Text!(Format=@n3, PRICE:Text=@n1) -> (Result)
     bPickA="Equal (Enum)"(A:EMcpSubscriptionState Enum=@n2, B:EMcpSubscriptionState Enum=Canceled)
    )
   ) -> (then)
   > n4 = Get!(execute, self:WBP UIKit Button CTA Base Object Reference=Get SubButton()) -> (then, else, Block_SecondaryText:WBP UIKit Block Text Object Reference)
   > SetTransformPolicy!(execute, self:WBP UIKit Block Text Object Reference=@n4, Transform Policy:ETextTransformPolicy Enum=None) -> (then)
  ]}
 SetSeasonPassVMFromBenefit {nodes [
   n1 = SetSeasonPassVMFromBenefit!() -> (then, Benefit:Crew Benefit VM Object Reference) {is=title K2Node_FunctionEntry}
   n2 = Get SeasonPassSystemVM() -> (SeasonPassSystemVM:Season Pass System VM Object Reference)
   > Branch(Condition=AND Boolean(A=IsValid(Object=@n1), B=IsValid(Object=@n2)))
   > Set SeasonPassVM(SeasonPassVM:Season Pass VM Object Reference=FindSeasonPass!(
     self:Season Pass System VM Object Reference=Get SeasonPassSystemVM()
     TemplateId:String=GetTemplateId!(self:Crew Benefit VM Object Reference=@n1) -> (ReturnValue:String)
    ) -> (ReturnValue:Season Pass VM Object Reference)) -> (Output_Get:Season Pass VM Object Reference)
  ]}
 SwitchOnType {nodes [
   n1 = SwitchOnType!() -> (then, Type:ECrewTileType Enum=None) {is=title K2Node_FunctionEntry}
   > n2 = Switch on ECrewTileType!(execute, Selection:ECrewTileType Enum=@n1) -> (None:Exec, Basic:Exec, BattlePass:Exec, CrewPack:Exec, ProgressiveCosmetic:Exec) {is=title K2Node_SwitchEnum}
   n3 = Get SeasonPassBadgeIcon() -> (SeasonPassBadgeIcon:Season Pass Badge Icon Object Reference)
   SetVisibility(execute=[@n4.else, @n2.Basic, @n2.CrewPack, @n2.None, @n2.ProgressiveCosmetic], self=@n3, InVisibility=Collapsed)
   n4 = Branch(execute=@n2.BattlePass, Condition=IsValid(Object=Get SeasonPassVM() -> (SeasonPassVM:Season Pass VM Object Reference)))
   > SetVisibility(self=@n3, InVisibility=SelfHitTestInvisible)
  ]}
 UpdatePlatformStoreIcon {nodes [
   n1 = UpdatePlatformStoreIcon!() -> (then, ShouldShow:bool=~) {is=title K2Node_FunctionEntry}
   > Set Visibility(Visibility=Select(Option 0=Collapsed, Option 1=SelfHitTestInvisible, Index=@n1), self=Get PlatformIconWidget())
  ]}
]
variables [
 PlatformIconSize {type=Vector2D, default=<40 40>}
 PlatformIconWidget {type=WBP_Price_PlatformStoreIcon_C, default=null}
 PreviewPlatformIcon {type=boolean}
 PurchaseClicked {}
]
widgets {
 info {inheritedWidgetCount=0, parentClass=&UserWidget, rootWidgetClass=&SizeBox, widgetCount=21}
 widgets [
  SizeBox_0 {widgetClassPath=&SizeBox, widget=SizeBox_0, properties {bOverride_WidthOverride=true, heightOverride=520, widthOverride=704}}
  VerticalBox_0 {widgetClassPath=&VerticalBox, parent=SizeBox_0, widget=VerticalBox_0, slot {class=SizeBoxSlot}}
  VerticalBox_60 {
   widgetClassPath=&VerticalBox
   parent=VerticalBox_0
   widget=VerticalBox_60
   slot {class=VerticalBoxSlot, properties {verticalAlignment=VAlign_Bottom, size {sizeRule=Fill, value=1}}}
  }
  CrewLogo {
   widgetClassPath=&Image
   bIsVariable=true
   parent=VerticalBox_60
   widget=CrewLogo
   properties {
    visibility=SelfHitTestInvisible
    brush {
     bIsDynamicallyLoaded=false
     drawAs=Image
     imageType=FullColor
     mirroring=NoMirror
     resourceName=None
     resourceObject=/Game/Athena/UI/Logos/Textures/M_UI_CrewLogo_Large
     tiling=NoTile
     imageSize=<432 128>
     margin=<ltrb 0 0 0 0>
     outlineSettings {
      bUseBrushTransparency=false
      roundingType=HalfHeightRadius
      width=0
      color {colorUseRule=UseColor_Specified, specifiedColor=<rgba 0 0 0 0>}
      cornerRadii=<0 0 0 0>
     }
     tintColor {colorUseRule=UseColor_Specified, specifiedColor=<rgba 1 1 1 1>}
     uVRegion {bIsValid=false, max=<0 0>, min=<0 0>}
    }
   }
   slot {class=VerticalBoxSlot, properties {horizontalAlignment=HAlign_Center, verticalAlignment=VAlign_Top, padding=<ltrb 0 0 0 40>}}
  }
  SeasonPassBadgeIcon {
   widgetClassPath=/SeasonPass/Widgets/SeasonPassBadgeIcon
   bIsVariable=true
   parent=VerticalBox_60
   widget=SeasonPassBadgeIcon
   properties {overrideToOwnPass=true, size=<108 108>}
   slot {class=VerticalBoxSlot, properties {horizontalAlignment=HAlign_Center, verticalAlignment=VAlign_Center, padding=<ltrb 0 0 0 8>}}
  }
  BenefitName {
   widgetClassPath=&CommonTextBlock
   bIsVariable=true
   parent=VerticalBox_60
   widget=BenefitName
   properties {
    justification=Center
    text=Text Block
    wrapTextAt=700
    font {
     bForceMonospaced=false
     bMaterialIsStencil=false
     fontMaterial=null
     fontObject=&NotoSans
     letterSpacing=0
     monospacedWidth=1
     size=24
     skewAmount=0
     typefaceFontName=NotoSans-Regular
     outlineSettings {
      bApplyOutlineToDropShadows=false
      bMiteredCorners=false
      bSeparateFillAlpha=false
      outlineMaterial=null
      outlineSize=0
      outlineColor=<rgba 0 0 0 1>
     }
    }
    shadowOffset=<0 0>
   }
   slot {class=VerticalBoxSlot, properties {horizontalAlignment=HAlign_Center}}
  }
  BenefitType {
   widgetClassPath=&CommonTextBlock
   bIsVariable=true
   parent=VerticalBox_60
   widget=BenefitType
   properties {
    text=Text Block
    colorAndOpacity {colorUseRule=UseColor_Specified, specifiedColor=<rgba 1 1 1 0.75>}
    font {
     bForceMonospaced=false
     bMaterialIsStencil=false
     fontMaterial=null
     fontObject=&NotoSans
     letterSpacing=0
     monospacedWidth=1
     size=24
     skewAmount=0
     typefaceFontName=NotoSans-Regular
     outlineSettings {
      bApplyOutlineToDropShadows=false
      bMiteredCorners=false
      bSeparateFillAlpha=false
      outlineMaterial=null
      outlineSize=0
      outlineColor=<rgba 0 0 0 1>
     }
    }
    shadowOffset=<0 0>
   }
   slot {class=VerticalBoxSlot, properties {horizontalAlignment=HAlign_Center, verticalAlignment=VAlign_Bottom, padding=<ltrb 0 0 0 20>}}
  }
  BenefitDescription {
   widgetClassPath=&CommonRichTextBlock
   bIsVariable=true
   parent=VerticalBox_60
   widget=BenefitDescription
   properties {
    bOverrideDefaultStyle=true
    justification=Center
    lineHeightPercentage=1.25
    text=Access this thing while subscribed to Crew
    wrapTextAt=700
    decoratorClasses [/Game/Athena/UI/Frontend/CosmeticGating/Data/RichTextDecorator_CosmeticGate.RichTextDecorator_CosmeticGate_C]
    defaultTextStyleOverride {
     overflowPolicy=Clip
     transformPolicy=None
     colorAndOpacity {colorUseRule=UseColor_Specified, specifiedColor=<rgba 1 1 1 1>}
     font {
      bForceMonospaced=false
      bMaterialIsStencil=false
      fontMaterial=null
      fontObject=&HeadingNow
      letterSpacing=10
      monospacedWidth=1
      size=18
      skewAmount=0
      typefaceFontName=74 Regular
      outlineSettings {
       bApplyOutlineToDropShadows=false
       bMiteredCorners=false
       bSeparateFillAlpha=false
       outlineMaterial=null
       outlineSize=0
       outlineColor=<rgba 0 0 0 1>
      }
     }
     highlightColor {colorUseRule=UseColor_Specified, specifiedColor=<rgba 0 0 0 1>}
     highlightShape {
      bIsDynamicallyLoaded=false
      drawAs=Image
      imageType=NoImage
      mirroring=NoMirror
      resourceName=None
      resourceObject=null
      tiling=NoTile
      imageSize=<32 32>
      margin=<ltrb 0 0 0 0>
      outlineSettings {
       bUseBrushTransparency=false
       roundingType=HalfHeightRadius
       width=0
       color {colorUseRule=UseColor_Specified, specifiedColor=<rgba 0 0 0 0>}
       cornerRadii=<0 0 0 0>
      }
      tintColor {colorUseRule=UseColor_Specified, specifiedColor=<rgba 1 1 1 1>}
      uVRegion {bIsValid=false, max=<0 0>, min=<0 0>}
     }
     selectedBackgroundColor {colorUseRule=UseColor_Foreground, specifiedColor=<rgba 1 0 1 1>}
     shadowColorAndOpacity=<rgba 0 0 0 1>
     shadowOffset=<0 0>
     strikeBrush {
      bIsDynamicallyLoaded=false
      drawAs=Image
      imageType=NoImage
      mirroring=NoMirror
      resourceName=None
      resourceObject=null
      tiling=NoTile
      imageSize=<32 32>
      margin=<ltrb 0 0 0 0>
      outlineSettings {
       bUseBrushTransparency=false
       roundingType=HalfHeightRadius
       width=0
       color {colorUseRule=UseColor_Specified, specifiedColor=<rgba 0 0 0 0>}
       cornerRadii=<0 0 0 0>
      }
      tintColor {colorUseRule=UseColor_Specified, specifiedColor=<rgba 1 1 1 1>}
      uVRegion {bIsValid=false, max=<0 0>, min=<0 0>}
     }
     underlineBrush {
      bIsDynamicallyLoaded=false
      drawAs=Image
      imageType=NoImage
      mirroring=NoMirror
      resourceName=None
      resourceObject=null
      tiling=NoTile
      imageSize=<32 32>
      margin=<ltrb 0 0 0 0>
      outlineSettings {
       bUseBrushTransparency=false
       roundingType=HalfHeightRadius
       width=0
       color {colorUseRule=UseColor_Specified, specifiedColor=<rgba 0 0 0 0>}
       cornerRadii=<0 0 0 0>
      }
      tintColor {colorUseRule=UseColor_Specified, specifiedColor=<rgba 1 1 1 1>}
      uVRegion {bIsValid=false, max=<0 0>, min=<0 0>}
     }
    }
   }
   slot {class=VerticalBoxSlot, properties {horizontalAlignment=HAlign_Center, verticalAlignment=VAlign_Bottom, padding=<ltrb 0 0 0 40>}}
  }
  SubTag {
   widgetClassPath=/FortUILibrary/UIKit/Tag/WBP_UIKit_Tag
   bIsVariable=true
   parent=VerticalBox_60
   widget=SubTag
   properties {labelText="", tagText=SUBSCRIBED, tagType=Primary Tag, colorAndOpacity=<rgba 0.0395 1 0.2307 1>}
   slot {class=VerticalBoxSlot, properties {horizontalAlignment=HAlign_Center, verticalAlignment=VAlign_Bottom, padding=<ltrb 0 0 0 40>}}
  }
  Spacer_52 {
   widgetClassPath=&Spacer
   parent=VerticalBox_0
   widget=Spacer_52
   properties {visibility=Collapsed}
   slot {class=VerticalBoxSlot, properties {size {sizeRule=Fill, value=1}}}
  }
  ServicesDownError {
   widgetClassPath=/FortUILibrary/UIKit/InfoBox/WBP_UIKit_InfoBox
   bIsVariable=true
   parent=VerticalBox_0
   widget=ServicesDownError
   properties {previewText="Services down, please try again later.", showIcon=false, showStatusIndicator=true, use Preview Text=true, visibility=Collapsed}
   slot {class=VerticalBoxSlot}
  }
  ButtonSwitcher {
   widgetClassPath=&CommonVisibilitySwitcher
   bIsVariable=true
   parent=VerticalBox_0
   widget=ButtonSwitcher
   slot {class=VerticalBoxSlot}
  }
  PurchaseAvailabilitySwitcher {
   widgetClassPath=&CommonVisibilitySwitcher
   bIsVariable=true
   parent=ButtonSwitcher
   widget=PurchaseAvailabilitySwitcher
   slot {class=CommonVisibilitySwitcherSlot, properties {verticalAlignment=VAlign_Top}}
  }
  SubButton {
   widgetClassPath=/FortUILibrary/UIKit/Buttons/WBP_UIKit_Button_Loud
   bIsVariable=true
   parent=PurchaseAvailabilitySwitcher
   widget=SubButton
   properties {
    bRequiresHold=true
    buttonSize=Large
    holdData=/FortUILibrary/Input/PurchaseInteraction_HoldData.PurchaseInteraction_HoldData_C
    minHeight=56
    minWidth=450
   }
   slot {class=CommonVisibilitySwitcherSlot, properties {horizontalAlignment=HAlign_Center}}
  }
  Button_OpenParentalControls {
   widgetClassPath=&WBP_UIKit_Button_Loud
   bIsVariable=true
   parent=PurchaseAvailabilitySwitcher
   widget=Button_OpenParentalControls
   properties {bHideInputAction=true, text=Open Parental Controls}
   slot {class=CommonVisibilitySwitcherSlot, properties {horizontalAlignment=HAlign_Center}}
  }
  VerticalBox_59 {widgetClassPath=&VerticalBox, parent=ButtonSwitcher, widget=VerticalBox_59, slot {class=CommonVisibilitySwitcherSlot}}
  WrapBox_1 {
   widgetClassPath=&WrapBox
   parent=VerticalBox_59
   widget=WrapBox_1
   properties {bExplicitWrapSize=true, horizontalAlignment=HAlign_Center, wrapSize=700, innerSlotPadding=<0 16>}
   slot {class=VerticalBoxSlot}
  }
  ResolveIssue_Button {
   widgetClassPath=&WBP_UIKit_Button_Loud
   bIsVariable=true
   parent=WrapBox_1
   widget=ResolveIssue_Button
   properties {bHideInputAction=true, minWidth=304, text=Resolve Issue}
   slot {class=WrapBoxSlot, properties {horizontalAlignment=HAlign_Center, padding=<ltrb 0 0 8 0>}}
  }
  CancelButton {
   widgetClassPath=&WBP_UIKit_Button_Regular
   bIsVariable=true
   parent=WrapBox_1
   widget=CancelButton
   properties {bHideInputAction=true, minWidth=375, text=Cancellation Info}
   slot {class=WrapBoxSlot, properties {horizontalAlignment=HAlign_Center}}
  }
  PriceNextBill {
   widgetClassPath=&CommonTextBlock
   bIsVariable=true
   parent=VerticalBox_59
   widget=PriceNextBill
   properties {
    text=You will be charged on your next billing cycle.
    wrapTextAt=700
    colorAndOpacity {colorUseRule=UseColor_Specified, specifiedColor=<rgba 0.5776 0.5776 0.5776 1>}
    font {
     bForceMonospaced=false
     bMaterialIsStencil=false
     fontMaterial=null
     fontObject=&NotoSans
     letterSpacing=0
     monospacedWidth=1
     size=24
     skewAmount=0
     typefaceFontName=NotoSans-Regular
     outlineSettings {
      bApplyOutlineToDropShadows=false
      bMiteredCorners=false
      bSeparateFillAlpha=false
      outlineMaterial=null
      outlineSize=0
      outlineColor=<rgba 0 0 0 1>
     }
    }
    shadowOffset=<0 0>
   }
   slot {class=VerticalBoxSlot, properties {horizontalAlignment=HAlign_Center, padding=<ltrb 0 16 0 0>}}
  }
  WBP_UIKit_InfoBox {
   widgetClassPath=&WBP_UIKit_InfoBox
   bIsVariable=true
   parent=ButtonSwitcher
   widget=WBP_UIKit_InfoBox
   properties {
    previewText=Fortnite Crew is not available. For more information please visit fn.gg/FortniteCrew
    showIcon=false
    showStatusIndicator=true
    textJustification=Center
    use Preview Text=true
   }
   slot {class=CommonVisibilitySwitcherSlot}
  }
 ]
}
