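# Fortnite Crew benefit tile, a UIKit holdable button with CMS image, title, highlight, status indicator, optional header and section divider.
# Hover and unhover set or clear HoveredBenefit on the selection view model.
# HandleSelectionVMChanged compares that benefit to its own, so UpdateVisual transitions the outline and plays or reverses the material-overlay animation, skipping unchanged focus unless forced.
# Helper functions set three background material colors, header text and divider visibility.
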WidgetBlueprint /Crew/UI/CrewBenefitTile
animations [Hover {
  binding_count=1
  length_seconds=0.1
  object=&self:Hover
  sequence {bindings [CMSImage_Background {
     id=#1
     possessed_class=EpicCMSImage
     tracks [{
       type=MovieSceneWidgetMaterialTrack
       sections [{
         type=MovieSceneParameterSection
         end_frame=2
         start_frame=0
         channels [Hover Overlay {type=MovieSceneScriptingFloatChannel, keys [[0,0],[2,1]]}]
        }]
      }]
    }]}
 }]
graphs [
 EventGraph {nodes [
   Event On Hovered()
   > Set HoveredBenefit!(
    execute
    HoveredBenefit:Crew Benefit VM Object Reference=Get CrewBenefitVM() -> (CrewBenefitVM:Crew Benefit VM Object Reference)
    self:Crew Selection VM Object Reference=Get CrewSelectionVM() -> (CrewSelectionVM:Crew Selection VM Object Reference)
   ) -> (then, Output_Get:Crew Benefit VM Object Reference)
   Event On Unhovered()
   > Set HoveredBenefit(self=Get CrewSelectionVM())
  ]}
 HandleSelectionVMChanged {nodes [
   n1 = HandleSelectionVMChanged!() -> (then, HoveredBenefitVM:Crew Benefit VM Object Reference) {is=title K2Node_FunctionEntry}
   > n2 = Set bShouldBeFocused(bShouldBeFocused:bool="Equal (Object)"(A=@n1, B=Get CrewBenefitVM())) -> (Output_Get:bool)
   > UpdateVisual!(execute, self, bShouldFocus:bool=@n2, bShouldForceUpdate:bool=false) -> (then)
   > Return Node()
  ]}
 SetDefaultVisual {nodes [
   SetDefaultVisual() {is=title K2Node_FunctionEntry}
   > StopAllAnimations(self:User Widget Object Reference=Get Block_Outline() -> (Block_Outline:WBP UIKit Block Outline Object Reference))
   n1 = Get Block_Outline()
   n2 = Get Anim_Hover!(self:WBP UIKit Block Outline Object Reference=@n1) -> (Anim_Hover:Widget Animation Object Reference)
   > PlayAnimation(self=@n1, InAnimation=@n2, StartAtTime=GetEndTime(self=@n2))
   n3 = Get Hover() -> (Hover:Widget Animation Object Reference)
   > PlayAnimation(InAnimation=@n3, StartAtTime=GetEndTime(self=@n3), PlayMode=Reverse)
   n4 = Get CrewSelectionVM()
   > UpdateVisual(
    bShouldFocus="Equal (Object)"(
     A=Get CrewBenefitVM()
     B=Get HoveredBenefit!(self:Crew Selection VM Object Reference=@n4) -> (HoveredBenefit:Crew Benefit VM Object Reference)
    )
    bShouldForceUpdate=true
   )
  ]}
 SetHeader {nodes [
   n1 = SetHeader!() -> (then, HeaderText:Text) {is=title K2Node_FunctionEntry}
   n2 = Get Header() -> (Header:Common Text Object Reference)
   > Set Text(Text=@n1, self=@n2)
   > SetVisibility(self=@n2, InVisibility=SelfHitTestInvisible)
  ]}
 SetMaterialColor1 {nodes [
   n1 = SetMaterialColor1!() -> (then, Color:Color Structure) {is=title K2Node_FunctionEntry}
   > n2 = GetDynamicMaterial(self=Get CMSImage_Background() -> (CMSImage_Background:Epic CMSImage Object Reference))
   > SetVectorParameterValue(
    self=@n2
    ParameterName=BackgroundColor1
    Value="To LinearColor (Color)"!(InColor:Color Structure=@n1) -> (ReturnValue:Linear Color Structure)
   )
  ]}
 SetMaterialColor2 {nodes [
   n1 = SetMaterialColor2!() -> (then, Color:Color Structure) {is=title K2Node_FunctionEntry}
   > n2 = GetDynamicMaterial(self=Get CMSImage_Background())
   > SetVectorParameterValue(self=@n2, ParameterName=BackgroundColor2, Value="To LinearColor (Color)"(InColor=@n1))
  ]}
 SetMaterialColor3 {nodes [
   n1 = SetMaterialColor3!() -> (then, Color:Color Structure) {is=title K2Node_FunctionEntry}
   > n2 = GetDynamicMaterial(self=Get CMSImage_Background())
   > SetVectorParameterValue(self=@n2, ParameterName=OverlayGradientColor, Value="To LinearColor (Color)"(InColor=@n1))
  ]}
 ShowSectionDivider {nodes [
   n1 = ShowSectionDivider!() -> (then, Show:bool=~) {is=title K2Node_FunctionEntry}
   > SetVisibility(
    self=Get NewSectionSpacer() -> (NewSectionSpacer:Horizontal Box Object Reference)
    InVisibility=Select(Option 0=Collapsed, Option 1=SelfHitTestInvisible, Index=@n1)
   )
  ]}
 UpdateVisual {nodes [
   n1 = UpdateVisual!() -> (then, bShouldFocus:bool=~, bShouldForceUpdate:bool=~) {is=title K2Node_FunctionEntry}
   n2 = Get bShouldFocus() -> (bShouldFocus:bool)
   n3 = Get bIsCurrentlyFocused() -> (bIsCurrentlyFocused:bool)
   > Branch(Condition=OR Boolean(A=@n1.bShouldForceUpdate, B=XOR Boolean!(A:bool=@n2, B:bool=@n3) -> (ReturnValue:bool)))
   > Set bIsCurrentlyFocused(bIsCurrentlyFocused:bool=Get bShouldFocus()) -> (Output_Get:bool)
   > RequestTransition!(
    execute
    self:UIKit Block Object Reference=Get Block_Outline()
    State:EUIKitBlockVisualState Enum=Select(Option 0:EUIKitBlockVisualState Enum=Unfocused, Option 1:EUIKitBlockVisualState Enum=Focused, Index=Get bShouldFocus()) -> (ReturnValue:EUIKitBlockVisualState Enum)
   ) -> (then)
   > PlayAnimation(
    InAnimation=Get Hover()
    PlayMode=Select(Option 0:EUMGSequencePlayMode Enum=Reverse, Option 1:EUMGSequencePlayMode Enum, Index=Get bShouldFocus()) -> (ReturnValue:EUMGSequencePlayMode Enum)
   )
  ]}
]
variables [
 A {type=string, default=""}
 bIsCurrentlyFocused {type=boolean}
 Ends {type=string, default=Ends}
 Format {type=string, default=Ends}
]
widgets {
 info {inheritedWidgetCount=1, parentClass=/Script/FortniteUI.UIKitHoldableModularButton, rootWidgetClass=&Overlay, widgetCount=16}
 widgets [
  Overlay_0 {widgetClassPath=&Overlay, widget=Overlay_0}
  SizeBox {
   widgetClassPath=&SizeBox
   bIsVariable=true
   parent=Overlay_0
   widget=SizeBox
   properties {bOverride_HeightOverride=true, bOverride_WidthOverride=true, heightOverride=160, widthOverride=260}
   slot {class=OverlaySlot}
  }
  Root {
   widgetClassPath=&GridPanel
   bIsVariable=true
   parent=SizeBox
   widget=Root
   properties {columnFill [1], rowFill [1]}
   slot {class=SizeBoxSlot}
  }
  CMSImage_Background {
   widgetClassPath=/Script/EpicCMSUIFramework.EpicCMSImage
   bIsVariable=true
   parent=Root
   widget=CMSImage_Background
   properties {
    materialTextureParamName=CMSTexture
    visibility=SelfHitTestInvisible
    brush {
     bIsDynamicallyLoaded=false
     drawAs=Image
     imageType=FullColor
     mirroring=NoMirror
     resourceName=None
     resourceObject=/Crew/UI/Materials/CrewBenefits/MI_UI_CrewBenefitsTile
     tiling=NoTile
     imageSize=<32 32>
     margin=<ltrb 0 0 0 0>
     outlineSettings {
      bUseBrushTransparency=false
      roundingType=HalfHeightRadius
      width=0
      color {colorUseRule=UseColor_Specified, specifiedColor=<rgba 0 0 0 0>}
      cornerRadii=<0 0 0 0>
     }
     tintColor {colorUseRule=UseColor_Specified, specifiedColor=<rgba 1 1 1 1>}
     uVRegion {bIsValid=false, max=<0 0>, min=<0 0>}
    }
    loadingFailFallback {
     bIsDynamicallyLoaded=false
     drawAs=Image
     imageType=FullColor
     mirroring=NoMirror
     resourceName=None
     resourceObject=/FortUILibrary/Textures/Icons/T_UI_Icon_UIKit_ImageError
     tiling=NoTile
     imageSize=<64 64>
     margin=<ltrb 0 0 0 0>
     outlineSettings {
      bUseBrushTransparency=false
      roundingType=HalfHeightRadius
      width=0
      color {colorUseRule=UseColor_Specified, specifiedColor=<rgba 0 0 0 0>}
      cornerRadii=<0 0 0 0>
     }
     tintColor {colorUseRule=UseColor_Specified, specifiedColor=<rgba 1 1 1 1>}
     uVRegion {bIsValid=false, max=<0 0>, min=<0 0>}
    }
   }
   slot {class=GridSlot}
  }
  Block_Outline {
   widgetClassPath=/FortUILibrary/UIKit/Buttons/Blocks/WBP_UIKit_Block_Outline
   bIsVariable=true
   parent=Root
   widget=Block_Outline
   properties {bottomLeftCornerRadius=30, bottomRightCornerRadius=30, radiusInPixels=true, topLeftCornerRadius=30, topRightCornerRadius=30}
   slot {class=GridSlot, properties {layer=20}}
  }
  Header {
   widgetClassPath=&CommonTextBlock
   bIsVariable=true
   parent=Root
   widget=Header
   properties {
    text=Recurring Benefits
    visibility=Collapsed
    colorAndOpacity {colorUseRule=UseColor_Specified, specifiedColor=<rgba 1 1 1 0.75>}
    font {
     bForceMonospaced=false
     bMaterialIsStencil=false
     fontMaterial=null
     fontObject=&NotoSans
     letterSpacing=0
     monospacedWidth=1
     size=24
     skewAmount=0
     typefaceFontName=NotoSans-Regular
     outlineSettings {
      bApplyOutlineToDropShadows=false
      bMiteredCorners=false
      bSeparateFillAlpha=false
      outlineMaterial=null
      outlineSize=0
      outlineColor=<rgba 0 0 0 1>
     }
    }
    renderTransform {angle=0, scale=<1 1>, shear=<0 0>, translation=<0 -45>}
    shadowOffset=<0 0>
   }
   slot {class=GridSlot, properties {verticalAlignment=VAlign_Top}}
  }
  Title {
   widgetClassPath=&CommonTextBlock
   parent=Root
   widget=Title
   properties {
    justification=Center
    scrollStyle=&TextScrollStyle-Base_C
    text=Music Pass
    font {
     bForceMonospaced=false
     bMaterialIsStencil=false
     fontMaterial=null
     fontObject=&NotoSans
     letterSpacing=0
     monospacedWidth=1
     size=24
     skewAmount=0
     typefaceFontName=NotoSans-Regular
     outlineSettings {
      bApplyOutlineToDropShadows=false
      bMiteredCorners=false
      bSeparateFillAlpha=false
      outlineMaterial=null
      outlineSize=0
      outlineColor=<rgba 0 0 0 1>
     }
    }
    shadowOffset=<0 0>
   }
   slot {class=GridSlot, properties {layer=3, verticalAlignment=VAlign_Bottom, padding=<ltrb 16 0 16 12>}}
  }
  HorizontalBox_1 {
   widgetClassPath=&HorizontalBox
   parent=Root
   widget=HorizontalBox_1
   slot {class=GridSlot, properties {layer=4, verticalAlignment=VAlign_Top, padding=<ltrb 8 8 8 8>}}
  }
  FillSpaceOverlay {
   widgetClassPath=&Overlay
   parent=HorizontalBox_1
   widget=FillSpaceOverlay
   slot {class=HorizontalBoxSlot, properties {size {sizeRule=Fill, value=1}}}
  }
  WBP_UIKit_Highlight_Base {
   widgetClassPath=/FortUILibrary/Systems/Highlight/WBP_UIKit_Highlight_Base
   bIsVariable=true
   parent=FillSpaceOverlay
   widget=WBP_UIKit_Highlight_Base
   properties {highlightSize=Compact, highlightText=Ends Soon, highlightType=Utility}
   slot {class=OverlaySlot}
  }
  WBP_UIKit_StatusIndicator {
   widgetClassPath=/FortUILibrary/UIKit/StatusIndicator/WBP_UIKit_StatusIndicator
   bIsVariable=true
   parent=HorizontalBox_1
   widget=WBP_UIKit_StatusIndicator
   properties {icon=/FortUILibrary/Textures/MaterialInstances/MI_UIKit_Icon_Check}
   slot {class=HorizontalBoxSlot, properties {padding=<ltrb 4 0 0 0>}}
  }
  NewSectionSpacer {
   widgetClassPath=&HorizontalBox
   bIsVariable=true
   parent=Overlay_0
   widget=NewSectionSpacer
   properties {visibility=Collapsed}
   slot {class=OverlaySlot, properties {verticalAlignment=VAlign_Fill, padding=<ltrb 260 0 0 0>}}
  }
  Spacer_67 {widgetClassPath=&Spacer, parent=NewSectionSpacer, widget=Spacer_67, properties {size=<39 1>}, slot {class=HorizontalBoxSlot}}
  WBP_UIKit_Divider_Base {
   widgetClassPath=/FortUILibrary/Systems/Tabs/WBP_UIKit_Divider_Base
   bIsVariable=true
   parent=NewSectionSpacer
   widget=WBP_UIKit_Divider_Base
   properties {div Width=2}
   slot {class=HorizontalBoxSlot}
  }
  Spacer {widgetClassPath=&Spacer, parent=NewSectionSpacer, widget=Spacer, properties {size=<15 1>}, slot {class=HorizontalBoxSlot}}
  InputActionWidget {widgetClassPath=&CommonActionWidget, bInherited=true, widget=null}
 ]
}
