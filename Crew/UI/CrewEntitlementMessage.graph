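# Crew entitlement message widget that wraps a UIKit InfoBox to show Fortnite Crew subscription notices.
# HandleSeasonPassVM stores the Season Pass view model and passes the Crew view model's subscription state to SetState.
# SetState switches on that state, setting text and icon for cancelled, non-recurring, payment, platform and reward-delay cases.
# Inactive shows a V-Buck grant notice only for pass owners; Active, Unavailable and Unknown states collapse the widget.
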WidgetBlueprint /Crew/UI/CrewEntitlementMessage
graphs [
 EventGraph {nodes [
   Event PreConstruct()
   Event Construct()
   Event Tick()
  ]}
 HandleSeasonPassVM {nodes [
   n1 = HandleSeasonPassVM!() -> (then, PassVM:Season Pass VM Object Reference) {is=title K2Node_FunctionEntry}
   > Set SeasonPassVM(SeasonPassVM:Season Pass VM Object Reference=@n1) -> (Output_Get:Season Pass VM Object Reference)
   n2 = Get CrewVM() -> (CrewVM:Crew VM Object Reference)
   > Branch(Condition=IsValid(Object=@n2))
   > SetState!(
    execute
    self
    SubState:EMcpSubscriptionState Enum=Get SubscriptionState!(self:Crew VM Object Reference=@n2) -> (SubscriptionState:EMcpSubscriptionState Enum=Inactive)
   ) -> (then)
  ]}
 SetState {nodes [
   n1 = SetState!() -> (then, SubState:EMcpSubscriptionState Enum=Inactive) {is=title K2Node_FunctionEntry}
   > n2 = Switch on EMcpSubscriptionState!(execute, Selection:EMcpSubscriptionState Enum=@n1) -> (
    Inactive:Exec
    Active:Exec
    ActiveWithExternalBilling:Exec
    NotAutoRenewable:Exec
    Canceled:Exec
    PaymentProcessError:Exec
    PlatformAuthError:Exec
    RewardsDelayed:Exec
    Unavailable:Exec
    Unknown:Exec
   ) {is=title K2Node_SwitchEnum}
   n3 = GetSubscriptionEndDate!(self:Crew VM Object Reference=Get CrewVM()) -> (ReturnValue:Text)
   n4 = Get InfoBox() -> (InfoBox:WBP UIKit Info Box Object Reference)
   n5 = SetText(
    execute=@n2.Canceled
    self:WBP UIKit Info Box Object Reference=@n4
    InText:Text=Format Text!(Format="LOCTEXT(\"Your membership will become inactive on {date}.\")", date:Text=@n3) -> (Result)
   )
   n6 = SetStatusIndicator!(
    execute=[@n5.then, @n13.then]
    self:WBP UIKit Info Box Object Reference=Get InfoBox()
    Icon:Object Reference=/FortUILibrary/Textures/MaterialInstances/MI_UIKit_Icon_ExclamationMark
    StatusType:E_UI_StatusIndicator_Type Enum=NewEnumerator2
    Config:E_UI_StatusIndicator_Configuration Enum=NewEnumerator0
   ) -> (then)
   SetVisibility(execute=[@n12.then, @n6.then, @n16.then, @n14.then], InVisibility=SelfHitTestInvisible)
   n7 = Branch(execute=@n2.Inactive, Condition=IsValid(Object=Get SeasonPassVM() -> (SeasonPassVM:Season Pass VM Object Reference)))
   > n8 = Branch(Condition=Get bDoesOwnsPass!(self:Season Pass VM Object Reference=Get SeasonPassVM()) -> (bDoesOwnsPass:bool))
   n9 = Get PassOffer!(self:Season Pass VM Object Reference=Get SeasonPassVM()) -> (PassOffer:Fort Simple Offer VM Object Reference)
   n10 = Get SeasonPassVM()
   n11 = Break Season Pass CMSData!(SeasonPassCMSData:"Season Pass CMSData Structure (by ref)"=Get SeasonPassCMSData!(self:Season Pass VM Object Reference=Get SeasonPassVM()) -> (SeasonPassCMSData:Season Pass CMSData Structure)) -> (
    PurchaseBackgroundURL:String
    PurchaseConfirmBackgroundURL:String
    PurchaseDisclaimer:Text
    PurchaseDescription:Text
    PurchaseConfirmDescription:Text
    LandingPageDisclaimer:Text
    LevelPurchaseDisclaimer:Text
    SeasonEndDateText:Text
    SeasonEndDateTime:Date Time Structure
   ) {is=title K2Node_BreakStruct}
   > SetText(
    self=Get InfoBox()
    InText=Format Text!(
     Format="LOCTEXT(\"Subscribing to Fortnite Crew before {MONTH DAY} may trigger a one-time {refundAmount} V-Buck grant and a change to your ability to claim premium rewards from the {Pass NAME} if you cancel your subscription. Learn more fn.gg/CrewVBucks.\")"
     MONTH DAY:Text="As Date (from UTC)"!(InDateTime:"Date Time Structure (by ref)"=@n11.SeasonEndDateTime, InTimeZone:String=ET, InDateStyle:EDateTimeStyle Enum=Default) -> (ReturnValue:Text)
     refundAmount:Text=GetPriceText!(self:Fort Base Offer VM Object Reference=@n9) -> (ReturnValue:Text)
     Pass NAME:Text=GetDisplayName!(self:Season Pass VM Object Reference=@n10) -> (ReturnValue:Text)
    ) -> (Result)
   )
   > n12 = SetStatusIndicator(self=Get InfoBox(), Icon=/FortUILibrary/Textures/MaterialInstances/MI_UIKit_Icon_VBuck, StatusType=NewEnumerator3, Config=NewEnumerator0)
   SetVisibility(execute=[@n7.else, @n8.else, @n2.Active, @n2.ActiveWithExternalBilling, @n2.Unavailable, @n2.Unknown], InVisibility=Collapsed)
   n13 = SetText(
    execute=@n2.NotAutoRenewable
    self=@n4
    InText=Format Text!(Format="LOCTEXT(\"Your Fortnite Crew access is non-recurring and ends on {date}.\")", date:Text=@n3) -> (Result)
   )
   SetText(
    execute=@n2.PaymentProcessError
    self=@n4
    InText="LOCTEXT(\"Your rewards are temporality on hold as there was an issue processing your payment.\")"
   )
   > n14 = SetStatusIndicator(self=Get InfoBox(), Icon=/Crew/UI/Materials/MI_Icon_Warning_Red, StatusType=NewEnumerator3, Config=NewEnumerator0)
   n15 = SetText(
    execute=@n2.PlatformAuthError
    self=@n4
    InText="LOCTEXT(\"Please, log in to the platform you subscribed to the Fortnite Crew on to receive your subscription benefits.\")"
   )
   n16 = SetStatusIndicator(
    execute=[@n15.then, @n17.then]
    self=Get InfoBox()
    Icon=&MI_UIKit_Icon_ExclamationMark
    StatusType=NewEnumerator4
    Config=NewEnumerator0
   )
   n17 = SetText(
    execute=@n2.RewardsDelayed
    self=@n4
    InText="LOCTEXT(\"<text color=\\\"##FF6B66\\\" fontface=\\\"75 Medium\\\">There appears to be an error processing your rewards</> - please try again later, or contact Player Support if the problem persists.\")"
   )
  ]}
]
variables []
widgets {
 info {
  inheritedWidgetCount=0
  parentClass=&UserWidget
  rootWidgetClass=/FortUILibrary/UIKit/InfoBox/WBP_UIKit_InfoBox.WBP_UIKit_InfoBox_C
  widgetCount=1
 }
 widgets [InfoBox {
   widgetClassPath=/FortUILibrary/UIKit/InfoBox/WBP_UIKit_InfoBox
   bIsVariable=true
   widget=InfoBox
   properties {
    iconMaterial=&MI_UIKit_Icon_VBuck
    previewText=""
    showIcon=false
    showStatusIndicator=true
    use Preview Text=true
   }
  }]
}
