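# Fortnite Crew modal prompting members to subscribe with a payment method to continue, built on SubscriptionModalBase.
# PreConstruct feeds a DialogVM_Continue into the UIKit dialog, which hosts scrollable terms, wrong-platform text and a SubscriberSince panel.
# Dialog clicks, bound while activated, either broadcast SubscribeStarted and deactivate, or back out like the touch-only close button.
# Construct collapses the second dialog button on mobile and registers the scroll box.
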WidgetBlueprint /Crew/UI/Subscription/SubscriptionContinueModal
graphs [
 EventGraph {nodes [
   n1 = "On Clicked (Button_TouchBack)"!() -> (OutputDelegate, then, Button:Common Button Base Object Reference) {is=title K2Node_ComponentBoundEvent}
   TriggerModalBackout(execute=[@n1.then, @n3.1], self:Subscription Modal Base Object Reference)
   n2 = DialogButtonClicked!() -> (OutputDelegate, then, ButtonIndex:int) {is=title K2Node_CustomEvent}
   > n3 = Switch on Int!(execute:Exec, Selection:int=@n2.ButtonIndex) -> (Default:Exec, 0:Exec, 1:Exec)
   Call Subscribe Started!(execute:Exec=@n3.0, self:Subscription Continue Modal Object Reference) -> (then:Exec) {is=title K2Node_CallDelegate}
   > DeactivateWidget()
   Event PreConstruct()
   > n4 = Construct Object from Class(Class=/Crew/UI/Subscription/Data/DialogVM_Continue.DialogVM_Continue_C) -> (ReturnValue:Dialog VM Continue Object Reference)
   > SetDialogData!(
    execute
    self:WBP UIKit Dialog Internal Object Reference=Get Dialog_Internal() -> (Dialog_Internal:WBP UIKit Dialog Internal Object Reference)
    Data:UI Kit Dialog Viewmodel Object Reference=@n4
   ) -> (then)
   Event On Activated()
   > OnActivatedDefault(self:Subscription Modal Base Object Reference)
   n5 = Create Event()
   n6 = Get Dialog_Internal()
   > Bind Event to Dialog Button Clicked Dispatcher(self:WBP UIKit Dialog Internal Object Reference=@n6, Delegate=@n5) {is=title K2Node_AddDelegate}
   Event Construct()
   n7 = Get Dialog_Internal()
   > SetVisibility(
    self="Get (a copy)"(
     Array:Array of WBP UIKit Button CTA Base Object References=Get ButtonWidgets!(self:WBP UIKit Dialog Internal Object Reference=@n7) -> (ButtonWidgets:Array of WBP UIKit Button CTA Base Object References)
     Dimension 1=1
    ) -> (Output:WBP UIKit Button CTA Base Object Reference)
    InVisibility=Select(Option 0=SelfHitTestInvisible, Option 1=Collapsed, Index=IsMobileGame())
   )
   > SetScrollWidget!(execute, ToScroll:Widget Object Reference=Get PrimaryScrollBox() -> (PrimaryScrollBox:Scroll Box Object Reference)) -> (then)
   Event On Deactivated()
   > Unbind Event from Dialog Button Clicked Dispatcher(self:WBP UIKit Dialog Internal Object Reference=@n6, Delegate=@n5) {is=title K2Node_RemoveDelegate}
  ]}
 SubscribeStarted {nodes [SubscribeStarted() {is=title K2Node_FunctionEntry}]}
]
variables [SubscribeStarted {}]
widgets {
 info {
  inheritedWidgetCount=0
  namedSlotCount=1
  parentClass=&/SubscriptionModalBase.SubscriptionModalBase_C
  rootWidgetClass=&Overlay
  widgetCount=21
 }
 widgets [
  Overlay_31 {widgetClassPath=&Overlay, widget=Overlay_31}
  Dialog_Internal {
   widgetClassPath=/FortUILibrary/Systems/Dialog/WBP_UIKit_Dialog_Internal
   bIsVariable=true
   parent=Overlay_31
   widget=Dialog_Internal
   properties {
    useText=None
    buttons {
     each Button Horizontal Fill=false
     each Button Min Width=157
     each Button Size=Medium
     focuse Button=0
     space Between Buttons=24
     cTA Buttons Types [Loud,Regular]
     group Placement {
      horizontal Alignment=HAlign_Fill
      vertical Alignment=VAlign_Bottom
      grid {column=1, column Span=0, layer=0, row=2, row Span=0, nudge=<0 0>}
      padding=<ltrb 80 64 80 48>
     }
    }
    dialogSIze {
     height Override {enable=true, value=750}
     max Aspect Ratio {enable=false, value=0}
     max Desired Height {enable=true, value=800}
     max Desired Width {enable=false, value=0}
     min Aspect Ratio {enable=false, value=0}
     min Desired Height {enable=false, value=0}
     min Desired Width {enable=false, value=0}
     width Override {enable=true, value=1490}
    }
    titleProperties {
     auto Wrap Text=true
     justification=Left
     line Height Percentage=1
     mobile Font Size Multiplier=1.3
     transform Policy=ToUpper
     wrap Text At=808
     color {colorUseRule=UseColor_Specified, specifiedColor=<rgba 1 1 1 1>}
     font {
      bForceMonospaced=false
      bMaterialIsStencil=false
      fontMaterial=null
      fontObject=&HeadingNow
      letterSpacing=0
      monospacedWidth=1
      size=25.5
      skewAmount=0
      typefaceFontName=86 Bold Italic
      outlineSettings {
       bApplyOutlineToDropShadows=false
       bMiteredCorners=false
       bSeparateFillAlpha=false
       outlineMaterial=null
       outlineSize=0
       outlineColor=<rgba 0 0 0 1>
      }
     }
     placement {
      horizontal Alignment=HAlign_Fill
      vertical Alignment=VAlign_Top
      grid {column=1, column Span=0, layer=0, row=0, row Span=0, nudge=<0 0>}
      padding=<ltrb 80 80 80 32>
     }
    }
   }
   slot {class=OverlaySlot, properties {horizontalAlignment=HAlign_Fill, verticalAlignment=VAlign_Fill}}
  }
  Overlay_0 {widgetClassPath=&Overlay, namedSlotHost=Dialog_Internal, widget=Overlay_0}
  Content {
   widgetClassPath=&HorizontalBox
   parent=Overlay_0
   widget=Content
   slot {class=OverlaySlot, properties {horizontalAlignment=HAlign_Fill, padding=<ltrb 80 0 80 0>}}
  }
  TextContent {
   widgetClassPath=&VerticalBox
   parent=Content
   widget=TextContent
   slot {class=HorizontalBoxSlot, properties {padding=<ltrb 0 0 96 0>, size {sizeRule=Fill, value=1}}}
  }
  PrimaryScrollBox {
   widgetClassPath=&ScrollBox
   bIsVariable=true
   parent=TextContent
   widget=PrimaryScrollBox
   properties {
    widgetBarStyle {
     thickness=16
     draggedThumbImage {
      bIsDynamicallyLoaded=false
      drawAs=Image
      imageType=FullColor
      mirroring=NoMirror
      resourceName=../../../Engine/Content/Slate/Common/Scrollbar_Thumb.png
      resourceObject=/Game/Athena/UI/ReplayBrowser/M_Ui_Skewed_Vfor9slices_BG_Separator_Inst
      tiling=NoTile
      imageSize=<0 0>
      margin=<ltrb 0.25 0.25 0.25 0.25>
      outlineSettings {
       bUseBrushTransparency=false
       roundingType=HalfHeightRadius
       width=0
       color {colorUseRule=UseColor_Specified, specifiedColor=<rgba 0 0 0 0>}
       cornerRadii=<0 0 0 0>
      }
      tintColor {colorUseRule=UseColor_Specified, specifiedColor=<rgba 1 1 1 1>}
      uVRegion {bIsValid=false, max=<0 0>, min=<0 0>}
     }
     horizontalBackgroundImage {
      bIsDynamicallyLoaded=false
      drawAs=Image
      imageType=NoImage
      mirroring=NoMirror
      resourceName=None
      resourceObject=&M_Ui_Skewed_Vfor9slices_BG_Separator_Inst
      tiling=NoTile
      imageSize=<0 0>
      margin=<ltrb 0 0 0 0>
      outlineSettings {
       bUseBrushTransparency=false
       roundingType=HalfHeightRadius
       width=0
       color {colorUseRule=UseColor_Specified, specifiedColor=<rgba 0 0 0 0>}
       cornerRadii=<0 0 0 0>
      }
      tintColor {colorUseRule=UseColor_Specified, specifiedColor=<rgba 0 0.2159 1 0.498>}
      uVRegion {bIsValid=false, max=<0 0>, min=<0 0>}
     }
     horizontalBottomSlotImage {
      bIsDynamicallyLoaded=false
      drawAs=NoDrawType
      imageType=FullColor
      mirroring=NoMirror
      resourceName=../../../Engine/Content/Slate/Common/Scrollbar_Background_Horizontal.png
      resourceObject=null
      tiling=NoTile
      imageSize=<8 8>
      margin=<ltrb 0 0 0 0>
      outlineSettings {
       bUseBrushTransparency=false
       roundingType=HalfHeightRadius
       width=0
       color {colorUseRule=UseColor_Specified, specifiedColor=<rgba 0 0 0 0>}
       cornerRadii=<0 0 0 0>
      }
      tintColor {colorUseRule=UseColor_Specified, specifiedColor=<rgba 1 1 1 1>}
      uVRegion {bIsValid=false, max=<0 0>, min=<0 0>}
     }
     horizontalTopSlotImage {
      bIsDynamicallyLoaded=false
      drawAs=NoDrawType
      imageType=FullColor
      mirroring=NoMirror
      resourceName=../../../Engine/Content/Slate/Common/Scrollbar_Background_Horizontal.png
      resourceObject=null
      tiling=NoTile
      imageSize=<8 8>
      margin=<ltrb 0 0 0 0>
      outlineSettings {
       bUseBrushTransparency=false
       roundingType=HalfHeightRadius
       width=0
       color {colorUseRule=UseColor_Specified, specifiedColor=<rgba 0 0 0 0>}
       cornerRadii=<0 0 0 0>
      }
      tintColor {colorUseRule=UseColor_Specified, specifiedColor=<rgba 1 1 1 1>}
      uVRegion {bIsValid=false, max=<0 0>, min=<0 0>}
     }
     hoveredThumbImage {
      bIsDynamicallyLoaded=false
      drawAs=Image
      imageType=FullColor
      mirroring=NoMirror
      resourceName=../../../Engine/Content/Slate/Common/Scrollbar_Thumb.png
      resourceObject=&M_Ui_Skewed_Vfor9slices_BG_Separator_Inst
      tiling=NoTile
      imageSize=<0 0>
      margin=<ltrb 0.25 0.25 0.25 0.25>
      outlineSettings {
       bUseBrushTransparency=false
       roundingType=HalfHeightRadius
       width=0
       color {colorUseRule=UseColor_Specified, specifiedColor=<rgba 0 0 0 0>}
       cornerRadii=<0 0 0 0>
      }
      tintColor {colorUseRule=UseColor_Specified, specifiedColor=<rgba 0.5573 0.808 1 1>}
      uVRegion {bIsValid=false, max=<0 0>, min=<0 0>}
     }
     normalThumbImage {
      bIsDynamicallyLoaded=false
      drawAs=Image
      imageType=FullColor
      mirroring=NoMirror
      resourceName=../../../Engine/Content/Slate/Common/Scrollbar_Thumb.png
      resourceObject=&M_Ui_Skewed_Vfor9slices_BG_Separator_Inst
      tiling=NoTile
      imageSize=<12 12>
      margin=<ltrb 0.25 0.25 0.25 0.25>
      outlineSettings {
       bUseBrushTransparency=false
       roundingType=HalfHeightRadius
       width=0
       color {colorUseRule=UseColor_Specified, specifiedColor=<rgba 0 0 0 0>}
       cornerRadii=<0 0 0 0>
      }
      tintColor {colorUseRule=UseColor_Specified, specifiedColor=<rgba 0.2122 0.6584 1 1>}
      uVRegion {bIsValid=false, max=<0 0>, min=<0 0>}
     }
     verticalBackgroundImage {
      bIsDynamicallyLoaded=false
      drawAs=Image
      imageType=NoImage
      mirroring=NoMirror
      resourceName=None
      resourceObject=&M_Ui_Skewed_Vfor9slices_BG_Separator_Inst
      tiling=NoTile
      imageSize=<0 0>
      margin=<ltrb 0 0 0 0>
      outlineSettings {
       bUseBrushTransparency=false
       roundingType=HalfHeightRadius
       width=0
       color {colorUseRule=UseColor_Specified, specifiedColor=<rgba 0 0 0 0>}
       cornerRadii=<0 0 0 0>
      }
      tintColor {colorUseRule=UseColor_Specified, specifiedColor=<rgba 0 0.2159 1 0.498>}
      uVRegion {bIsValid=false, max=<0 0>, min=<0 0>}
     }
     verticalBottomSlotImage {
      bIsDynamicallyLoaded=false
      drawAs=NoDrawType
      imageType=FullColor
      mirroring=NoMirror
      resourceName=../../../Engine/Content/Slate/Common/Scrollbar_Background_Horizontal.png
      resourceObject=null
      tiling=NoTile
      imageSize=<8 8>
      margin=<ltrb 0 0 0 0>
      outlineSettings {
       bUseBrushTransparency=false
       roundingType=HalfHeightRadius
       width=0
       color {colorUseRule=UseColor_Specified, specifiedColor=<rgba 0 0 0 0>}
       cornerRadii=<0 0 0 0>
      }
      tintColor {colorUseRule=UseColor_Specified, specifiedColor=<rgba 1 1 1 1>}
      uVRegion {bIsValid=false, max=<0 0>, min=<0 0>}
     }
     verticalTopSlotImage {
      bIsDynamicallyLoaded=false
      drawAs=NoDrawType
      imageType=FullColor
      mirroring=NoMirror
      resourceName=../../../Engine/Content/Slate/Common/Scrollbar_Background_Horizontal.png
      resourceObject=null
      tiling=NoTile
      imageSize=<8 8>
      margin=<ltrb 0 0 0 0>
      outlineSettings {
       bUseBrushTransparency=false
       roundingType=HalfHeightRadius
       width=0
       color {colorUseRule=UseColor_Specified, specifiedColor=<rgba 0 0 0 0>}
       cornerRadii=<0 0 0 0>
      }
      tintColor {colorUseRule=UseColor_Specified, specifiedColor=<rgba 1 1 1 1>}
      uVRegion {bIsValid=false, max=<0 0>, min=<0 0>}
     }
    }
    widgetStyle {
     barThickness=9
     bottomShadowBrush {
      bIsDynamicallyLoaded=false
      drawAs=NoDrawType
      imageType=FullColor
      mirroring=NoMirror
      resourceName=../../../Engine/Content/Slate/Common/ScrollBoxShadowLeft.png
      resourceObject=null
      tiling=NoTile
      imageSize=<8 16>
      margin=<ltrb 1 0.5 0 0.5>
      outlineSettings {
       bUseBrushTransparency=false
       roundingType=HalfHeightRadius
       width=0
       color {colorUseRule=UseColor_Specified, specifiedColor=<rgba 0 0 0 0>}
       cornerRadii=<0 0 0 0>
      }
      tintColor {colorUseRule=UseColor_Specified, specifiedColor=<rgba 1 1 1 1>}
      uVRegion {bIsValid=false, max=<0 0>, min=<0 0>}
     }
     horizontalScrolledContentPadding=<ltrb 0 0 1 0>
     leftShadowBrush {
      bIsDynamicallyLoaded=false
      drawAs=NoDrawType
      imageType=FullColor
      mirroring=NoMirror
      resourceName=../../../Engine/Content/Slate/Common/ScrollBoxShadowLeft.png
      resourceObject=null
      tiling=NoTile
      imageSize=<8 16>
      margin=<ltrb 1 0.5 0 0.5>
      outlineSettings {
       bUseBrushTransparency=false
       roundingType=HalfHeightRadius
       width=0
       color {colorUseRule=UseColor_Specified, specifiedColor=<rgba 0 0 0 0>}
       cornerRadii=<0 0 0 0>
      }
      tintColor {colorUseRule=UseColor_Specified, specifiedColor=<rgba 1 1 1 1>}
      uVRegion {bIsValid=false, max=<0 0>, min=<0 0>}
     }
     rightShadowBrush {
      bIsDynamicallyLoaded=false
      drawAs=NoDrawType
      imageType=FullColor
      mirroring=NoMirror
      resourceName=../../../Engine/Content/Slate/Common/ScrollBoxShadowRight.png
      resourceObject=null
      tiling=NoTile
      imageSize=<8 16>
      margin=<ltrb 0 0.5 1 0.5>
      outlineSettings {
       bUseBrushTransparency=false
       roundingType=HalfHeightRadius
       width=0
       color {colorUseRule=UseColor_Specified, specifiedColor=<rgba 0 0 0 0>}
       cornerRadii=<0 0 0 0>
      }
      tintColor {colorUseRule=UseColor_Specified, specifiedColor=<rgba 1 1 1 1>}
      uVRegion {bIsValid=false, max=<0 0>, min=<0 0>}
     }
     topShadowBrush {
      bIsDynamicallyLoaded=false
      drawAs=NoDrawType
      imageType=FullColor
      mirroring=NoMirror
      resourceName=../../../Engine/Content/Slate/Common/ScrollBoxShadowLeft.png
      resourceObject=null
      tiling=NoTile
      imageSize=<8 16>
      margin=<ltrb 1 0.5 0 0.5>
      outlineSettings {
       bUseBrushTransparency=false
       roundingType=HalfHeightRadius
       width=0
       color {colorUseRule=UseColor_Specified, specifiedColor=<rgba 0 0 0 0>}
       cornerRadii=<0 0 0 0>
      }
      tintColor {colorUseRule=UseColor_Specified, specifiedColor=<rgba 1 1 1 1>}
      uVRegion {bIsValid=false, max=<0 0>, min=<0 0>}
     }
     verticalScrolledContentPadding=<ltrb 0 0 0 1>
    }
   }
   slot {class=VerticalBoxSlot, properties {padding=<ltrb 0 50 0 50>, size {sizeRule=Fill, value=1}}}
  }
  VBox_PlatformSpecific {widgetClassPath=&VerticalBox, bIsVariable=true, parent=PrimaryScrollBox, widget=VBox_PlatformSpecific, slot {class=ScrollBoxSlot}}
  WidgetSwitcher_PlatformSpecific {
   widgetClassPath=&WidgetSwitcher
   bIsVariable=true
   parent=VBox_PlatformSpecific
   widget=WidgetSwitcher_PlatformSpecific
   slot {class=VerticalBoxSlot}
  }
  VerticalBox_PlatformSpecificEntries {
   widgetClassPath=&VerticalBox
   bIsVariable=true
   parent=WidgetSwitcher_PlatformSpecific
   widget=VerticalBox_PlatformSpecificEntries
   slot {class=WidgetSwitcherSlot}
  }
  SubscriptionModalEntry_5 {
   widgetClassPath=&/SubscriptionModalEntry
   bIsVariable=true
   parent=VerticalBox_PlatformSpecificEntries
   widget=SubscriptionModalEntry_5
   properties {text=Fortnite Crew is offered on a subscription basis with recurring payments. If you want to continue your Fortnite Crew membership you will need to subscribe to Fortnite Crew with a payment method.}
   slot {class=VerticalBoxSlot}
  }
  SubscriptionModalEntry_3 {
   widgetClassPath=&SubscriptionModalEntry
   bIsVariable=true
   parent=VerticalBox_PlatformSpecificEntries
   widget=SubscriptionModalEntry_3
   properties {text=You will be charged immediately when you subscribe. The validity period will be extended by the selected subscription term.}
   slot {class=VerticalBoxSlot}
  }
  SubscriptionModalEntry_4 {
   widgetClassPath=&SubscriptionModalEntry
   bIsVariable=true
   parent=VerticalBox_PlatformSpecificEntries
   widget=SubscriptionModalEntry_4
   properties {text="Payment methods will be automatically charged monthly. Cancel at anytime. View \"Subscription Terms\" for more info."}
   slot {class=VerticalBoxSlot}
  }
  VerticalBox_WrongPlatform {
   widgetClassPath=&VerticalBox
   bIsVariable=true
   parent=WidgetSwitcher_PlatformSpecific
   widget=VerticalBox_WrongPlatform
   slot {class=WidgetSwitcherSlot}
  }
  SubscriptionModalEntry {
   widgetClassPath=&SubscriptionModalEntry
   bIsVariable=true
   parent=VerticalBox_WrongPlatform
   widget=SubscriptionModalEntry
   properties {text=Visit the platform on which you originally joined The Crew.}
   slot {class=VerticalBoxSlot}
  }
  SubscriptionModalEntry_2 {
   widgetClassPath=&SubscriptionModalEntry
   bIsVariable=true
   parent=VerticalBox_WrongPlatform
   widget=SubscriptionModalEntry_2
   properties {text=Return to this screen.}
   slot {class=VerticalBoxSlot}
  }
  SubscriptionModalEntry_6 {
   widgetClassPath=&SubscriptionModalEntry
   bIsVariable=true
   parent=VerticalBox_WrongPlatform
   widget=SubscriptionModalEntry_6
   properties {text="Follow the final resolve issue step(s)."}
   slot {class=VerticalBoxSlot}
  }
  ImageContent {widgetClassPath=&VerticalBox, parent=Content, widget=ImageContent, slot {class=HorizontalBoxSlot}}
  WBP_SubscriberSince {
   widgetClassPath=&/WBP_SubscriberSince
   bIsVariable=true
   parent=ImageContent
   widget=WBP_SubscriberSince
   slot {class=VerticalBoxSlot, properties {verticalAlignment=VAlign_Center, size {sizeRule=Fill, value=1}}}
  }
  HardwareVisibilityBorder_TouchBack {
   widgetClassPath=&CommonHardwareVisibilityBorder
   parent=Overlay_31
   widget=HardwareVisibilityBorder_TouchBack
   properties {visibilityQuery {
     autoDescription=" ANY( Input.Touch )"
     tokenStreamVersion=0
     userDescription=""
     queryTokenStream [0,1,1,1,0]
     tagDictionary [{tagName=Input.Touch}]
    }}
   slot {class=OverlaySlot, properties {horizontalAlignment=HAlign_Right, padding=<ltrb 0 32 32 0>}}
  }
  SafeZone_TouchBack {
   widgetClassPath=&SafeZone
   bIsVariable=true
   parent=HardwareVisibilityBorder_TouchBack
   widget=SafeZone_TouchBack
   slot {class=BorderSlot, properties {padding=<ltrb 0 0 0 0>}}
  }
  Button_TouchBack {
   widgetClassPath=/FortUILibrary/Foundation/Buttons/CloseButton
   bIsVariable=true
   parent=SafeZone_TouchBack
   widget=Button_TouchBack
   properties {
    onButtonBaseClicked=(&self:WidgetTree.Button_TouchBack.HandleOnButtonClicked)
    onButtonBaseDoubleClicked=(&self:WidgetTree.Button_TouchBack.HandleOnButtonDoubleClicked)
    onButtonBaseHovered=(&self:WidgetTree.Button_TouchBack.HandleOnButtonHovered)
    onButtonBaseUnhovered=(&self:WidgetTree.Button_TouchBack.HandleOnButtonUnhovered)
    onSelectedChangedBase=(&self:WidgetTree.Button_TouchBack.HandleOnSelectedChanged)
   }
   slot {class=SafeZoneSlot}
  }
 ]
}
