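# CrewBenefitsScreen is the Fortnite Crew subscription screen, showing benefit tiles in a scrolling row with a detail panel.
# Activation builds tiles from CrewVM's benefit groups, plays the intro and starts purchase analytics.
# Hovered benefits update the info, entitlement message and layered subscribed, default-gradient or season-pass backgrounds, with a retriggerable delay before reverting.
# Tile clicks open benefit details, while the purchase button switches on subscription state to purchase, rejoin or continue.
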WidgetBlueprint /Crew/UI/CrewBenefitsScreen
animations [
 ContextTransition {
  binding_count=4
  length_seconds=0.15
  object=&self:ContextTransition
  sequence {bindings [
    CrewBenefitInfo {
     id=#1
     possessed_class=CrewBenefitInfo_C
     tracks [{
       type=MovieSceneFloatTrack
       display_name=Render Opacity
       sections [{type=MovieSceneFloatSection, end_frame=3, start_frame=0, channels [None {type=MovieSceneScriptingFloatChannel, keys [[0,0],[3,1]]}]}]
      }]
    }
    Image_BG_Benefit_Default {
     id=#2
     possessed_class=FortLazyImage
     tracks [{
       type=MovieSceneFloatTrack
       display_name=Render Opacity
       sections [{type=MovieSceneFloatSection, end_frame=3, start_frame=0, channels [None {type=MovieSceneScriptingFloatChannel, keys [[0,0],[3,1]]}]}]
      }]
    }
    CMSImage_BG_Benefit_Loaded {
     id=#3
     possessed_class=EpicCMSImage
     tracks [{
       type=MovieSceneFloatTrack
       display_name=Render Opacity
       sections [{type=MovieSceneFloatSection, end_frame=3, start_frame=0, channels [None {type=MovieSceneScriptingFloatChannel, keys [[0,0],[3,1]]}]}]
      }]
    }
    Image_BG_Benefit_Season {
     id=#4
     possessed_class=FortLazyImage
     tracks [{
       type=MovieSceneFloatTrack
       display_name=Render Opacity
       sections [{type=MovieSceneFloatSection, end_frame=3, start_frame=0, channels [None {type=MovieSceneScriptingFloatChannel, keys [[0,0],[3,1]]}]}]
      }]
    }
   ]}
 }
 Intro {
  binding_count=6
  length_seconds=0.6
  object=&self:Intro
  sequence {bindings [
    CrewBenefitInfo {
     id=#5
     possessed_class=CrewBenefitInfo_C
     tracks [
      {
       type=MovieSceneFloatTrack
       display_name=Render Opacity
       sections [{type=MovieSceneFloatSection, end_frame=7, start_frame=0, channels [None {type=MovieSceneScriptingFloatChannel, keys [[0,0],[7,1]]}]}]
      }
      {
       type=MovieScene2DTransformTrack
       sections [{
         type=MovieScene2DTransformSection
         end_frame=10
         start_frame=0
         channels [Translation.Y {type=MovieSceneScriptingFloatChannel, keys [[0,100],[10,0]]}]
        }]
      }
     ]
    }
    CMSImage_BG_Crew_Subbed {
     id=#6
     possessed_class=EpicCMSImage
     tracks [{
       type=MovieSceneFloatTrack
       display_name=Render Opacity
       sections [{
         type=MovieSceneFloatSection
         end_frame=12
         start_frame=0
         channels [None {type=MovieSceneScriptingFloatChannel, keys [[0,0],[4.9997,0],[11.9997,1]]}]
        }]
      }]
    }
    BenefitsScrollBox {
     id=#7
     possessed_class=AthenaScrollBox
     tracks [
      {
       type=MovieSceneFloatTrack
       display_name=Render Opacity
       sections [{type=MovieSceneFloatSection, end_frame=7, start_frame=0, channels [None {type=MovieSceneScriptingFloatChannel, keys [[0,0],[7,1]]}]}]
      }
      {
       type=MovieScene2DTransformTrack
       sections [{
         type=MovieScene2DTransformSection
         end_frame=10
         start_frame=0
         channels [Translation.Y {type=MovieSceneScriptingFloatChannel, keys [[0,100],[10,0]]}]
        }]
      }
     ]
    }
    SB_CrewEntitlement {
     id=#8
     possessed_class=SizeBox
     tracks [
      {
       type=MovieSceneFloatTrack
       display_name=Render Opacity
       sections [{type=MovieSceneFloatSection, end_frame=9, start_frame=2, channels [None {type=MovieSceneScriptingFloatChannel, keys [[2,0],[9,1]]}]}]
      }
      {
       type=MovieScene2DTransformTrack
       sections [{
         type=MovieScene2DTransformSection
         end_frame=12
         start_frame=2
         channels [
          Translation.X {type=MovieSceneScriptingFloatChannel, keys [[2,100],[12,0]]}
          Translation.Y {type=MovieSceneScriptingFloatChannel, const [0,1]}
         ]
        }]
      }
     ]
    }
    MKBScrollArrows {
     id=#9
     possessed_class=HorizontalBox
     tracks [
      {
       type=MovieSceneFloatTrack
       display_name=Render Opacity
       sections [{type=MovieSceneFloatSection, end_frame=7, start_frame=0, channels [None {type=MovieSceneScriptingFloatChannel, keys [[0,0],[7,1]]}]}]
      }
      {
       type=MovieScene2DTransformTrack
       sections [{
         type=MovieScene2DTransformSection
         end_frame=10
         start_frame=0
         channels [Translation.Y {type=MovieSceneScriptingFloatChannel, keys [[0,100],[10,0]]}]
        }]
      }
     ]
    }
    CMSImage_BG_Benefit_Loaded {
     id=#10
     possessed_class=EpicCMSImage
     tracks [
      {
       type=MovieSceneFloatTrack
       display_name=Render Opacity
       sections [{type=MovieSceneFloatSection, end_frame=7, start_frame=0, channels [None {type=MovieSceneScriptingFloatChannel, keys [[0,0],[7,1]]}]}]
      }
      {
       type=MovieScene2DTransformTrack
       sections [{
         type=MovieScene2DTransformSection
         end_frame=10
         start_frame=0
         channels [Translation.Y {type=MovieSceneScriptingFloatChannel, keys [[0,100],[10,0]]}]
        }]
      }
     ]
    }
   ]}
 }
]
graphs [
 BP_GetDesiredFocusTarget {nodes [
   Get Desired Focus Target()
   > n1 = GetDesiredFocusTarget!(execute, self:Crew Benefit Info Object Reference=Get CrewBenefitInfo() -> (CrewBenefitInfo:Crew Benefit Info Object Reference)) -> (then, Desired:Widget Object Reference)
   n2 = GetAllEntries!(self:Dynamic Entry Box Base Object Reference=Get DynamicEntryBox() -> (DynamicEntryBox:Fort Dynamic Entry Box Object Reference)) -> (ReturnValue:Array of User Widget Object References)
   > Return Node!(
    execute
    ReturnValue:Widget Object Reference=Select(
     Option 0:Widget Object Reference="Get (a copy)"(Array:Array of User Widget Object References=@n2) -> (Output:User Widget Object Reference)
     Option 1:Widget Object Reference=@n1
     Index=IsValid(Object=@n1)
    ) -> (ReturnValue:Widget Object Reference)
   )
  ]}
 BenefitClicked {nodes [
   n1 = BenefitClicked!() -> (then, Button:Common Button Base Object Reference) {is=title K2Node_FunctionEntry}
   > n2 = Cast To CrewBenefitTile!(execute, Object=@n1) -> (then, CastFailed, AsCrew Benefit Tile:Crew Benefit Tile Object Reference) {is=title K2Node_DynamicCast}
   n3 = Get CrewBenefitVM!(self:Crew Benefit Tile Object Reference=@n2) -> (CrewBenefitVM:Crew Benefit VM Object Reference)
   > Set CurrentCrewBenefitVM(CurrentCrewBenefitVM:Crew Benefit VM Object Reference=@n3) -> (Output_Get:Crew Benefit VM Object Reference)
   n4 = Get CrewSelectionVM() -> (CrewSelectionVM:Crew Selection VM Object Reference)
   > n5 = Branch(Condition="Equal (Object)"(A=@n3, B=Get HoveredBenefit!(self:Crew Selection VM Object Reference=@n4) -> (HoveredBenefit:Crew Benefit VM Object Reference)))
   n6 = GetType!(self:Crew Benefit VM Object Reference=Get CurrentCrewBenefitVM() -> (CurrentCrewBenefitVM:Crew Benefit VM Object Reference)) -> (ReturnValue:ECrewTileType Enum=None)
   > AnalyticsSendBattlePassPurchaseInteraction!(
    execute
    self:Fort Subscription Content Base Object Reference
    Interaction:String=Append(A=RequestMoreInfoFor-, B=Enum to String!(Enumerator:ECrewTileType Enum=@n6) -> (ReturnValue:String))
   ) -> (then)
   > n7 = Switch on ECrewTileType!(execute, Selection:ECrewTileType Enum=@n6) -> (None:Exec, Basic:Exec, BattlePass:Exec, CrewPack:Exec, ProgressiveCosmetic:Exec) {is=title K2Node_SwitchEnum}
   Crew Pack More Info!(execute=@n7.CrewPack, self, Benefit VM:Crew Benefit VM Object Reference=Get CurrentCrewBenefitVM()) -> (then)
   RequestSwitchToSubscriptionContent!(
    execute=@n7.ProgressiveCosmetic
    self:Fort Subscription Content Base Object Reference
    SubscriptionContentTab:ESubscriptionContentTab Enum=ProgressiveItemScreen
   ) -> (then)
   Branch(execute=@n5.else, Condition=IsUsingTouch(PlayerController=GetOwningPlayer()))
   > Set HoveredBenefit!(
    execute
    HoveredBenefit:Crew Benefit VM Object Reference=Get CurrentCrewBenefitVM()
    self:Crew Selection VM Object Reference=Get CrewSelectionVM()
   ) -> (then, Output_Get:Crew Benefit VM Object Reference)
   Reroute Node()
   > RequestSwitchToSubscriptionContent(SubscriptionContentTab=TemporaryItemsScreen)
   Request Battle Pass More Info(self:Object Reference)
  ]}
 CloseScreen {nodes [CloseScreen() {is=title K2Node_FunctionEntry}]}
 CrewPurchased {nodes [CrewPurchased() {is=title K2Node_FunctionEntry}]}
 DownToBenefits {nodes [
   DownToBenefits!() -> (then, Navigation:EUINavigation Enum) {is=title K2Node_FunctionEntry}
   n1 = GetAllEntries(self=Get DynamicEntryBox())
   > n2 = For Each Loop with Break!(Exec:Exec, Array:Array of User Widget Object References=@n1, Break:Exec=@n5.then) -> (LoopBody:Exec, Array Element:User Widget Object Reference, Array Index:int, Completed:Exec)
   > n3 = Cast To CrewBenefitTile(Object=@n2)
   n4 = "Equal (Object)"(A=Get CrewBenefitVM(self=@n3), B=Get LastHoveredBenefit() -> (LastHoveredBenefit:Crew Benefit VM Object Reference))
   Return Node(execute=@n2.Completed, ReturnValue=Select(Option 0="Get (a copy)"(Array=@n1), Option 1=@n3, Index=@n4))
   > n5 = Branch(Condition=@n4)
  ]}
 EventGraph {nodes [
   "Purchase Clicked (CrewBenefitInfo)"!() -> (OutputDelegate, then, HoldTime:double) {is=title K2Node_ComponentBoundEvent}
   > n1 = Switch on EMcpSubscriptionState!(
    execute
    Selection:EMcpSubscriptionState Enum=Get SubscriptionState!(self:Crew VM Object Reference=Get CrewVM() -> (CrewVM:Crew VM Object Reference)) -> (SubscriptionState:EMcpSubscriptionState Enum=Inactive)
   ) -> (
    Inactive:Exec
    Active:Exec
    ActiveWithExternalBilling:Exec
    NotAutoRenewable:Exec
    Canceled:Exec
    PaymentProcessError:Exec
    PlatformAuthError:Exec
    RewardsDelayed:Exec
    Unavailable:Exec
    Unknown:Exec
   ) {is=title K2Node_SwitchEnum}
   n2 = RejoinSubscription!(execute:Exec=@n1.Canceled) -> (then:Exec, OnOperationCompleted:Exec, Success:bool=~) {is=title K2Node_AsyncAction}
   n3 = Branch(execute=@n2.OnOperationCompleted, Condition=@n2)
   > Call Crew Purchased!(execute:Exec, self:Crew Benefits Screen Object Reference) -> (then:Exec) {is=title K2Node_CallDelegate}
   n4 = QueueModalWidgetClassAsync!(
    execute:Exec=@n3.else
    WidgetClass:Common Activatable Widget Class Reference=/Crew/UI/Subscription/SubscriptionRejoinModal.SubscriptionRejoinModal_C
   ) -> (then:Exec, OnModalWidgetQueued:Exec, ActivatableWidget:Common Activatable Widget Object Reference) {is=title K2Node_AsyncAction}
   n5 = Cast To SubscriptionRejoinModal!(execute=@n4.OnModalWidgetQueued, Object=@n4) -> (then, CastFailed, AsSubscription Rejoin Modal:Subscription Rejoin Modal Object Reference) {is=title K2Node_DynamicCast}
   > Set CrewSubscriptionModalInfoVM!(
    execute
    CrewSubscriptionModalInfoVM:Crew Subscription Modal Info VM Object Reference=GetModalViewModelById!(
     self:Crew VM Object Reference=Get CrewVM()
     ModalId:String=Get ModalId!(self:Subscription Modal Base Object Reference=@n5) -> (ModalId:String)
    ) -> (ReturnValue:Crew Subscription Modal Info VM Object Reference)
    self:Subscription Rejoin Modal Object Reference=@n5
   ) -> (then, Output_Get:Crew Subscription Modal Info VM Object Reference)
   > FireSubscriptionSubviewEvent!(execute, self, Widget:Widget Object Reference=@n5) -> (then)
   n6 = Create Widget!(
    execute:Exec=@n1.Inactive
    Class:User Widget Class Reference=/Crew/UI/Subscription/SubscriptionLegalInfoModal.SubscriptionLegalInfoModal_C
    OwningPlayer:Player Controller Object Reference=GetOwningPlayer()
   ) -> (then:Exec, ReturnValue:Subscription Legal Info Modal Object Reference)
   > TryPushWidgetAsModalPanel!(execute, PushWidget:User Widget Object Reference=@n6) -> (then)
   > n7 = SetPurchaseFlow!(execute, self:Subscription Legal Info Modal Object Reference=@n6, IsPurchase:bool=true) -> (then)
   n8 = PurchaseConfirmed() {is=title K2Node_CustomEvent}
   Bind Event to Sub Confirmed(execute=@n7.then, self:Subscription Legal Info Modal Object Reference=@n6, Delegate=@n8) {is=title K2Node_AddDelegate}
   > FireSubscriptionSubviewEvent(Widget=@n6)
   n9 = ContinueSubscription!(execute:Exec=@n1.NotAutoRenewable) -> (then:Exec, OnOperationCompleted:Exec, Success:bool=~) {is=title K2Node_AsyncAction}
   n10 = Branch(execute=@n9.OnOperationCompleted, Condition=@n9)
   > Call Crew Purchased()
   n11 = QueueModalWidgetClassAsync(execute=@n10.else, WidgetClass=/Crew/UI/Subscription/SubscriptionContinueModal.SubscriptionContinueModal_C)
   n12 = Cast To SubscriptionContinueModal!(execute=@n11.OnModalWidgetQueued, Object=@n11) -> (then, CastFailed, AsSubscription Continue Modal:Subscription Continue Modal Object Reference) {is=title K2Node_DynamicCast}
   > n13 = Set CrewSubscriptionModalInfoVM(
    CrewSubscriptionModalInfoVM=GetModalViewModelById(self=Get CrewVM(), ModalId=Get ModalId(self=@n12))
    self:Subscription Continue Modal Object Reference=@n12
   )
   n14 = SubscribeStarted() {is=title K2Node_CustomEvent}
   Bind Event to Subscribe Started(execute=@n13.then, self:Subscription Continue Modal Object Reference=@n12, Delegate=@n14) {is=title K2Node_AddDelegate}
   > FireSubscriptionSubviewEvent(Widget=@n12)
   "On Clicked (ScrollRight)"!() -> (OutputDelegate, then, Button:Common Button Base Object Reference) {is=title K2Node_ComponentBoundEvent}
   n15 = Get BenefitsScrollBox() -> (BenefitsScrollBox:Athena Scroll Box Object Reference)
   n16 = GetScrollOffsetOfEnd!(self:Scroll Box Object Reference=@n15) -> (ReturnValue:float)
   n17 = GetScrollOffset!(self:Scroll Box Object Reference=@n15) -> (ReturnValue:float)
   n18 = float + float(A=@n17, B=Get ButtonScrollAmount() -> (ButtonScrollAmount:double))
   n19 = Get BenefitsScrollBox()
   > SetScrollOffset!(execute, self:Scroll Box Object Reference=@n19, NewScrollOffset:float=SelectFloat(A=@n16, B=@n18, bPickA="float >= float"(A=@n18, B=@n16))) -> (then)
   > n20 = Branch(Condition="Equal (Float)"(A=@n17, B=@n16))
   > n21 = SetVisibility(self=Get ScrollRight() -> (ScrollRight:WBP UIKit Button Regular Object Reference), InVisibility=Collapsed)
   SetVisibility(
    execute=[@n21.then, @n20.else]
    self=Get ScrollLeft() -> (ScrollLeft:WBP UIKit Button Regular Object Reference)
    InVisibility=SelfHitTestInvisible
   )
   "On Clicked (TermsButton)"!() -> (OutputDelegate, then, Button:Common Button Base Object Reference) {is=title K2Node_ComponentBoundEvent}
   > n22 = Create Widget(Class=&SubscriptionLegalInfoModal_C, OwningPlayer=GetOwningPlayer())
   > TryPushWidgetAsModalPanel(PushWidget=@n22)
   > n23 = SetPurchaseFlow(self=@n22, IsPurchase=false)
   n24 = OnCloseTerms() {is=title K2Node_CustomEvent}
   Bind Event to On Widget Deactivated(execute=@n23.then, self:Common Activatable Widget Object Reference=@n22, Delegate=@n24) {is=title K2Node_AddDelegate}
   > n25 = GetPurchaseTraceId!(execute) -> (then, ReturnValue:Flow Analytics Additional Info Structure)
   > Fire Event Sub View Opened!(execute, View:Widget Object Reference=@n22, AdditionalInfo:Flow Analytics Additional Info Structure=@n25, bIsOverlay:bool=false) -> (then)
   n26 = "On Clicked (BackButton)"!() -> (OutputDelegate, then, Button:Common Button Base Object Reference) {is=title K2Node_ComponentBoundEvent}
   > n27 = GetPurchaseTraceId()
   > Fire Event Element Interacted(Element=@n26.Button, AdditionalInfo=@n27)
   > Call Close Screen!(execute:Exec, self:Crew Benefits Screen Object Reference) -> (then:Exec) {is=title K2Node_CallDelegate}
   "On Clicked (ScrollLeft)"!() -> (OutputDelegate, then, Button:Common Button Base Object Reference) {is=title K2Node_ComponentBoundEvent}
   n28 = GetScrollOffset(self=Get BenefitsScrollBox())
   n29 = float - float(A=@n28, B=Get ButtonScrollAmount())
   > SetScrollOffset(self=@n19, NewScrollOffset=SelectFloat(A=0, B=@n29, bPickA="float <= float"(A=@n29) -> (ReturnValue=~)))
   > n30 = Branch(Condition="Equal (Float)"(A=@n28))
   > n31 = SetVisibility(self=Get ScrollLeft(), InVisibility=Collapsed)
   SetVisibility(execute=[@n31.then, @n30.else], self=Get ScrollRight(), InVisibility=SelfHitTestInvisible)
   n32 = Crew Pack More Info!() -> (OutputDelegate, then, Benefit VM:Crew Benefit VM Object Reference) {is=title K2Node_CustomEvent}
   > n33 = PushPrimaryContentClassAsync!(execute:Exec, WidgetClass:Common Activatable Widget Class Reference=/Crew/UI/Preview/CrewItemPreviewSreen.CrewItemPreviewSreen_C) -> (then:Exec, OnPrimaryContentPushed:Exec, ActivatableWidget:Common Activatable Widget Object Reference) {is=title K2Node_AsyncAction}
   n34 = Cast To CrewItemPreviewSreen!(execute=@n33.OnPrimaryContentPushed, Object=@n33) -> (then, CastFailed, AsCrew Item Preview Sreen:Crew Item Preview Sreen Object Reference) {is=title K2Node_DynamicCast}
   > n35 = Set with Broadcast Rewards!(
    execute
    Rewards:Array of Fort Item VM Object References=Get CrewPackItems!(self:Crew Benefit VM Object Reference=@n32.Benefit VM) -> (CrewPackItems:Array of Fort Item VM Object References)
    self:Crew Item Preview Sreen Object Reference=@n34
   ) -> (then, Output_Get:Array of Fort Item VM Object References)
   n36 = SubscribeComplete() {is=title K2Node_CustomEvent}
   Bind Event to Subscribe Completed(execute=@n35.then, self:Crew Item Preview Sreen Object Reference=@n34, Delegate=@n36) {is=title K2Node_AddDelegate}
   DelayReturnToDefault() {is=title K2Node_CustomEvent}
   > Set DelayActive(DelayActive:bool=true) -> (Output_Get:bool)
   > n37 = Sequence()
   > RetriggerableDelay(Duration=0.1)
   > Gate!(Enter:Exec, Open:Exec=@n37.then_1, Close:Exec=@n38.then, Toggle:Exec, bStartClosed:bool=true) -> (Exit:Exec) {is=title K2Node_MacroInstance}
   > SetBenefit!(execute, self, HoveredBenefit:Crew Benefit VM Object Reference) -> (then)
   CancelDelayReturnToDefault() {is=title K2Node_CustomEvent}
   > n38 = Set DelayActive(DelayActive=false)
   Event On Activated()
   n39 = Create Event()
   n40 = Get CommonInputSubsystem()
   > Bind Event to On Input Method Changed(self=@n40, Delegate=@n39)
   > HandleInputChanged!(execute, self, bNewInputType:ECommonInputType Enum=GetCurrentInputType(self=@n40)) -> (then)
   > SetupBenefitsList()
   > PlayAnimation(InAnimation=Get Intro())
   > PlayUpsellMusic!(execute, self, bPlay:bool=true) -> (then)
   > DelayUntilNextTick!(execute) -> (then)
   > SetupArrowVisibility()
   > n41 = GetGameUserSettings!(execute) -> (then, ReturnValue:Game User Settings Object Reference)
   n42 = Create Event()
   > Bind Event to On Game User Settings UINeeds Update(self:Game User Settings Object Reference=@n41, Delegate=@n42) {is=title K2Node_AddDelegate}
   n43 = Get CrewVM()
   > n44 = GeneratePurchaseTraceId!(
    execute
    OfferId:String=GetEpicOfferId!(self:Fort Simple Offer VM Object Reference=Get CrewOffer!(self:Crew VM Object Reference=@n43) -> (CrewOffer:Fort Simple Offer VM Object Reference)) -> (ReturnValue:String)
   ) -> (then, AnalyticsAdditionalInfo:Flow Analytics Additional Info Structure)
   > Fire Event Sub View Opened(View=Self-Reference(), AdditionalInfo=@n44, bIsOverlay=false)
   Event PreConstruct()
   > SetCompactUIStyle() {is=title K2Node_CallFunction}
   > SetIndicatorType!(
    execute
    self:WBP UIKit Status Indicator Base Object Reference=Get WBP_UIKit_StatusIndicator!(self:WBP UIKit Info Box Object Reference=Get CrewNoticeMessage() -> (CrewNoticeMessage:WBP UIKit Info Box Object Reference)) -> (WBP_UIKit_StatusIndicator:WBP UIKit Status Indicator Object Reference)
    IndicatorType:E_UI_StatusIndicator_Type Enum=NewEnumerator4
   ) -> (then)
   > SetText(
    self:WBP UIKit Info Box Object Reference=Get ParentalControlsMessage() -> (ParentalControlsMessage:WBP UIKit Info Box Object Reference)
    InText:Text="LOCTEXT(\"<text color=\\\"FFFFFFFF\\\" fontface=\\\"75 Medium\\\">Account setup required:</> to purchase this item or bundle ask your parent or guardian to set up your account from Parental Controls.\")"
   )
   Event On Deactivated()
   > Unbind Event from On Input Method Changed(self=@n40, Delegate=@n39)
   > PlayUpsellMusic(bPlay=false)
   > n45 = GetGameUserSettings()
   > Unbind Event from On Game User Settings UINeeds Update(self:Game User Settings Object Reference=@n45, Delegate=@n42) {is=title K2Node_RemoveDelegate}
   > n46 = Cast To FortPlayerController(Object=GetOwningPlayer())
   > Fire Event Flush Impressions!(execute, FortPC:Fort Player Controller Object Reference=@n46) -> (then)
   > ClearPurchaseUXTraceId!(
    execute
    self:Fort Mcp Context Object Reference=Get FortMcpContext!() -> (ReturnValue:Fort Mcp Context Object Reference) {is=title K2Node_GetBlueprintContextEx}
    DebugSourceContext:String
   ) -> (then)
   Call Crew Purchased(execute=@n49.then)
   Call Crew Purchased(execute=@n51.then)
   DeactivateWidget(execute=@n36.then, self=@n34)
   n47 = GetPurchaseTraceId(execute=@n24.then)
   > Fire Event Sub View Opened(View=Self-Reference(), AdditionalInfo=@n47, bIsOverlay=false)
   n48 = PurchaseSubscription!(execute:Exec=@n8.then) -> (then:Exec, OnOperationCompleted:Exec, Success:bool=~) {is=title K2Node_AsyncAction}
   n49 = Branch(execute=@n48.OnOperationCompleted, Condition=@n48)
   n50 = ContinueSubscription(execute=@n14.then) -> (Success=~)
   n51 = Branch(execute=@n50.OnOperationCompleted, Condition=@n50)
  ]}
 FireSubscriptionSubviewEvent {nodes [
   n1 = FireSubscriptionSubviewEvent!() -> (then, Widget:Widget Object Reference) {is=title K2Node_FunctionEntry}
   > n2 = GetPurchaseTraceId()
   n3 = Get SubscriptionState(self=Get CrewVM()) -> (SubscriptionState=Inactive)
   > Fire Event Sub View Opened(
    View=@n1
    AdditionalInfo="Add Text (Localized)"!(
     Info:"Flow Analytics Additional Info Structure (by ref)"=@n2
     Key:String=SubscriptionState
     Value:"Text (by ref)"=To Text!(Value:EMcpSubscriptionState Enum=@n3) -> (ReturnValue:Text) {is=title K2Node_GenericToText}
    ) -> (ReturnValue:Flow Analytics Additional Info Structure)
    bIsOverlay=false
   )
  ]}
 GetLegalText {nodes [
   n1 = GetLegalText!() -> (then, Disclaimer:Text, PaymentLegal:Text) {is=title K2Node_FunctionEntry}
   > Return Node!(
    execute
    LegalText:Text=Format Text!(Format="LOCTEXT(\"{Disclaimer} {PaymentLegal}\")", Disclaimer:Text=@n1, PaymentLegal:Text=@n1.PaymentLegal) -> (Result)
   )
  ]}
 GetNoticeMessageVisibility {nodes [
   n1 = GetNoticeMessageVisibility!() -> (then, NoticeText:Text) {is=title K2Node_FunctionEntry}
   > Return Node!(
    execute
    NoticeVisibility:ESlateVisibility Enum=Select(Option 0:ESlateVisibility Enum=SelfHitTestInvisible, Option 1:ESlateVisibility Enum=Collapsed, Index=TextIsEmpty(InText=@n1)) -> (ReturnValue:ESlateVisibility Enum)
   )
  ]}
 HandleInputChanged {nodes [
   n1 = HandleInputChanged!() -> (then, bNewInputType:ECommonInputType Enum) {is=title K2Node_FunctionEntry}
   > n2 = Branch(Condition="Equal (Enum)"(A=@n1, B=MouseAndKeyboard))
   n3 = Get MKBScrollArrows() -> (MKBScrollArrows:Horizontal Box Object Reference)
   > SetVisibility(self=@n3, InVisibility=SelfHitTestInvisible)
   SetVisibility(execute=@n2.else, self=@n3, InVisibility=Collapsed)
  ]}
 HandleIsSubscribed {nodes [
   n1 = HandleIsSubscribed!() -> (then, IsSubbed:bool=~) {is=title K2Node_FunctionEntry}
   > n2 = Set IsSubbed(IsSubbed:bool=@n1) -> (Output_Get:bool)
   > UpdateBackgroundVisibility!(
    execute
    self
    IsSubbed:bool=@n2
    IsBenefitHovered:bool=IsValid(Object=Get CrewBenefitVM!() -> (CrewBenefitVM))
    IsSeasonPassBenefit:bool=IsValid(Object=Get SeasonPassVM() -> (SeasonPassVM:Season Pass VM Object Reference))
   ) -> (then)
  ]}
 InitializeSeasonPassSystemVM {nodes [
   InitializeSeasonPassSystemVM!() -> (then, Passes:Array of Season Pass VM Object References) {is=title K2Node_FunctionEntry}
   > SetSeasonPassVMFromBenefit!(execute, self, Benefit:Crew Benefit VM Object Reference=Get CrewBenefitVM()) -> (then)
  ]}
 PlayUpsellMusic {nodes [PlayUpsellMusic!() -> (then, bPlay:bool=~) {is=title K2Node_FunctionEntry}]}
 SetBackgroundColors {nodes [
   n1 = SetBackgroundColors!() -> (then, Benefit:Crew Benefit VM Object Reference) {is=title K2Node_FunctionEntry}
   > Branch(Condition=IsValid(Object=@n1))
   > n2 = GetDynamicMaterial(self=Get Image_BG_Benefit_Default() -> (Image_BG_Benefit_Default:Fort Lazy Image Object Reference))
   > SetVectorParameterValue(
    self=@n2
    ParameterName=Color1
    Value="To LinearColor (Color)"!(InColor:Color Structure=GetColor1!(self:Crew Benefit VM Object Reference=@n1) -> (ReturnValue:Color Structure)) -> (ReturnValue:Linear Color Structure)
   )
   > SetVectorParameterValue(
    self=@n2
    ParameterName=Color2
    Value="To LinearColor (Color)"(InColor=GetColor2!(self:Crew Benefit VM Object Reference=@n1) -> (ReturnValue:Color Structure))
   )
   > SetVectorParameterValue(
    self=@n2
    ParameterName=Color3
    Value="To LinearColor (Color)"(InColor=GetColor3!(self:Crew Benefit VM Object Reference=@n1) -> (ReturnValue:Color Structure))
   )
   > SetVectorParameterValue(
    self=@n2
    ParameterName=Color4
    Value="To LinearColor (Color)"(InColor=GetColor4!(self:Crew Benefit VM Object Reference=@n1) -> (ReturnValue:Color Structure))
   )
  ]}
 SetBenefit {nodes [
   n1 = SetBenefit!() -> (then, HoveredBenefit:Crew Benefit VM Object Reference) {is=title K2Node_FunctionEntry}
   n2 = IsValid(Object=@n1)
   > n3 = Branch(Condition=@n2)
   > n4 = Branch(Condition=Get DelayActive() -> (DelayActive:bool))
   > n5 = CancelDelayReturnToDefault() {is=title K2Node_CallFunction}
   n6 = Set LastHoveredBenefit(execute=[@n5.then, @n4.else], LastHoveredBenefit:Crew Benefit VM Object Reference=Get HoveredBenefit!() -> (HoveredBenefit)) -> (Output_Get:Crew Benefit VM Object Reference)
   SetVisibility(
    execute=[@n11.then, @n6.then]
    self=Get CrewLogo() -> (CrewLogo:Image Object Reference)
    InVisibility=Select(Option 0=Collapsed, Option 1=SelfHitTestInvisible, Index=@n2)
   )
   > SetBenefit!(execute, self:Crew Benefit Info Object Reference=Get CrewBenefitInfo(), Benefit:Crew Benefit VM Object Reference=Get HoveredBenefit()) -> (then) {is=title K2Node_CallFunction}
   > SetSeasonPassVMFromBenefit(Benefit=Get HoveredBenefit())
   > HandleSeasonPassVM!(
    execute
    self:Crew Entitlement Message Object Reference=Get CrewEntitlementMessage() -> (CrewEntitlementMessage:Crew Entitlement Message Object Reference)
    PassVM:Season Pass VM Object Reference=Get SeasonPassVM()
   ) -> (then)
   > UpdateBackgroundVisibility(
    IsSubbed=Get IsSubbed() -> (IsSubbed:bool)
    IsBenefitHovered=IsValid(Object=Get HoveredBenefit())
    IsSeasonPassBenefit=IsValid(Object=Get SeasonPassVM())
   )
   n7 = Get HoveredBenefit()
   > Is Valid(InputObject=@n7)
   n8 = GetTitle!(self:Crew Benefit VM Object Reference=@n7) -> (ReturnValue:Text)
   > n9 = Branch(Condition=TextIsEmpty(InText=@n8))
   "Fire Event Add Impression (Localized)"!(
    execute=@n9.else
    SectionID:"Text (by ref)"=@n8
    SectionContents:Array of Flow Analytics Impression Section Content Structures=Make Array("[0]":Flow Analytics Impression Section Content Structure=Make Flow Analytics Impression Section Content!(ContentID:Text, OfferID:Text, Row:int=-1, Column:int=-1, Index:int=-1) -> (FlowAnalyticsImpressionSectionContent:Flow Analytics Impression Section Content Structure) {is=title K2Node_MakeStruct}) -> (Array:Array of Flow Analytics Impression Section Content Structures)
   ) -> (then)
   n10 = Branch(execute=@n3.else, Condition=Get DelayActive())
   > n11 = Set DelayActive(DelayActive=false)
   DelayReturnToDefault(execute=@n10.else) {is=title K2Node_CallFunction}
  ]}
 SetCompactUIStyle {nodes [
   SetCompactUIStyle() {is=title K2Node_FunctionEntry}
   > Branch(Condition=IsUsingCompactUI())
   > SetPadding!(
    execute
    self:Overlay Slot Object Reference=SlotAsOverlaySlot(Widget=Get SafeZone() -> (SafeZone:Safe Zone Object Reference))
    InPadding_Left:float=48
    InPadding_Top:float=24
    InPadding_Right:float=48
    InPadding_Bottom:float=24
   ) -> (then)
   > ClearWidthOverride(self:Size Box Object Reference=Get SizeBox() -> (SizeBox:Size Box Object Reference))
   > SetPadding(self=SlotAsOverlaySlot(Widget=Get CrewLogo()), InPadding_Left=0, InPadding_Top=0, InPadding_Right=0, InPadding_Bottom=0)
   > SetPadding(
    self=SlotAsOverlaySlot(Widget=Get WBP_PlayerInfo() -> (WBP_PlayerInfo:WBP Player Info Object Reference))
    InPadding_Left=0
    InPadding_Top=0
    InPadding_Right=0
    InPadding_Bottom=0
   )
   > SetPadding(
    self:Vertical Box Slot Object Reference=SlotAsVerticalBoxSlot(Widget=Get HBox_LegalActionBar() -> (HBox_LegalActionBar:Horizontal Box Object Reference))
    InPadding_Left=0
    InPadding_Top=0
    InPadding_Right=0
    InPadding_Bottom=0
   )
   n1 = Get SB_CrewEntitlement() -> (SB_CrewEntitlement:Size Box Object Reference)
   > SetPadding(self:Overlay Slot Object Reference=SlotAsOverlaySlot(Widget=@n1), InPadding_Left=0, InPadding_Top=0, InPadding_Right=0, InPadding_Bottom=360)
   > SetWidthOverride(self=@n1, InWidthOverride=600)
   n2 = SlotAsOverlaySlot(Widget=Get CrewBenefitInfo())
   > SetVerticalAlignment(self=@n2, InVerticalAlignment=VAlign_Bottom)
   n3 = GetViewportSize!() -> (ReturnValue_X:double, ReturnValue_Y:double)
   n4 = float / float(A=@n3, B=@n3.ReturnValue_Y)
   > SetPadding(
    self=@n2
    InPadding_Left=160
    InPadding_Top=0
    InPadding_Right=0
    InPadding_Bottom=SelectFloat(A=368, B=584, bPickA="float >= float"(A=@n4, B=1.34))
   )
  ]}
 SetModalAnalyticContext {nodes [
   n1 = SetModalAnalyticContext!() -> (then, Obj:Object Reference) {is=title K2Node_FunctionEntry}
   > n2 = Cast To SubscriptionModalBase!(execute, Object=@n1) -> (then, CastFailed, AsSubscription Modal Base:Subscription Modal Base Object Reference) {is=title K2Node_DynamicCast}
   > Set AnalyticsContext!(execute, AnalyticsContext:String=CrewPurchaseScreen, self:Subscription Modal Base Object Reference=@n2) -> (then, Output_Get:String)
  ]}
 SetSeasonPassVMFromBenefit {nodes [
   n1 = SetSeasonPassVMFromBenefit!() -> (then, Benefit:Crew Benefit VM Object Reference) {is=title K2Node_FunctionEntry}
   n2 = Get SeasonPassSystemVM() -> (SeasonPassSystemVM:Season Pass System VM Object Reference)
   > n3 = Branch(Condition=AND Boolean(A=IsValid(Object=@n1), B=IsValid(Object=@n2)))
   > Set SeasonPassVM(SeasonPassVM:Season Pass VM Object Reference=FindSeasonPass!(
     self:Season Pass System VM Object Reference=Get SeasonPassSystemVM()
     TemplateId:String=GetTemplateId!(self:Crew Benefit VM Object Reference=@n1) -> (ReturnValue:String)
    ) -> (ReturnValue:Season Pass VM Object Reference)) -> (Output_Get:Season Pass VM Object Reference)
   Set SeasonPassVM(execute=@n3.else)
  ]}
 SetThemeData {nodes [
   n1 = SetThemeData!() -> (then, ThemeData:Season Pass Theme Data Structure) {is=title K2Node_FunctionEntry}
   > SetBrushFromLazyDisplayAsset!(
    execute
    self:Common Lazy Image Object Reference=Get Image_BG_Benefit_Season() -> (Image_BG_Benefit_Season:Fort Lazy Image Object Reference)
    LazyObject:"Soft Object Reference (by ref)"=Break Season Pass Theme Data!(SeasonPassThemeData:"Season Pass Theme Data Structure (by ref)"=@n1) -> (
     GiftboxVaultBackgroundMaterial:Soft Object Reference
     UIBackgroundMaterial:Soft Object Reference
     CustomLogo:Soft Object Reference
     CustomSeasonTitle:Text
     CustomSeasonTitleColor:Linear Color Structure
     Upsell2BackgrounMaterial:Soft Object Reference
     ImageReferenceTable:Fort Image Reference Table Structure
     ColorReferenceTable:Fort Color Reference Table Structure
     TextReferenceTable:Fort Text Reference Table Structure
    ) {is=title K2Node_BreakStruct}.UIBackgroundMaterial
    bMatchTextureSize:bool=false
   ) -> (then)
  ]}
 SetupArrowVisibility {nodes [
   SetupArrowVisibility() {is=title K2Node_FunctionEntry}
   > n1 = Branch(Condition=ShouldDisplayNavigationArrows!() -> (Output:bool))
   n2 = GetScrollOffset(self=Get BenefitsScrollBox())
   > SetVisibility(self=Get ScrollRight(), InVisibility=Select(Option 0=Collapsed, Option 1=SelfHitTestInvisible, Index="Equal (Float)"(A=@n2)))
   n3 = GetScrollOffset(self=Get BenefitsScrollBox())
   > SetVisibility(self=Get ScrollLeft(), InVisibility=Select(Option 0=Collapsed, Option 1=SelfHitTestInvisible, Index="float > float"(A=@n3)))
   SetVisibility(execute=@n1.else, self=[Get ScrollRight(), Get ScrollLeft()], InVisibility=Collapsed)
  ]}
 SetupBenefitsList {nodes [
   SetupBenefitsList() {is=title K2Node_FunctionEntry}
   > Reset!(execute, self:Dynamic Entry Box Object Reference=Get DynamicEntryBox(), bDeleteWidgets:bool=true) -> (then)
   n1 = Get AltBenefitsGroups!(self:Crew VM Object Reference=Get CrewVM()) -> (AltBenefitsGroups:Array of Crew Benefits Group VM Object References)
   > n2 = For Each Loop(Array:Array of Crew Benefits Group VM Object References=@n1) -> (Array Element:Crew Benefits Group VM Object Reference)
   n3 = Get Benefits!(self:Crew Benefits Group VM Object Reference=@n2) -> (Benefits:Array of Crew Benefit VM Object References)
   > n4 = For Each Loop(Array:Array of Crew Benefit VM Object References=@n3) -> (Array Element:Crew Benefit VM Object Reference)
   > n5 = Create Entry!(execute, self:Dynamic Entry Box Object Reference=Get DynamicEntryBox()) -> (then, ReturnValue:User Widget Object Reference)
   > n6 = Cast To CrewBenefitTile(Object=@n5)
   > Set CrewBenefitVM!(execute, CrewBenefitVM:Crew Benefit VM Object Reference=@n4, self:Crew Benefit Tile Object Reference=@n6) -> (then, Output_Get:Crew Benefit VM Object Reference)
   > SetDefaultVisual(self:Crew Benefit Tile Object Reference=@n6)
   > Bind Event to On Clicked(self:Common Button Base Object Reference=@n6, Delegate=Create Event()) {is=title K2Node_AddDelegate}
   > SetHeader!(
    execute
    self:Crew Benefit Tile Object Reference=@n6
    HeaderText:Text=SelectText(A=Get Label!(self:Crew Benefits Group VM Object Reference=@n2) -> (Label:Text), bPickA="Equal (Integer)"(A=@n4.Array Index))
   ) -> (then)
   n7 = Last Index!(TargetArray:Array of Crew Benefit VM Object References=@n3) -> (ReturnValue=0)
   n8 = Last Index(TargetArray:Array of Crew Benefits Group VM Object References=@n1) -> (ReturnValue=0)
   > ShowSectionDivider!(
    execute
    self:Crew Benefit Tile Object Reference=@n6
    Show:bool=AND Boolean(A="Equal (Integer)"(A=@n4.Array Index, B=@n7), B="Not Equal (Integer)"(A=@n8, B=@n2.Array Index))
   ) -> (then)
  ]}
 ShouldDisplayNavigationArrows {nodes [
   ShouldDisplayNavigationArrows() {is=title K2Node_FunctionEntry}
   n1 = GetViewportSize()
   n2 = GetViewportScale!() -> (ReturnValue:float)
   n3 = GetDesiredSize!(self:Widget Object Reference=Get DynamicEntryBox()) -> (ReturnValue_X:double, ReturnValue_Y:double)
   n4 = Get NavigationScrollPadding!(self:Scroll Box Object Reference=Get BenefitsScrollBox()) -> (NavigationScrollPadding:float)
   > Return Node!(execute, Output:bool="float > float"(A=float - float(A=@n3, B=@n4), B=float / float(A=@n1, B=@n2)))
  ]}
 UpToSub {nodes [
   UpToSub!() -> (then, Navigation:EUINavigation Enum) {is=title K2Node_FunctionEntry}
   > Return Node!(
    execute
    ReturnValue:Widget Object Reference=GetDesiredFocusTarget!(self:Common Activatable Widget Object Reference) -> (ReturnValue:Widget Object Reference)
   )
  ]}
 UpdateBackgroundVisibility {nodes [
   UpdateBackgroundVisibility!() -> (then, IsSubbed:bool=~, IsBenefitHovered:bool=~, IsSeasonPassBenefit:bool=~) {is=title K2Node_FunctionEntry}
   n1 = Get IsSubbed()
   > SetVisibility(
    self=Get CMSImage_BG_Crew_Unsubbed() -> (CMSImage_BG_Crew_Unsubbed:Epic CMSImage Object Reference)
    InVisibility=Select(Option 0=Collapsed, Option 1=SelfHitTestInvisible, Index=NOT Boolean(A=@n1))
   )
   > SetVisibility(
    self=Get CMSImage_BG_Crew_Subbed() -> (CMSImage_BG_Crew_Subbed:Epic CMSImage Object Reference)
    InVisibility=Select(Option 0=Collapsed, Option 1=SelfHitTestInvisible, Index=Get IsSubbed())
   )
   n2 = NOT Boolean(A=Get IsSeasonPassBenefit() -> (IsSeasonPassBenefit:bool))
   n3 = Get IsBenefitHovered() -> (IsBenefitHovered:bool)
   > SetVisibility(self=Get Image_BG_Benefit_Default(), InVisibility=Select(Option 0=Collapsed, Option 1=SelfHitTestInvisible, Index=AND Boolean(A=@n3, B=@n2)))
   n4 = Get IsSeasonPassBenefit()
   n5 = Get IsBenefitHovered()
   > SetVisibility(self=Get Image_BG_Benefit_Season(), InVisibility=Select(Option 0=Collapsed, Option 1=SelfHitTestInvisible, Index=AND Boolean(A=@n5, B=@n4)))
   > SetVisibility(
    self=Get CMSImage_BG_Benefit_Loaded() -> (CMSImage_BG_Benefit_Loaded:Epic CMSImage Object Reference)
    InVisibility=Select(Option 0=Collapsed, Option 1=SelfHitTestInvisible, Index=Get IsBenefitHovered())
   )
  ]}
]
variables [
 BentoCMSTexture {type=Texture2D}
 BentoUnsubMaterial {type=MaterialInstanceDynamic}
 ButtonScrollAmount {type=number, default=1720}
 CancelDelay {type=boolean}
 CloseScreen {}
 CrewPurchased {}
 DelayActive {type=boolean}
 IsSubbed {type=boolean}
 LastHoveredBenefit {type=CrewBenefitVM, default=null}
]
widgets {
 info {
  inheritedWidgetCount=0
  namedSlotCount=1
  parentClass=/Script/FortniteUI.FortSubscriptionContentBase
  rootWidgetClass=&Overlay
  widgetCount=32
 }
 widgets [
  RootOverlay {widgetClassPath=&Overlay, bIsVariable=true, widget=RootOverlay}
  CMSImage_BG_Crew_Unsubbed {
   widgetClassPath=/Script/EpicCMSUIFramework.EpicCMSImage
   bIsVariable=true
   parent=RootOverlay
   widget=CMSImage_BG_Crew_Unsubbed
   properties {
    bMatchImageSize=false
    materialTextureParamName=CMS_Crew_Bento
    visibility=Collapsed
    brush {
     bIsDynamicallyLoaded=false
     drawAs=Image
     imageType=FullColor
     mirroring=NoMirror
     resourceName=None
     resourceObject=/Crew/UI/Materials/M_UI_CrewUpsell_Landing_UnSub
     tiling=NoTile
     imageSize=<200 200>
     margin=<ltrb 0 0 0 0>
     outlineSettings {
      bUseBrushTransparency=false
      roundingType=HalfHeightRadius
      width=0
      color {colorUseRule=UseColor_Specified, specifiedColor=<rgba 0 0 0 0>}
      cornerRadii=<0 0 0 0>
     }
     tintColor {colorUseRule=UseColor_Specified, specifiedColor=<rgba 1 1 1 1>}
     uVRegion {bIsValid=false, max=<0 0>, min=<0 0>}
    }
   }
   slot {class=OverlaySlot, properties {horizontalAlignment=HAlign_Fill, verticalAlignment=VAlign_Fill}}
  }
  CMSImage_BG_Crew_Subbed {
   widgetClassPath=&EpicCMSImage
   bIsVariable=true
   parent=RootOverlay
   widget=CMSImage_BG_Crew_Subbed
   properties {
    bMatchImageSize=false
    materialTextureParamName=Image
    renderOpacity=0
    visibility=Collapsed
    brush {
     bIsDynamicallyLoaded=false
     drawAs=Image
     imageType=FullColor
     mirroring=NoMirror
     resourceName=None
     resourceObject=/Crew/UI/Materials/M_UI_CrewUpsell_Landing_Sub
     tiling=NoTile
     imageSize=<200 200>
     margin=<ltrb 0 0 0 0>
     outlineSettings {
      bUseBrushTransparency=false
      roundingType=HalfHeightRadius
      width=0
      color {colorUseRule=UseColor_Specified, specifiedColor=<rgba 0 0 0 0>}
      cornerRadii=<0 0 0 0>
     }
     tintColor {colorUseRule=UseColor_Specified, specifiedColor=<rgba 1 1 1 1>}
     uVRegion {bIsValid=false, max=<0 0>, min=<0 0>}
    }
   }
   slot {class=OverlaySlot, properties {horizontalAlignment=HAlign_Fill, verticalAlignment=VAlign_Fill}}
  }
  Image_BG_Benefit_Default {
   widgetClassPath=&FortLazyImage
   bIsVariable=true
   parent=RootOverlay
   widget=Image_BG_Benefit_Default
   properties {
    materialTextureParamName=CMS_Texture
    visibility=Collapsed
    brush {
     bIsDynamicallyLoaded=false
     drawAs=Image
     imageType=FullColor
     mirroring=NoMirror
     resourceName=None
     resourceObject=/Crew/UI/Materials/CrewBenefits/MI_UI_CrewBenefits_Gradient
     tiling=NoTile
     imageSize=<32 32>
     margin=<ltrb 0 0 0 0>
     outlineSettings {
      bUseBrushTransparency=false
      roundingType=HalfHeightRadius
      width=0
      color {colorUseRule=UseColor_Specified, specifiedColor=<rgba 0 0 0 0>}
      cornerRadii=<0 0 0 0>
     }
     tintColor {colorUseRule=UseColor_Specified, specifiedColor=<rgba 1 1 1 1>}
     uVRegion {bIsValid=false, max=<0 0>, min=<0 0>}
    }
    loadingBackgroundBrush {
     bIsDynamicallyLoaded=false
     drawAs=NoDrawType
     imageType=NoImage
     mirroring=NoMirror
     resourceName=None
     resourceObject=null
     tiling=NoTile
     imageSize=<32 32>
     margin=<ltrb 0 0 0 0>
     outlineSettings {
      bUseBrushTransparency=false
      roundingType=HalfHeightRadius
      width=0
      color {colorUseRule=UseColor_Specified, specifiedColor=<rgba 0 0 0 0>}
      cornerRadii=<0 0 0 0>
     }
     tintColor {colorUseRule=UseColor_Specified, specifiedColor=<rgba 0 0 0 0>}
     uVRegion {bIsValid=false, max=<0 0>, min=<0 0>}
    }
   }
   slot {class=OverlaySlot, properties {horizontalAlignment=HAlign_Fill, verticalAlignment=VAlign_Fill}}
  }
  Image_BG_Benefit_Season {
   widgetClassPath=&FortLazyImage
   bIsVariable=true
   parent=RootOverlay
   widget=Image_BG_Benefit_Season
   properties {
    visibility=Collapsed
    brush {
     bIsDynamicallyLoaded=false
     drawAs=Image
     imageType=FullColor
     mirroring=NoMirror
     resourceName=None
     resourceObject=/Game/Global/Textures/Default/Blanks/T_Blank_White_ZeroAlpha_sRGBColor
     tiling=NoTile
     imageSize=<1 1>
     margin=<ltrb 0 0 0 0>
     outlineSettings {
      bUseBrushTransparency=false
      roundingType=HalfHeightRadius
      width=0
      color {colorUseRule=UseColor_Specified, specifiedColor=<rgba 0 0 0 0>}
      cornerRadii=<0 0 0 0>
     }
     tintColor {colorUseRule=UseColor_Specified, specifiedColor=<rgba 1 1 1 1>}
     uVRegion {bIsValid=false, max=<0 0>, min=<0 0>}
    }
    loadingBackgroundBrush {
     bIsDynamicallyLoaded=false
     drawAs=NoDrawType
     imageType=NoImage
     mirroring=NoMirror
     resourceName=None
     resourceObject=null
     tiling=NoTile
     imageSize=<32 32>
     margin=<ltrb 0 0 0 0>
     outlineSettings {
      bUseBrushTransparency=false
      roundingType=HalfHeightRadius
      width=0
      color {colorUseRule=UseColor_Specified, specifiedColor=<rgba 0 0 0 0>}
      cornerRadii=<0 0 0 0>
     }
     tintColor {colorUseRule=UseColor_Specified, specifiedColor=<rgba 0 0 0 0>}
     uVRegion {bIsValid=false, max=<0 0>, min=<0 0>}
    }
   }
   slot {class=OverlaySlot, properties {horizontalAlignment=HAlign_Fill, verticalAlignment=VAlign_Fill}}
  }
  CMSImage_BG_Benefit_Loaded {
   widgetClassPath=&EpicCMSImage
   bIsVariable=true
   parent=RootOverlay
   widget=CMSImage_BG_Benefit_Loaded
   properties {
    bMatchImageSize=false
    materialTextureParamName=Image
    visibility=Collapsed
    brush {
     bIsDynamicallyLoaded=false
     drawAs=Image
     imageType=FullColor
     mirroring=NoMirror
     resourceName=None
     resourceObject=/SeasonPass/Materials/Background/MI_UI_SeasonPass_PurchaseScreen_Background
     tiling=NoTile
     imageSize=<1 1>
     margin=<ltrb 0 0 0 0>
     outlineSettings {
      bUseBrushTransparency=false
      roundingType=HalfHeightRadius
      width=0
      color {colorUseRule=UseColor_Specified, specifiedColor=<rgba 0 0 0 0>}
      cornerRadii=<0 0 0 0>
     }
     tintColor {colorUseRule=UseColor_Specified, specifiedColor=<rgba 1 1 1 1>}
     uVRegion {bIsValid=false, max=<0 0>, min=<0 0>}
    }
    loadingBackgroundBrush {
     bIsDynamicallyLoaded=false
     drawAs=Image
     imageType=FullColor
     mirroring=NoMirror
     resourceName=None
     resourceObject=null
     tiling=NoTile
     imageSize=<32 32>
     margin=<ltrb 0 0 0 0>
     outlineSettings {
      bUseBrushTransparency=false
      roundingType=HalfHeightRadius
      width=0
      color {colorUseRule=UseColor_Specified, specifiedColor=<rgba 0 0 0 0>}
      cornerRadii=<0 0 0 0>
     }
     tintColor {colorUseRule=UseColor_Specified, specifiedColor=<rgba 0 0 0 0>}
     uVRegion {bIsValid=false, max=<0 0>, min=<0 0>}
    }
    loadingFailFallback {
     bIsDynamicallyLoaded=false
     drawAs=Image
     imageType=NoImage
     mirroring=NoMirror
     resourceName=None
     resourceObject=null
     tiling=NoTile
     imageSize=<32 32>
     margin=<ltrb 0 0 0 0>
     outlineSettings {
      bUseBrushTransparency=false
      roundingType=HalfHeightRadius
      width=0
      color {colorUseRule=UseColor_Specified, specifiedColor=<rgba 0 0 0 0>}
      cornerRadii=<0 0 0 0>
     }
     tintColor {colorUseRule=UseColor_Specified, specifiedColor=<rgba 0 0 0 0>}
     uVRegion {bIsValid=false, max=<0 0>, min=<0 0>}
    }
   }
   slot {class=OverlaySlot, properties {horizontalAlignment=HAlign_Fill, verticalAlignment=VAlign_Fill}}
  }
  PostBufferUpdate {
   widgetClassPath=/Script/UMG.PostBufferUpdate
   bIsVariable=true
   parent=RootOverlay
   widget=PostBufferUpdate
   properties {
    visibility=HitTestInvisible
    buffersToUpdate [ESlatePostRT_0]
    updateBufferInfos [{bufferToUpdate=ESlatePostRT_0, postParamUpdater=&self:WidgetTree.PostBufferUpdate.PostBufferBlurUpdater_0}]
   }
   slot {class=OverlaySlot, properties {horizontalAlignment=HAlign_Fill, verticalAlignment=VAlign_Fill}}
  }
  SafeZone {
   widgetClassPath=&SafeZone
   bIsVariable=true
   parent=RootOverlay
   widget=SafeZone
   slot {class=OverlaySlot, properties {horizontalAlignment=HAlign_Fill, verticalAlignment=VAlign_Fill}}
  }
  SizeBox {
   widgetClassPath=&SizeBox
   bIsVariable=true
   parent=SafeZone
   widget=SizeBox
   properties {bOverride_WidthOverride=true, widthOverride=1920}
   slot {class=SafeZoneSlot}
  }
  Overlay_19 {widgetClassPath=&Overlay, parent=SizeBox, widget=Overlay_19, slot {class=SizeBoxSlot}}
  CrewLogo {
   widgetClassPath=&Image
   bIsVariable=true
   parent=Overlay_19
   widget=CrewLogo
   properties {
    brush {
     bIsDynamicallyLoaded=false
     drawAs=Image
     imageType=FullColor
     mirroring=NoMirror
     resourceName=None
     resourceObject=/Game/Athena/UI/Logos/Textures/T_Crew_Workmark_S
     tiling=NoTile
     imageSize=<210 58>
     margin=<ltrb 0 0 0 0>
     outlineSettings {
      bUseBrushTransparency=false
      roundingType=HalfHeightRadius
      width=0
      color {colorUseRule=UseColor_Specified, specifiedColor=<rgba 0 0 0 0>}
      cornerRadii=<0 0 0 0>
     }
     tintColor {colorUseRule=UseColor_Specified, specifiedColor=<rgba 1 1 1 1>}
     uVRegion {bIsValid=false, max=<0 0>, min=<0 0>}
    }
    renderTransform {angle=0, scale=<0.9 1>, shear=<0 0>, translation=<0 0>}
   }
   slot {class=OverlaySlot, properties {padding=<ltrb 80 48 0 0>}}
  }
  WBP_PlayerInfo {
   widgetClassPath=/Game/UI/Frontend/ItemShop/MultiProduct/WBP_PlayerInfo
   bIsVariable=true
   parent=Overlay_19
   widget=WBP_PlayerInfo
   slot {class=OverlaySlot, properties {horizontalAlignment=HAlign_Right, padding=<ltrb 0 40 80 0>}}
  }
  CrewBenefitInfo {
   widgetClassPath=&/CrewBenefitInfo
   bIsVariable=true
   parent=Overlay_19
   widget=CrewBenefitInfo
   slot {class=OverlaySlot, properties {verticalAlignment=VAlign_Fill, padding=<ltrb 159 125 0 440>}}
  }
  SB_CrewEntitlement {
   widgetClassPath=&SizeBox
   bIsVariable=true
   parent=Overlay_19
   widget=SB_CrewEntitlement
   properties {bOverride_WidthOverride=true, widthOverride=600}
   slot {class=OverlaySlot, properties {horizontalAlignment=HAlign_Right, verticalAlignment=VAlign_Bottom, padding=<ltrb 0 0 127 416>}}
  }
  VerticalBox_0 {
   widgetClassPath=&VerticalBox
   parent=SB_CrewEntitlement
   widget=VerticalBox_0
   properties {bOverride_Cursor=true}
   slot {class=SizeBoxSlot}
  }
  CrewEntitlementMessage {
   widgetClassPath=&/CrewEntitlementMessage
   bIsVariable=true
   parent=VerticalBox_0
   widget=CrewEntitlementMessage
   slot {class=VerticalBoxSlot}
  }
  CrewNoticeMessage {
   widgetClassPath=/FortUILibrary/UIKit/InfoBox/WBP_UIKit_InfoBox
   bIsVariable=true
   parent=VerticalBox_0
   widget=CrewNoticeMessage
   properties {previewText="", showIcon=false, showStatusIndicator=true}
   slot {class=VerticalBoxSlot, properties {padding=<ltrb 0 12 0 0>}}
  }
  ParentalControlsMessage {
   widgetClassPath=&WBP_UIKit_InfoBox
   bIsVariable=true
   parent=VerticalBox_0
   widget=ParentalControlsMessage
   properties {previewText="", showIcon=false, showStatusIndicator=true, visibility=Collapsed}
   slot {class=VerticalBoxSlot, properties {padding=<ltrb 0 12 0 0>}}
  }
  VerticalBox_1 {
   widgetClassPath=&VerticalBox
   parent=Overlay_19
   widget=VerticalBox_1
   slot {class=OverlaySlot, properties {horizontalAlignment=HAlign_Fill, verticalAlignment=VAlign_Bottom}}
  }
  Overlay_1 {
   widgetClassPath=&Overlay
   parent=VerticalBox_1
   widget=Overlay_1
   slot {class=VerticalBoxSlot, properties {verticalAlignment=VAlign_Top, padding=<ltrb 0 0 0 24>}}
  }
  BenefitsScrollBox {
   widgetClassPath=/Script/FortniteUI.AthenaScrollBox
   bIsVariable=true
   parent=Overlay_1
   widget=BenefitsScrollBox
   properties {
    bAllowRightClickDragScrolling=false
    clipping=Inherit
    consumeMouseWheel=Never
    navigationScrollPadding=100
    orientation=Orient_Horizontal
    scrollBarVisibility=Collapsed
    scrollWhenFocusChanges=AnimatedScroll
    widgetStyle {
     barThickness=9
     bottomShadowBrush {
      bIsDynamicallyLoaded=false
      drawAs=Box
      imageType=FullColor
      mirroring=NoMirror
      resourceName=../../../Engine/Content/Slate/Common/ScrollBoxShadowBottom.png
      resourceObject=null
      tiling=NoTile
      imageSize=<16 8>
      margin=<ltrb 0.5 0 0.5 1>
      outlineSettings {
       bUseBrushTransparency=false
       roundingType=HalfHeightRadius
       width=0
       color {colorUseRule=UseColor_Specified, specifiedColor=<rgba 0 0 0 0>}
       cornerRadii=<0 0 0 0>
      }
      tintColor {colorUseRule=UseColor_Specified, specifiedColor=<rgba 1 1 1 1>}
      uVRegion {bIsValid=false, max=<0 0>, min=<0 0>}
     }
     horizontalScrolledContentPadding=<ltrb 0 0 1 0>
     leftShadowBrush {
      bIsDynamicallyLoaded=false
      drawAs=Box
      imageType=FullColor
      mirroring=NoMirror
      resourceName=../../../Engine/Content/Slate/Common/ScrollBoxShadowLeft.png
      resourceObject=null
      tiling=NoTile
      imageSize=<0 0>
      margin=<ltrb 1 0.5 0 0.5>
      outlineSettings {
       bUseBrushTransparency=false
       roundingType=HalfHeightRadius
       width=0
       color {colorUseRule=UseColor_Specified, specifiedColor=<rgba 0 0 0 0>}
       cornerRadii=<0 0 0 0>
      }
      tintColor {colorUseRule=UseColor_Specified, specifiedColor=<rgba 1 1 1 1>}
      uVRegion {bIsValid=false, max=<0 0>, min=<0 0>}
     }
     rightShadowBrush {
      bIsDynamicallyLoaded=false
      drawAs=Box
      imageType=FullColor
      mirroring=NoMirror
      resourceName=../../../Engine/Content/Slate/Common/ScrollBoxShadowRight.png
      resourceObject=null
      tiling=NoTile
      imageSize=<0 0>
      margin=<ltrb 0 0.5 1 0.5>
      outlineSettings {
       bUseBrushTransparency=false
       roundingType=HalfHeightRadius
       width=0
       color {colorUseRule=UseColor_Specified, specifiedColor=<rgba 0 0 0 0>}
       cornerRadii=<0 0 0 0>
      }
      tintColor {colorUseRule=UseColor_Specified, specifiedColor=<rgba 1 1 1 1>}
      uVRegion {bIsValid=false, max=<0 0>, min=<0 0>}
     }
     topShadowBrush {
      bIsDynamicallyLoaded=false
      drawAs=Box
      imageType=FullColor
      mirroring=NoMirror
      resourceName=../../../Engine/Content/Slate/Common/ScrollBoxShadowTop.png
      resourceObject=null
      tiling=NoTile
      imageSize=<16 8>
      margin=<ltrb 0.5 1 0.5 0>
      outlineSettings {
       bUseBrushTransparency=false
       roundingType=HalfHeightRadius
       width=0
       color {colorUseRule=UseColor_Specified, specifiedColor=<rgba 0 0 0 0>}
       cornerRadii=<0 0 0 0>
      }
      tintColor {colorUseRule=UseColor_Specified, specifiedColor=<rgba 1 1 1 1>}
      uVRegion {bIsValid=false, max=<0 0>, min=<0 0>}
     }
     verticalScrolledContentPadding=<ltrb 0 0 0 1>
    }
   }
   slot {class=OverlaySlot}
  }
  DynamicEntryBox {
   widgetClassPath=/Script/FortniteUI.FortDynamicEntryBox
   bIsVariable=true
   parent=BenefitsScrollBox
   widget=DynamicEntryBox
   properties {
    entryWidgetClass=&/CrewBenefitTile.CrewBenefitTile_C
    navigation=&self:WidgetTree.DynamicEntryBox.WidgetNavigation_0
    numDesignerPreviewEntries=7
    entrySpacing=<24 0>
   }
   slot {class=ScrollBoxSlot, properties {horizontalAlignment=HAlign_Left, verticalAlignment=VAlign_Top, padding=<ltrb 100 0 100 0>}}
  }
  MKBScrollArrows {
   widgetClassPath=&HorizontalBox
   bIsVariable=true
   parent=Overlay_1
   widget=MKBScrollArrows
   slot {class=OverlaySlot, properties {horizontalAlignment=HAlign_Fill, verticalAlignment=VAlign_Center, padding=<ltrb 100 0 100 0>}}
  }
  ScrollLeft {
   widgetClassPath=&WBP_UIKit_Button_Regular
   bIsVariable=true
   parent=MKBScrollArrows
   widget=ScrollLeft
   properties {iconMaterial=/Crew/UI/Materials/MI_UI_ScrollArrowLeft, text="", visibility=Collapsed}
   slot {class=HorizontalBoxSlot}
  }
  ScrollRight {
   widgetClassPath=&WBP_UIKit_Button_Regular
   bIsVariable=true
   parent=MKBScrollArrows
   widget=ScrollRight
   properties {iconMaterial=/Crew/UI/Materials/MI_UI_ScrollArrowRight, text="", visibility=Collapsed}
   slot {class=HorizontalBoxSlot, properties {horizontalAlignment=HAlign_Right, size {sizeRule=Fill, value=1}}}
  }
  HBox_LegalActionBar {
   widgetClassPath=&HorizontalBox
   bIsVariable=true
   parent=VerticalBox_1
   widget=HBox_LegalActionBar
   slot {class=VerticalBoxSlot, properties {verticalAlignment=VAlign_Bottom, padding=<ltrb 80 0 80 32>}}
  }
  Text_Legal {
   widgetClassPath=&CommonTextBlock
   bIsVariable=true
   parent=HBox_LegalActionBar
   widget=Text_Legal
   properties {
    autoWrapText=true
    mobileFontSizeMultiplier=1.25
    scrollOrientation=Orient_Vertical
    text="[PH] Auto-renews each month until cancelled. Cancel anytime. Territorial restrictions apply. Ability to claim Premium Rewards during active subscription only. Select \"Subscription Terms\" for more info."
    colorAndOpacity {colorUseRule=UseColor_Specified, specifiedColor=<rgba 0.5776 0.5776 0.5776 1>}
    font {
     bForceMonospaced=false
     bMaterialIsStencil=false
     fontMaterial=null
     fontObject=&NotoSans
     letterSpacing=0
     monospacedWidth=1
     size=24
     skewAmount=0
     typefaceFontName=NotoSans-Regular
     outlineSettings {
      bApplyOutlineToDropShadows=false
      bMiteredCorners=false
      bSeparateFillAlpha=false
      outlineMaterial=null
      outlineSize=0
      outlineColor=<rgba 0 0 0 1>
     }
    }
    shadowOffset=<0 0>
   }
   slot {class=HorizontalBoxSlot, properties {verticalAlignment=VAlign_Bottom, padding=<ltrb 0 0 40 0>, size {sizeRule=Fill, value=1}}}
  }
  WBP_ActionBarContainer {
   widgetClassPath=/Game/UI/Frontend/BottomBar/WBP_ActionBarContainer
   bIsVariable=true
   parent=HBox_LegalActionBar
   widget=WBP_ActionBarContainer
   slot {class=HorizontalBoxSlot, properties {horizontalAlignment=HAlign_Right, verticalAlignment=VAlign_Bottom}}
  }
  HorizontalBox {widgetClassPath=&HorizontalBox, namedSlotHost=WBP_ActionBarContainer, widget=HorizontalBox}
  WBP_AthenaBottomBar_WithNotifications {
   widgetClassPath=/Game/UI/Frontend/BottomBar/WBP_AthenaBottomBar_WithNotifications
   bIsVariable=true
   parent=HorizontalBox
   widget=WBP_AthenaBottomBar_WithNotifications
   slot {class=HorizontalBoxSlot, properties {padding=<ltrb 0 0 16 0>}}
  }
  TermsButton {
   widgetClassPath=/FortUILibrary/UIKit/Buttons/WBP_UIKit_Button_Quiet
   bIsVariable=true
   parent=HorizontalBox
   widget=TermsButton
   properties {text=Subscription Terms, triggeringInputAction {dataTable=&InputActionDataTable, rowName=GenericFaceButtonTop}}
   slot {class=HorizontalBoxSlot, properties {padding=<ltrb 0 0 16 0>}}
  }
  BackButton {
   widgetClassPath=&WBP_UIKit_Button_Quiet
   bIsVariable=true
   parent=HorizontalBox
   widget=BackButton
   properties {text=Back, triggeringInputAction {dataTable=&InputActionDataTable, rowName=GenericBack}}
   slot {class=HorizontalBoxSlot}
  }
 ]
}
