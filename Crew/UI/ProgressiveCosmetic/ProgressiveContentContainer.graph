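# ProgressiveContentContainer is the Crew UI widget that hosts progressive cosmetic content, built on the native FortProgressiveContentContainer class.
# Its root activatable widget switcher holds an inherited table of contents screen and an item screen, with a zero transition duration.
# The table of contents screen's Back Clicked and Open Upsell events are re-broadcast through the CloseScreen and UpSellScreen dispatchers.
# The Blueprint has no logic for switching between the two screens.
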WidgetBlueprint /Crew/UI/ProgressiveCosmetic/ProgressiveContentContainer
graphs [
 CloseScreen {nodes [CloseScreen() {is=title K2Node_FunctionEntry}]}
 EventGraph {nodes [
   "Back Clicked (Widget_ProgressiveTableOfContentsScreen)"() {is=title K2Node_ComponentBoundEvent}
   > Call Close Screen!(execute:Exec, self:Progressive Content Container Object Reference) -> (then:Exec) {is=title K2Node_CallDelegate}
   "Open Upsell (Widget_ProgressiveTableOfContentsScreen)"() {is=title K2Node_ComponentBoundEvent}
   > Call Up Sell Screen!(execute:Exec, self:Progressive Content Container Object Reference) -> (then:Exec) {is=title K2Node_CallDelegate}
   Event PreConstruct()
   Event Construct()
   Event Tick()
  ]}
 UpSellScreen {nodes [UpSellScreen() {is=title K2Node_FunctionEntry}]}
]
variables [
 CloseScreen {}
 UpSellScreen {}
]
widgets {
 info {
  inheritedWidgetCount=3
  parentClass=/Script/CrewUI.FortProgressiveContentContainer
  rootWidgetClass=&CommonActivatableWidgetSwitcher
  widgetCount=3
 }
 widgets [
  Switcher_PrimaryContent {
   widgetClassPath=&CommonActivatableWidgetSwitcher
   bInherited=true
   bIsVariable=true
   widget=Switcher_PrimaryContent
   properties {transitionCurveType=Linear, transitionDuration=0}
  }
  Widget_ProgressiveTableOfContentsScreen {
   widgetClassPath=&/ProgressiveTableOfContentsScreen
   bInherited=true
   bIsVariable=true
   parent=Switcher_PrimaryContent
   widget=Widget_ProgressiveTableOfContentsScreen
   slot {class=WidgetSwitcherSlot}
  }
  Widget_ProgressiveItemScreen {
   widgetClassPath=&/ProgressiveItemScreen
   bInherited=true
   bIsVariable=true
   parent=Switcher_PrimaryContent
   widget=Widget_ProgressiveItemScreen
   slot {class=WidgetSwitcherSlot}
  }
 ]
}
