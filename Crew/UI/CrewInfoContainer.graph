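# Activatable menu screen for Crew info, switching between the Crew benefits screen and the progressive content container through a top UIKit tab group.
# Tab selection sets the switcher index, and activation resets it to benefits.
# Either child's close event deactivates the screen, and the benefits screen's purchase event is re-broadcast as CrewPurchased.
# Overrides allow frontend flow interrupts and hide the bars and chat, while mobile games get compact tab sizing and padding.
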WidgetBlueprint /Crew/UI/CrewInfoContainer
graphs [
 BP_GetDesiredInputConfig {nodes [
   Get Desired Input Config() {is=title K2Node_FunctionEntry}
   > Return Node!(
    execute
    ReturnValue_bIgnoreMoveInput:bool=False
    ReturnValue_bIgnoreLookInput:bool=False
    ReturnValue_InputMode:ECommonInputMode Enum=Menu
    ReturnValue_MouseCaptureMode:EMouseCaptureMode Enum=NoCapture
    ReturnValue_MouseLockMode:EMouseLockMode Enum=DoNotLock
    ReturnValue_bHideCursorDuringViewportCapture:bool=True
   )
  ]}
 CrewPurchased {nodes [CrewPurchased() {is=title K2Node_FunctionEntry}]}
 EventGraph {nodes [
   n1 = "Close Screen (CrewBenefitsScreen)"() {is=title K2Node_ComponentBoundEvent}
   DeactivateWidget(execute=[@n1.then, @n3.then])
   n2 = "On UIKit Tab Selected (WBP_UIKit_TabGroup)"!() -> (OutputDelegate, then, TabID:Name=None, TabVM:MVVM Base Viewmodel Object Reference) {is=title K2Node_ComponentBoundEvent}
   > SetActiveWidgetIndex(
    self=Get CommonActivatableWidgetSwitcher_43() -> (CommonActivatableWidgetSwitcher_43:Common Activatable Widget Switcher Object Reference)
    Index=SelectInt(A=0, B=1, bPickA="Equal (Name)"(A=@n2.TabID, B=0))
   )
   n3 = "Close Screen (ProgressiveContentContainer)"() {is=title K2Node_ComponentBoundEvent}
   "Crew Purchased (CrewBenefitsScreen)"() {is=title K2Node_ComponentBoundEvent}
   > Call Crew Purchased!(execute:Exec, self:Crew Info Container Object Reference) -> (then:Exec) {is=title K2Node_CallDelegate}
   Event OnInitialized()
   > SetCompactUIStyle() {is=title K2Node_CallFunction}
   Event On Activated()
   > SetActiveWidgetIndex(self=Get CommonActivatableWidgetSwitcher_43(), Index=0)
  ]}
 GetFrontendFlowInterruptConfig {nodes [
   GetFrontendFlowInterruptConfig() {is=title K2Node_FunctionEntry}
   > Return Node!(
    execute
    ReturnValue:Frontend Flow Interrup Config Structure=Make Frontend Flow Interrup Config!(bCanFrontendFlowInterrupt:bool=True, bCanInterruptByEmergentOnly:bool=true) -> (FrontendFlowInterrupConfig:Frontend Flow Interrup Config Structure) {is=title K2Node_MakeStruct}
   )
  ]}
 GetPrimaryContentSetup {nodes [
   GetPrimaryContentSetup() {is=title K2Node_FunctionEntry}
   > Return Node(ReturnValue:Primary Content Setup Structure=Make Primary Content Setup!(
     bShowBottomBar:bool=false
     bShowChatWidget:bool=false
     bEnableNewTextChat:bool=False
     bShowTopBar:bool=false
     bTabSupportsInterface:bool=False
    ) -> (PrimaryContentSetup:Primary Content Setup Structure) {is=title K2Node_MakeStruct})
  ]}
 SetCompactUIStyle {nodes [
   SetCompactUIStyle() {is=title K2Node_FunctionEntry}
   > Branch(Condition=IsMobileGame())
   > Set!(
    execute
    MinButtonSize_X:int=0
    MinButtonSize_Y:int=80
    self:WBP UIKit Tab Group Base Object Reference=Get WBP_UIKit_TabGroup() -> (WBP_UIKit_TabGroup:WBP UIKit Tab Group Object Reference)
   ) -> (then, Output_Get:Int Point Structure)
   > SetPadding!(
    execute
    self:Overlay Slot Object Reference=SlotAsOverlaySlot(Widget=Get SafeZone_TabGroup() -> (SafeZone_TabGroup:Safe Zone Object Reference))
    InPadding_Left:float=0
    InPadding_Top:float=24
    InPadding_Right:float=0
    InPadding_Bottom:float=0
   ) -> (then)
  ]}
]
variables [CrewPurchased {}]
widgets {
 info {inheritedWidgetCount=0, parentClass=&CommonActivatableWidget, rootWidgetClass=&Overlay, widgetCount=6}
 widgets [
  Overlay_0 {widgetClassPath=&Overlay, widget=Overlay_0}
  CommonActivatableWidgetSwitcher_43 {
   widgetClassPath=&CommonActivatableWidgetSwitcher
   bIsVariable=true
   parent=Overlay_0
   widget=CommonActivatableWidgetSwitcher_43
   slot {class=OverlaySlot, properties {horizontalAlignment=HAlign_Fill, verticalAlignment=VAlign_Fill}}
  }
  CrewBenefitsScreen {
   widgetClassPath=&/CrewBenefitsScreen
   bIsVariable=true
   parent=CommonActivatableWidgetSwitcher_43
   widget=CrewBenefitsScreen
   slot {class=WidgetSwitcherSlot}
  }
  ProgressiveContentContainer {
   widgetClassPath=/Crew/UI/ProgressiveCosmetic/ProgressiveContentContainer
   bIsVariable=true
   parent=CommonActivatableWidgetSwitcher_43
   widget=ProgressiveContentContainer
   slot {class=WidgetSwitcherSlot}
  }
  SafeZone_TabGroup {
   widgetClassPath=&SafeZone
   bIsVariable=true
   parent=Overlay_0
   widget=SafeZone_TabGroup
   properties {padBottom=false, padLeft=false, padRight=false}
   slot {class=OverlaySlot, properties {horizontalAlignment=HAlign_Center, padding=<ltrb 0 42 0 0>}}
  }
  WBP_UIKit_TabGroup {
   widgetClassPath=/FortUILibrary/UIKit/Tabs/WBP_UIKit_TabGroup
   bIsVariable=true
   parent=SafeZone_TabGroup
   widget=WBP_UIKit_TabGroup
   properties {
    backplateBrightness=Dark
    backplateCornerRadius=Tab
    initialSelectedTabId=0
    tabSpacing=0
    useInitialTabs=true
    useMinButtonSize=false
    buttonPadding=<ltrb 8 8 8 8>
    initialTabsData [
     {bEnableTabId=false, tabId=None, tabViewModel=&self:WidgetTree.WBP_UIKit_TabGroup.UIKitTabGroupButtonViewModel_0}
     {bEnableTabId=false, tabId=None, tabViewModel=&self:WidgetTree.WBP_UIKit_TabGroup.UIKitTabGroupButtonViewModel_1}
    ]
    tabGroupPadding=<ltrb 0 0 0 0>
   }
   slot {class=SafeZoneSlot}
  }
 ]
}
