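# Fortnite Crew landing page widget that shows the player's subscription status with actions to resolve issues, view cancellation info, rejoin or continue membership.
# HandleSubState maps the subscription state to a visibility switcher page, button visibility and info box text.
# Action buttons queue modals asynchronously and inject view models from CrewVM, while Back and See Benefits fire dispatchers.
# Focus follows the visible button, and BackgroundState swaps the lazy-loaded background material.
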WidgetBlueprint /Crew/UI/WBP_CrewLandingPage_Upsell
graphs [
 BP_GetDesiredFocusTarget {nodes [
   Get Desired Focus Target()
   n1 = Get CrewSubState() -> (CrewSubState:Common Visibility Switcher Object Reference)
   n2 = Get ResolveIssueButton() -> (ResolveIssueButton:WBP UIKit Button Loud Object Reference)
   n3 = IsVisible(self=@n2)
   n4 = Get CancelButton() -> (CancelButton:WBP UIKit Button Regular Object Reference)
   n5 = Select(Option 0:Widget Object Reference, Option 1:Widget Object Reference=@n4, Index=IsVisible(self=@n4)) -> (ReturnValue:Widget Object Reference)
   > Return Node!(
    execute
    ReturnValue:Widget Object Reference=Select(
     Option 0=Select(Option 0=@n5, Option 1=@n2, Index:bool=@n3)
     Option 1=Get RejoinCrew_Button() -> (RejoinCrew_Button:WBP UIKit Button Loud Object Reference)
     Index:int=GetActiveWidgetIndex!(self:Common Visibility Switcher Object Reference=@n1) -> (ReturnValue:int=0)
    )
   )
  ]}
 BackgroundState {nodes [
   n1 = BackgroundState!() -> (then, isBundle:bool=~) {is=title K2Node_FunctionEntry}
   > SetBrushFromLazyMaterial!(
    execute
    self:Common Lazy Image Object Reference=Get BG_Crew() -> (BG_Crew:Epic CMSImage Object Reference)
    LazyMaterial:"Material Interface Soft Object Reference (by ref)"=Select(
     Option 0:"Material Interface Soft Object Reference (by ref)"=Get CrewMaterial() -> (CrewMaterial:Material Instance Soft Object Reference)
     Option 1:"Material Interface Soft Object Reference (by ref)"=Get SubMaterial() -> (SubMaterial:Material Instance Soft Object Reference)
     Index=@n1
    ) -> (ReturnValue:"Material Interface Soft Object Reference (by ref)")
   ) -> (then)
  ]}
 ClosePage {nodes [ClosePage() {is=title K2Node_FunctionEntry}]}
 EventGraph {nodes [
   "On Clicked (ResolveIssueButton)"!() -> (OutputDelegate, then, Button:Common Button Base Object Reference) {is=title K2Node_ComponentBoundEvent}
   > n1 = QueueModalWidgetClassAsync!(
    execute:Exec
    WidgetClass:Common Activatable Widget Class Reference=/Crew/UI/Subscription/SubscriptionResolveIssueModal.SubscriptionResolveIssueModal_C
   ) -> (then:Exec, OnModalWidgetQueued:Exec, ActivatableWidget:Common Activatable Widget Object Reference) {is=title K2Node_AsyncAction}
   n2 = Cast To SubscriptionResolveIssueModal!(execute=@n1.OnModalWidgetQueued, Object=@n1) -> (then, CastFailed, AsSubscription Resolve Issue Modal:Subscription Resolve Issue Modal Object Reference) {is=title K2Node_DynamicCast}
   > Set CrewSubscriptionModalInfoVM!(
    execute
    CrewSubscriptionModalInfoVM:Crew Subscription Modal Info VM Object Reference=GetModalViewModelById!(
     self:Crew VM Object Reference=Get CrewVM() -> (CrewVM:Crew VM Object Reference)
     ModalId:String=Get ModalId!(self:Subscription Modal Base Object Reference=@n2) -> (ModalId:String)
    ) -> (ReturnValue:Crew Subscription Modal Info VM Object Reference)
    self:Subscription Resolve Issue Modal Object Reference=@n2
   ) -> (then, Output_Get:Crew Subscription Modal Info VM Object Reference)
   "On Clicked (Button_SeeBenefits)"!() -> (OutputDelegate, then, Button:Common Button Base Object Reference) {is=title K2Node_ComponentBoundEvent}
   > Call See Benefits!(execute:Exec, self:WBP Crew Landing Page Upsell Object Reference) -> (then:Exec) {is=title K2Node_CallDelegate}
   "On Clicked (Button_Back)"!() -> (OutputDelegate, then, Button:Common Button Base Object Reference) {is=title K2Node_ComponentBoundEvent}
   > Call Close Page!(execute:Exec, self:WBP Crew Landing Page Upsell Object Reference) -> (then:Exec) {is=title K2Node_CallDelegate}
   "On Clicked (CancelButton)"!() -> (OutputDelegate, then, Button:Common Button Base Object Reference) {is=title K2Node_ComponentBoundEvent}
   > n3 = QueueModalWidgetSoftClassAsync!(
    execute:Exec
    SoftWidgetClass:Common Activatable Widget Soft Class Reference=/Crew/UI/Subscription/SubscriptionCancellationInfoModal.SubscriptionCancellationInfoModal_C
   ) -> (then:Exec, OnModalWidgetQueued:Exec, ActivatableWidget:Common Activatable Widget Object Reference) {is=title K2Node_AsyncAction}
   n4 = Cast To SubscriptionCancellationInfoModal!(execute=@n3.OnModalWidgetQueued, Object=@n3) -> (then, CastFailed, AsSubscription Cancellation Info Modal:Subscription Cancellation Info Modal Object Reference) {is=title K2Node_DynamicCast}
   > Set CrewSubscriptionModalInfoVM(
    CrewSubscriptionModalInfoVM=GetModalViewModelById(self=Get CrewVM(), ModalId=Get ModalId(self=@n4))
    self:Subscription Cancellation Info Modal Object Reference=@n4
   )
   > Set CrewCancellationInfoVM!(
    execute
    CrewCancellationInfoVM:Crew Cancellation Info VM Object Reference=Get CancellationInfo!(self:Crew VM Object Reference=Get CrewVM()) -> (CancellationInfo:Crew Cancellation Info VM Object Reference)
    self:Subscription Cancellation Info Modal Object Reference=@n4
   ) -> (then, Output_Get:Crew Cancellation Info VM Object Reference)
   "On Clicked (RejoinCrew_Button)"!() -> (OutputDelegate, then, Button:Common Button Base Object Reference) {is=title K2Node_ComponentBoundEvent}
   > n5 = Branch(Condition=Get IsNotAutoRenewal() -> (IsNotAutoRenewal:bool))
   > n6 = QueueModalWidgetClassAsync(WidgetClass=/Crew/UI/Subscription/SubscriptionContinueModal.SubscriptionContinueModal_C)
   n7 = Cast To SubscriptionContinueModal!(execute=@n6.OnModalWidgetQueued, Object=@n6) -> (then, CastFailed, AsSubscription Continue Modal:Subscription Continue Modal Object Reference) {is=title K2Node_DynamicCast}
   > Set CrewSubscriptionModalInfoVM(
    CrewSubscriptionModalInfoVM=GetModalViewModelById(self=Get CrewVM(), ModalId=Get ModalId(self=@n7))
    self:Subscription Continue Modal Object Reference=@n7
   )
   n8 = QueueModalWidgetClassAsync(execute=@n5.else, WidgetClass=/Crew/UI/Subscription/SubscriptionRejoinModal.SubscriptionRejoinModal_C)
   > n9 = Cast To SubscriptionRejoinModal!(execute, Object=@n8) -> (then, CastFailed, AsSubscription Rejoin Modal:Subscription Rejoin Modal Object Reference) {is=title K2Node_DynamicCast}
   > Set CrewSubscriptionModalInfoVM(
    CrewSubscriptionModalInfoVM=GetModalViewModelById(self=Get CrewVM(), ModalId=Get ModalId(self=@n9))
    self:Subscription Rejoin Modal Object Reference=@n9
   )
   Event PreConstruct()
   Event Construct()
   Event Tick()
  ]}
 GetLegalText {nodes [
   n1 = GetLegalText!() -> (then, Disclamer:Text, Payment:Text) {is=title K2Node_FunctionEntry}
   > Return Node!(
    execute
    LegalText:Text=Format Text!(Format="LOCTEXT(\"{Disclaimer} {PaymentLegal}\")", Disclaimer:Text=@n1, PaymentLegal:Text=@n1.Payment) -> (Result)
   )
  ]}
 HandleSubState {nodes [
   n1 = HandleSubState!() -> (then, SubState:EMcpSubscriptionState Enum=Inactive) {is=title K2Node_FunctionEntry}
   > LoudButtonText_ReJoinCrew!(execute, self, SubState:EMcpSubscriptionState Enum=@n1) -> (then)
   > SetVisibility(self=Get ActiveWExternalState() -> (ActiveWExternalState:Text Block Object Reference), InVisibility=Collapsed)
   > SetVisibility(self=[Get Divider() -> (Divider:Image Object Reference), Get ResolveIssueButton()], InVisibility=Collapsed)
   > n2 = Sequence()
   > n3 = Switch on EMcpSubscriptionState!(execute, Selection:EMcpSubscriptionState Enum=@n1) -> (
    Inactive:Exec
    Active:Exec
    ActiveWithExternalBilling:Exec
    NotAutoRenewable:Exec
    Canceled:Exec
    PaymentProcessError:Exec
    PlatformAuthError:Exec
    RewardsDelayed:Exec
    Unavailable:Exec
    Unknown:Exec
   ) {is=title K2Node_SwitchEnum}
   SetActiveWidget(execute=@n3.PaymentProcessError, self=Get CrewSubState(), Widget=Get CrewSubscribed() -> (CrewSubscribed:Vertical Box Object Reference))
   > SetVisibility(self=Get CancelButton(), InVisibility=Visible)
   > SetVisibility(self=[Get Divider(), Get ResolveIssueButton()], InVisibility=Visible)
   > SetText(
    self:WBP UIKit Info Box Object Reference=Get WBP_UIKit_InfoBox() -> (WBP_UIKit_InfoBox:WBP UIKit Info Box Object Reference)
    InText:Text="LOCTEXT(\"Your rewards are temporarily on hold as there was an issue processing your payment.\")"
   )
   SetActiveWidget(execute=[@n3.Active, @n3.Unknown], self=Get CrewSubState(), Widget=Get CrewSubscribed())
   > n4 = SetVisibility(self=Get CancelButton(), InVisibility=Visible)
   SetText(execute=[@n4.then, @n5.then], self=Get WBP_UIKit_InfoBox())
   SetActiveWidget(
    execute=[@n3.Canceled, @n3.NotAutoRenewable]
    self=Get CrewSubState()
    Widget=Get UnsbCrewReJoinCrew() -> (UnsbCrewReJoinCrew:Vertical Box Object Reference)
   )
   > SetText(self=Get WBP_UIKit_InfoBox(), InText=GetSubscriptionNextBillingDateAsText!(self:Crew VM Object Reference=Get CrewVM()) -> (ReturnValue:Text))
   SetActiveWidget(execute=@n3.ActiveWithExternalBilling, self=Get CrewSubState(), Widget=Get CrewSubscribed())
   > SetVisibility(self=Get CancelButton(), InVisibility=Visible)
   > n5 = SetVisibility(self=Get ActiveWExternalState(), InVisibility=Collapsed)
   n6 = Sequence(execute=@n3.PlatformAuthError)
   SetText(
    execute=@n6.then_1
    self=Get WBP_UIKit_InfoBox()
    InText="LOCTEXT(\"Please, log in to the platform you subscribed to the Fortnite Crew on to receive your subscription benefits.\")"
   )
   SetActiveWidget(execute=[@n7.then_0, @n6.then_0], self=Get CrewSubState(), Widget=Get CrewSubscribed())
   > SetVisibility(self=Get CancelButton(), InVisibility=Visible)
   n7 = Sequence(execute=@n3.RewardsDelayed)
   SetText(
    execute=@n7.then_1
    self=Get WBP_UIKit_InfoBox()
    InText="LOCTEXT(\"There appears to be an error processing your rewards - please try again later, or contact Player Support if the problem persists.\")"
   )
   SetActiveWidget(execute=@n3.Unavailable, self=Get CrewSubState(), Widget=Get CrewSubscribed())
   > SetVisibility(self=Get CancelButton(), InVisibility=Collapsed)
   > SetText(self=Get WBP_UIKit_InfoBox(), InText="LOCTEXT(\"Fortnite Crew is not available. For more information please visit fn.gg/FortniteCrew\")")
   n8 = Get WBP_UIKit_InfoBox()
   n9 = GetText!(self:Rich Text Block Object Reference=Get Text(self:WBP UIKit Info Box Object Reference=@n8) -> (Text:Common Rich Text Block Object Reference)) -> (ReturnValue:Text)
   SetVisibility(
    execute=@n2.then_1
    self=@n8
    InVisibility=Select(Option 0:ESlateVisibility Enum=SelfHitTestInvisible, Option 1:ESlateVisibility Enum=Collapsed, Index=TextIsEmpty(InText=@n9)) -> (ReturnValue:ESlateVisibility Enum)
   )
  ]}
 LoudButtonText_ReJoinCrew {nodes [
   n1 = LoudButtonText_ReJoinCrew!() -> (then, SubState:EMcpSubscriptionState Enum=Inactive) {is=title K2Node_FunctionEntry}
   n2 = Get CrewVM()
   > Set SecondaryText!(
    execute
    SecondaryText:Text=Format Text!(Format="LOCTEXT(\"{price} PER MONTH\")", price:Text=Get PriceText!(self:Crew VM Object Reference=@n2) -> (PriceText:Text)) -> (Result)
    self:Fort CTAButton Object Reference=Get RejoinCrew_Button()
   ) -> (then, Output_Get:Text)
   > n3 = Switch on EMcpSubscriptionState(Selection=@n1)
   Set Text(execute=@n3.Canceled, Text="LOCTEXT(\"rejoin fortnite crew\")", self:Fort CTAButton Object Reference=Get RejoinCrew_Button())
   > Set IsNotAutoRenewal(IsNotAutoRenewal:bool=false) -> (Output_Get:bool)
   Set Text(execute=@n3.NotAutoRenewable, Text="LOCTEXT(\"continue membership\")", self=Get RejoinCrew_Button())
   > Set IsNotAutoRenewal(IsNotAutoRenewal=true)
  ]}
 SeeBenefits {nodes [SeeBenefits() {is=title K2Node_FunctionEntry}]}
]
variables [
 CharBGBundle {type=Texture2D}
 CharBGCrew {type=Texture2D, default=/Crew/UI/Textures/T_UI_CrewCharacters}
 ClosePage {}
 CrewMaterial {type=MaterialInstance, default=/Crew/UI/Materials/M_UI_CrewUpsell_Landing_UnSub}
 IsNotAutoRenewal {type=boolean, default=true}
 SeeBenefits {}
 SubMaterial {type=MaterialInstance, default=/Crew/UI/Materials/M_UI_CrewUpsell_Landing_Sub}
]
widgets {
 info {
  inheritedWidgetCount=0
  namedSlotCount=1
  parentClass=/Script/CrewUI.FortProgressiveTableOfContentsScreen
  rootWidgetClass=&SafeZone
  widgetCount=32
 }
 widgets [
  SafeZone {widgetClassPath=&SafeZone, bIsVariable=true, widget=SafeZone}
  Overlay_1 {widgetClassPath=&Overlay, parent=SafeZone, widget=Overlay_1, slot {class=SafeZoneSlot}}
  BG_Crew {
   widgetClassPath=/Script/EpicCMSUIFramework.EpicCMSImage
   bIsVariable=true
   parent=Overlay_1
   widget=BG_Crew
   properties {
    bMatchImageSize=false
    materialTextureParamName=CMS_Crew_Bento
    visibility=SelfHitTestInvisible
    brush {
     bIsDynamicallyLoaded=false
     drawAs=Image
     imageType=FullColor
     mirroring=NoMirror
     resourceName=None
     resourceObject=&M_UI_CrewUpsell_Landing_UnSub
     tiling=NoTile
     imageSize=<200 200>
     margin=<ltrb 0 0 0 0>
     outlineSettings {
      bUseBrushTransparency=false
      roundingType=HalfHeightRadius
      width=0
      color {colorUseRule=UseColor_Specified, specifiedColor=<rgba 0 0 0 0>}
      cornerRadii=<0 0 0 0>
     }
     tintColor {colorUseRule=UseColor_Specified, specifiedColor=<rgba 1 1 1 1>}
     uVRegion {bIsValid=false, max=<0 0>, min=<0 0>}
    }
   }
   slot {class=OverlaySlot, properties {horizontalAlignment=HAlign_Fill, verticalAlignment=VAlign_Fill}}
  }
  BottomScreenGradient {
   widgetClassPath=&Image
   bIsVariable=true
   parent=Overlay_1
   widget=BottomScreenGradient
   properties {
    visibility=SelfHitTestInvisible
    brush {
     bIsDynamicallyLoaded=false
     drawAs=Image
     imageType=FullColor
     mirroring=NoMirror
     resourceName=None
     resourceObject=/Crew/UI/Materials/MI_UI_CrewUpsell_Gradient
     tiling=NoTile
     imageSize=<1200 704>
     margin=<ltrb 0 0 0 0>
     outlineSettings {
      bUseBrushTransparency=false
      roundingType=HalfHeightRadius
      width=0
      color {colorUseRule=UseColor_Specified, specifiedColor=<rgba 0 0 0 0>}
      cornerRadii=<0 0 0 0>
     }
     tintColor {colorUseRule=UseColor_Specified, specifiedColor=<rgba 1 1 1 1>}
     uVRegion {bIsValid=false, max=<0 0>, min=<0 0>}
    }
   }
   slot {class=OverlaySlot, properties {horizontalAlignment=HAlign_Fill, verticalAlignment=VAlign_Fill}}
  }
  SizeBox {
   widgetClassPath=&SizeBox
   bIsVariable=true
   parent=Overlay_1
   widget=SizeBox
   properties {bOverride_WidthOverride=true, widthOverride=1920}
   slot {class=OverlaySlot, properties {horizontalAlignment=HAlign_Center}}
  }
  Overlay_19 {widgetClassPath=&Overlay, parent=SizeBox, widget=Overlay_19, slot {class=SizeBoxSlot}}
  Char_BG {
   widgetClassPath=&FortLazyImage
   bIsVariable=true
   parent=Overlay_19
   widget=Char_BG
   properties {brush {
     bIsDynamicallyLoaded=false
     drawAs=Image
     imageType=FullColor
     mirroring=NoMirror
     resourceName=None
     resourceObject=/Crew/UI/Textures/T_UI_CrewCharacters_PH
     tiling=NoTile
     imageSize=<1920 1080>
     margin=<ltrb 0 0 0 0>
     outlineSettings {
      bUseBrushTransparency=false
      roundingType=HalfHeightRadius
      width=0
      color {colorUseRule=UseColor_Specified, specifiedColor=<rgba 0 0 0 0>}
      cornerRadii=<0 0 0 0>
     }
     tintColor {colorUseRule=UseColor_Specified, specifiedColor=<rgba 1 1 1 1>}
     uVRegion {bIsValid=false, max=<0 0>, min=<0 0>}
    }}
   slot {class=OverlaySlot, properties {horizontalAlignment=HAlign_Right, verticalAlignment=VAlign_Bottom}}
  }
  VerticalBox_0 {
   widgetClassPath=&VerticalBox
   parent=Overlay_19
   widget=VerticalBox_0
   properties {bOverride_Cursor=true}
   slot {class=OverlaySlot, properties {horizontalAlignment=HAlign_Fill, verticalAlignment=VAlign_Fill}}
  }
  CrewGrid {
   widgetClassPath=&GridPanel
   parent=VerticalBox_0
   widget=CrewGrid
   properties {columnFill [2], rowFill [1,0]}
   slot {class=VerticalBoxSlot, properties {padding=<ltrb 0 0 0 150>, size {sizeRule=Fill, value=1}}}
  }
  LeftUpsell {
   widgetClassPath=&SizeBox
   parent=CrewGrid
   widget=LeftUpsell
   properties {bOverride_WidthOverride=true, widthOverride=800}
   slot {class=GridSlot, properties {horizontalAlignment=HAlign_Left, verticalAlignment=VAlign_Center, padding=<ltrb 159 0 0 0>}}
  }
  ContentVB {widgetClassPath=&VerticalBox, parent=LeftUpsell, widget=ContentVB, slot {class=SizeBoxSlot}}
  CrewLogo_N {
   widgetClassPath=&Image
   bIsVariable=true
   parent=ContentVB
   widget=CrewLogo_N
   properties {
    visibility=SelfHitTestInvisible
    brush {
     bIsDynamicallyLoaded=false
     drawAs=Image
     imageType=FullColor
     mirroring=NoMirror
     resourceName=None
     resourceObject=/Game/Athena/UI/Logos/Textures/M_UI_CrewLogo_Large
     tiling=NoTile
     imageSize=<432 128>
     margin=<ltrb 0 0 0 0>
     outlineSettings {
      bUseBrushTransparency=false
      roundingType=HalfHeightRadius
      width=0
      color {colorUseRule=UseColor_Specified, specifiedColor=<rgba 0 0 0 0>}
      cornerRadii=<0 0 0 0>
     }
     tintColor {colorUseRule=UseColor_Specified, specifiedColor=<rgba 1 1 1 1>}
     uVRegion {bIsValid=false, max=<0 0>, min=<0 0>}
    }
   }
   slot {class=VerticalBoxSlot, properties {verticalAlignment=VAlign_Center, padding=<ltrb 0 0 0 10>}}
  }
  CrewSubState {
   widgetClassPath=&CommonVisibilitySwitcher
   bIsVariable=true
   parent=ContentVB
   widget=CrewSubState
   slot {class=VerticalBoxSlot, properties {verticalAlignment=VAlign_Bottom}}
  }
  CrewSubscribed {
   widgetClassPath=&VerticalBox
   bIsVariable=true
   parent=CrewSubState
   widget=CrewSubscribed
   slot {class=CommonVisibilitySwitcherSlot}
  }
  WBP_UIKit_Highlight_Base {
   widgetClassPath=/FortUILibrary/Systems/Highlight/WBP_UIKit_Highlight_Base
   bIsVariable=true
   parent=CrewSubscribed
   widget=WBP_UIKit_Highlight_Base
   properties {highlightText=Subscribed, highlightType=Social}
   slot {class=VerticalBoxSlot, properties {horizontalAlignment=HAlign_Center, verticalAlignment=VAlign_Center, padding=<ltrb 0 0 0 15>}}
  }
  ResolveIssueButton {
   widgetClassPath=/FortUILibrary/UIKit/Buttons/WBP_UIKit_Button_Loud
   bIsVariable=true
   parent=CrewSubscribed
   widget=ResolveIssueButton
   properties {
    bRequiresHold=true
    buttonSize=Small
    holdData=/FortUILibrary/Input/PurchaseInteraction_HoldData.PurchaseInteraction_HoldData_C
    minWidth=450
    text=RESOLVE ISSUE
   }
   slot {class=VerticalBoxSlot, properties {horizontalAlignment=HAlign_Center, verticalAlignment=VAlign_Center, padding=<ltrb 0 30 0 30>}}
  }
  Divider {
   widgetClassPath=&Image
   bIsVariable=true
   parent=CrewSubscribed
   widget=Divider
   properties {brush {
     bIsDynamicallyLoaded=false
     drawAs=Image
     imageType=NoImage
     mirroring=NoMirror
     resourceName=None
     resourceObject=null
     tiling=NoTile
     imageSize=<400 2>
     margin=<ltrb 0 0 0 0>
     outlineSettings {
      bUseBrushTransparency=false
      roundingType=HalfHeightRadius
      width=0
      color {colorUseRule=UseColor_Specified, specifiedColor=<rgba 0 0 0 0>}
      cornerRadii=<0 0 0 0>
     }
     tintColor {colorUseRule=UseColor_Specified, specifiedColor=<rgba 1 1 1 1>}
     uVRegion {bIsValid=false, max=<0 0>, min=<0 0>}
    }}
   slot {class=VerticalBoxSlot, properties {horizontalAlignment=HAlign_Center, verticalAlignment=VAlign_Center}}
  }
  CancelButton {
   widgetClassPath=&WBP_UIKit_Button_Regular
   bIsVariable=true
   parent=CrewSubscribed
   widget=CancelButton
   properties {bHideInputAction=true, minWidth=375, text=Cancellation Info}
   slot {class=VerticalBoxSlot, properties {horizontalAlignment=HAlign_Center, padding=<ltrb 30 30 30 30>}}
  }
  ActiveWExternalState {
   widgetClassPath=&UEFN_TextBlock
   bIsVariable=true
   parent=CrewSubscribed
   widget=ActiveWExternalState
   properties {
    justification=Center
    text=Your subscription is managed by a billing partner.
    font {
     bForceMonospaced=false
     bMaterialIsStencil=false
     fontMaterial=null
     fontObject=&HeadingNow
     letterSpacing=0
     monospacedWidth=1
     size=13.5
     skewAmount=0
     typefaceFontName=74 Regular
     outlineSettings {
      bApplyOutlineToDropShadows=false
      bMiteredCorners=false
      bSeparateFillAlpha=false
      outlineMaterial=null
      outlineSize=0
      outlineColor=<rgba 0 0 0 1>
     }
    }
   }
   slot {class=VerticalBoxSlot}
  }
  UnsbCrewReJoinCrew {
   widgetClassPath=&VerticalBox
   bIsVariable=true
   parent=CrewSubState
   widget=UnsbCrewReJoinCrew
   slot {class=CommonVisibilitySwitcherSlot, properties {horizontalAlignment=HAlign_Center}}
  }
  Subed_1 {
   widgetClassPath=&WBP_UIKit_Highlight_Base
   bIsVariable=true
   parent=UnsbCrewReJoinCrew
   widget=Subed_1
   properties {highlightText=Subscribed, highlightType=Social}
   slot {class=VerticalBoxSlot, properties {horizontalAlignment=HAlign_Center, verticalAlignment=VAlign_Center, padding=<ltrb 0 15 0 15>}}
  }
  RejoinCrew_Button {
   widgetClassPath=&WBP_UIKit_Button_Loud
   bIsVariable=true
   parent=UnsbCrewReJoinCrew
   widget=RejoinCrew_Button
   properties {
    bRequiresHold=true
    buttonSize=Large
    holdData=&PurchaseInteraction_HoldData_C
    minWidth=450
    secondaryText=$XX.XX PER MONTH
    text=REJOIN FORTNITE CREW
   }
   slot {class=VerticalBoxSlot, properties {horizontalAlignment=HAlign_Center, verticalAlignment=VAlign_Center, padding=<ltrb 0 30 0 30>}}
  }
  WBP_UIKit_InfoBox {
   widgetClassPath=/FortUILibrary/UIKit/InfoBox/WBP_UIKit_InfoBox
   bIsVariable=true
   parent=CrewGrid
   widget=WBP_UIKit_InfoBox
   properties {
    previewText=Fortnite Crew is not available. For more information please visit fn.gg/FortniteCrew
    showIcon=false
    showStatusIndicator=true
    textJustification=Center
    use Preview Text=true
   }
   slot {class=GridSlot, properties {column=1, horizontalAlignment=HAlign_Right, verticalAlignment=VAlign_Bottom}}
  }
  HorizontalBox_0 {
   widgetClassPath=&HorizontalBox
   parent=VerticalBox_0
   widget=HorizontalBox_0
   slot {class=VerticalBoxSlot, properties {verticalAlignment=VAlign_Bottom, padding=<ltrb 100 0 100 50>}}
  }
  SizeBox_1 {
   widgetClassPath=&SizeBox
   parent=HorizontalBox_0
   widget=SizeBox_1
   properties {bOverride_WidthOverride=true, maxDesiredWidth=900, widthOverride=900}
   slot {class=HorizontalBoxSlot, properties {horizontalAlignment=HAlign_Left, verticalAlignment=VAlign_Bottom, padding=<ltrb 0 0 40 40>}}
  }
  Text_Legal {
   widgetClassPath=&CommonTextBlock
   bIsVariable=true
   parent=SizeBox_1
   widget=Text_Legal
   properties {
    autoWrapText=true
    mobileFontSizeMultiplier=1.25
    scrollOrientation=Orient_Vertical
    text="[PH] Auto-renews each month until cancelled. Cancel anytime. Territorial restrictions apply. Ability to claim Premium Rewards during active subscription only. Select \"Subscription Terms\" for more info."
    colorAndOpacity {colorUseRule=UseColor_Specified, specifiedColor=<rgba 0.5776 0.5776 0.5776 1>}
    font {
     bForceMonospaced=false
     bMaterialIsStencil=false
     fontMaterial=null
     fontObject=&NotoSans
     letterSpacing=0
     monospacedWidth=1
     size=24
     skewAmount=0
     typefaceFontName=NotoSans-Regular
     outlineSettings {
      bApplyOutlineToDropShadows=false
      bMiteredCorners=false
      bSeparateFillAlpha=false
      outlineMaterial=null
      outlineSize=0
      outlineColor=<rgba 0 0 0 1>
     }
    }
    shadowOffset=<0 0>
   }
   slot {class=SizeBoxSlot}
  }
  WBP_ActionBarContainer {
   widgetClassPath=/Game/UI/Frontend/BottomBar/WBP_ActionBarContainer
   bIsVariable=true
   parent=HorizontalBox_0
   widget=WBP_ActionBarContainer
   slot {
    class=HorizontalBoxSlot
    properties {horizontalAlignment=HAlign_Right, verticalAlignment=VAlign_Bottom, padding=<ltrb 0 0 0 10>, size {sizeRule=Fill, value=1}}
   }
  }
  HBox_BottomButtons {widgetClassPath=&HorizontalBox, namedSlotHost=WBP_ActionBarContainer, widget=HBox_BottomButtons}
  Button_SeeBenefits {
   widgetClassPath=/FortUILibrary/UIKit/Buttons/WBP_UIKit_Button_Quiet
   bIsVariable=true
   parent=HBox_BottomButtons
   widget=Button_SeeBenefits
   properties {minWidth=184, text=See Benefits, triggeringInputAction {dataTable=&InputActionDataTable, rowName=GenericFaceButtonTop}}
   slot {class=HorizontalBoxSlot, properties {horizontalAlignment=HAlign_Right, verticalAlignment=VAlign_Top}}
  }
  WBP_AthenaBottomBar_WithNotifications {
   widgetClassPath=/Game/UI/Frontend/BottomBar/WBP_AthenaBottomBar_WithNotifications
   bIsVariable=true
   parent=HBox_BottomButtons
   widget=WBP_AthenaBottomBar_WithNotifications
   slot {class=HorizontalBoxSlot, properties {padding=<ltrb 24 0 0 0>}}
  }
  Button_Back {
   widgetClassPath=&WBP_UIKit_Button_Quiet
   bIsVariable=true
   parent=HBox_BottomButtons
   widget=Button_Back
   properties {text=Back, triggeringInputAction {dataTable=&InputActionDataTable, rowName=GenericBack}}
   slot {class=HorizontalBoxSlot, properties {horizontalAlignment=HAlign_Right, verticalAlignment=VAlign_Top, padding=<ltrb 24 0 0 0>}}
  }
  WBP_PlayerInfo {
   widgetClassPath=/Game/UI/Frontend/ItemShop/MultiProduct/WBP_PlayerInfo
   bIsVariable=true
   parent=Overlay_1
   widget=WBP_PlayerInfo
   slot {class=OverlaySlot, properties {horizontalAlignment=HAlign_Right, padding=<ltrb 0 40 80 0>}}
  }
 ]
}
