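# Crew benefit tile button showing an image, gradient background, title, highlight tag and description, configured through setter functions.
# Images come from a downloaded URL or supplied texture, feeding a dynamic material along with pixel dimensions for its ratio parameters.
# Hover focuses the outline and plays the Selected animation, revealing the description and scaling the image.
# Touch input hides the outline and expands inactive tiles; compact UI overrides the description font size.
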WidgetBlueprint /Crew/UI/WBP_CrewBenefit_Tile
animations [Selected {
  binding_count=3
  length_seconds=0.25
  object=&self:Selected
  sequence {bindings [
    DescriptionSize {
     id=#1
     possessed_class=SizeBox
     tracks [{
       type=MovieSceneFloatTrack
       display_name=Max Desired Height
       sections [{type=MovieSceneFloatSection, end_frame=5, start_frame=0, channels [None {type=MovieSceneScriptingFloatChannel, keys [[0,0],[1,0],[5,256]]}]}]
      }]
    }
    CMS_Image {
     id=#2
     possessed_class=FortLazyImage
     tracks [{
       type=MovieSceneWidgetMaterialTrack
       sections [{
         type=MovieSceneParameterSection
         end_frame=5
         start_frame=0
         channels [TextureScale {type=MovieSceneScriptingFloatChannel, keys [[0,1.88],[5,2]]}]
        }]
      }]
    }
    Description {
     id=#3
     possessed_class=RichTextBlock
     tracks [{
       type=MovieSceneFloatTrack
       display_name=Render Opacity
       sections [{type=MovieSceneFloatSection, end_frame=5, start_frame=0, channels [None {type=MovieSceneScriptingFloatChannel, keys [[0,0],[3,0],[5,1]]}]}]
      }]
    }
   ]}
 }]
graphs [
 EventGraph {nodes [
   n1 = OnImageURL!() -> (OutputDelegate, then, URL:String) {is=title K2Node_CustomEvent}
   > n2 = DownloadImage!(execute:Exec, URL:String=@n1.URL) -> (then:Exec, OnSuccess:Exec, OnFail:Exec, Texture:Texture 2DDynamic Object Reference) {is=title K2Node_AsyncAction}
   n3 = GetDynamicMaterial(execute=@n2.OnSuccess, self=Get CMS_Image() -> (CMS_Image:Fort Lazy Image Object Reference))
   > SetTextureParameterValue(self=@n3, ParameterName=MainTexture, Value=@n2)
   > n4 = Get Texture Parameter Value!(execute, self:Material Instance Dynamic Object Reference=@n3, ParameterName:Name=MainTexture) -> (then, ReturnValue:Texture Object Reference) {is=title K2Node_CallFunction}
   > n5 = Cast To Texture2D(Object=@n4)
   > SetScalarParameterValue(
    self=[@n3, @n3]
    ParameterName=RatioX
    Value="To Float (Integer)"(InInt=GetSizeX!(self:Texture 2D Object Reference=@n5) -> (ReturnValue:int=0))
   )
   > SetScalarParameterValue(self=@n3, ParameterName=RatioY, Value="To Float (Integer)"(InInt=GetSizeY!(self:Texture 2D Object Reference=@n5) -> (ReturnValue:int=0)))
   n6 = Event PreConstruct()
   > Set CachedInputType(CachedInputType:ECommonInputType Enum=GetCurrentInputType(self=Get CommonInputSubsystem())) -> (Output_Get:ECommonInputType Enum)
   > SetCompactUIStyle() {is=title K2Node_CallFunction}
   > Branch(Condition=@n6.IsDesignTime)
   > SetTileSize!(execute, self, TileSize:double=Get TileSize() -> (TileSize:double)) -> (then)
   Event On Hovered()
   > n7 = Set IsActiveTile(IsActiveTile:bool=true) -> (Output_Get:bool)
   > UpdateVisuals!(execute, self, ShouldFocus:bool=@n7) -> (then)
   Event On Unhovered()
   > n8 = Set IsActiveTile(IsActiveTile=false)
   > UpdateVisuals(ShouldFocus=@n8)
   n9 = Event On Input Method Changed!() -> (OutputDelegate, then, CurrentInputType:ECommonInputType Enum) {is=title K2Node_Event}
   > InputMethodChanged!(execute, self, InputType:ECommonInputType Enum=@n9.CurrentInputType) -> (then)
  ]}
 InputMethodChanged {nodes [
   n1 = InputMethodChanged!() -> (then, InputType:ECommonInputType Enum) {is=title K2Node_FunctionEntry}
   > n2 = Sequence()
   n3 = "Not Equal (Enum)"!(A:ECommonInputType Enum=@n1, B:ECommonInputType Enum=Touch) -> (ReturnValue:bool)
   > SetBlockVisibility!(
    execute
    self:WBP UIKit Block Base Object Reference=Get Block_Outline() -> (Block_Outline:WBP UIKit Block Outline Object Reference)
    Visible:bool=@n3
   ) -> (then)
   > n4 = Branch(Condition=Get IsActiveTile() -> (IsActiveTile:bool))
   n5 = Branch(execute=@n4.else, Condition=@n3)
   > Branch(Condition="Equal (Enum)"(A=Get CachedInputType() -> (CachedInputType:ECommonInputType Enum), B=Touch))
   > PlayAnimation(InAnimation=Get Selected() -> (Selected:Widget Animation Object Reference), PlayMode=Reverse)
   PlayAnimation(execute=@n5.else, InAnimation=Get Selected())
   Set CachedInputType(execute=@n2.then_1, CachedInputType=@n1)
  ]}
 SetColorBot {nodes [
   n1 = SetColorBot!() -> (then, Color:Color Structure) {is=title K2Node_FunctionEntry}
   > n2 = GetDynamicMaterial(self=Get BG() -> (BG:Image Object Reference))
   > SetVectorParameterValue(self=@n2, ParameterName=ColorBot, Value="To LinearColor (Color)"!(InColor:Color Structure=@n1) -> (ReturnValue:Linear Color Structure))
  ]}
 SetColorTop {nodes [
   n1 = SetColorTop!() -> (then, Color:Color Structure) {is=title K2Node_FunctionEntry}
   > n2 = GetDynamicMaterial(self=Get BG())
   > SetVectorParameterValue(self=@n2, ParameterName=ColorTop, Value="To LinearColor (Color)"(InColor=@n1))
  ]}
 SetCompactUIStyle {nodes [
   SetCompactUIStyle() {is=title K2Node_FunctionEntry}
   > Branch(Condition=IsUsingCompactUI())
   n1 = Get Description() -> (Description:Rich Text Block Object Reference)
   n2 = Break Text Block Style!(TextBlockStyle:"Text Block Style Structure (by ref)"=Get DefaultTextStyleOverride!(self:Rich Text Block Object Reference=@n1) -> (DefaultTextStyleOverride:Text Block Style Structure)) -> (
    Font:Slate Font Info Structure
    ColorAndOpacity:Slate Color Structure
    ShadowOffset:Vector 2D Structure
    ShadowColorAndOpacity:Linear Color Structure
    HighlightColor:Slate Color Structure
    HighlightShape:Slate Brush Structure
    StrikeBrush:Slate Brush Structure
    UnderlineBrush:Slate Brush Structure
    TransformPolicy:ETextTransformPolicy Enum
    OverflowPolicy:ETextOverflowPolicy Enum
   ) {is=title K2Node_BreakStruct}
   n3 = Break Slate Font Info!(SlateFontInfo:"Slate Font Info Structure (by ref)"=@n2) -> (
    FontObject:Object Reference
    FontMaterial:Object Reference
    OutlineSettings:Font Outline Settings Structure
    TypefaceFontName:Name
    Size:float=~
    LetterSpacing:int
    SkewAmount:float=~
    bForceMonospaced:bool=~
    bMaterialIsStencil:bool=~
    MonospacedWidth:float=~
   ) {is=title K2Node_BreakStruct}
   > Set DefaultTextStyleOverride!(
    execute
    DefaultTextStyleOverride:Text Block Style Structure=Make Text Block Style!(
     Font:Slate Font Info Structure=MakeSlateFontInfo!(
      FontObject:Object Reference=@n3
      FontMaterial:Object Reference=@n3.FontMaterial
      OutlineSettings:Font Outline Settings Structure=@n3.OutlineSettings
      TypefaceFontName:Name=@n3.TypefaceFontName
      Size:float=15.75
      LetterSpacing:int=@n3.LetterSpacing
      SkewAmount:float=@n3.SkewAmount
      bForceMonospaced:bool=@n3.bForceMonospaced
      bMaterialIsStencil:bool=@n3.bMaterialIsStencil
      MonospacedWidth:float=@n3.MonospacedWidth
     ) -> (ReturnValue:Slate Font Info Structure)
     ColorAndOpacity:Slate Color Structure=@n2.ColorAndOpacity
     ShadowOffset:Vector 2D Structure=@n2.ShadowOffset
     ShadowColorAndOpacity:Linear Color Structure=@n2.ShadowColorAndOpacity
     HighlightColor:Slate Color Structure=@n2.HighlightColor
     HighlightShape:Slate Brush Structure=@n2.HighlightShape
     StrikeBrush:Slate Brush Structure=@n2.StrikeBrush
     UnderlineBrush:Slate Brush Structure=@n2.UnderlineBrush
     TransformPolicy:ETextTransformPolicy Enum=@n2.TransformPolicy
     OverflowPolicy:ETextOverflowPolicy Enum=@n2.OverflowPolicy
    ) -> (TextBlockStyle:Text Block Style Structure) {is=title K2Node_MakeStruct}
    self:Rich Text Block Object Reference=@n1
   ) -> (then, Output_Get:Text Block Style Structure)
  ]}
 SetIMGTexture {nodes [
   n1 = SetIMGTexture!() -> (then, Texture:Texture Object Reference) {is=title K2Node_FunctionEntry}
   > n2 = GetDynamicMaterial(self=Get CMS_Image())
   > SetTextureParameterValue(self=@n2, ParameterName=MainTexture, Value=@n1)
   > n3 = Get Texture Parameter Value(self=@n2, ParameterName=MainTexture)
   > n4 = Cast To Texture2D(Object=@n3)
   > SetScalarParameterValue(self=@n2, ParameterName=RatioX, Value="To Float (Integer)"(InInt=GetSizeX(self=@n4) -> (ReturnValue=0)))
   > SetScalarParameterValue(self=@n2, ParameterName=RatioY, Value="To Float (Integer)"(InInt=GetSizeY(self=@n4) -> (ReturnValue=0)))
  ]}
 SetImageURL {nodes [
   n1 = SetImageURL!() -> (then, URL:String) {is=title K2Node_FunctionEntry}
   > OnImageURL!(execute, self, URL:String=@n1) -> (then) {is=title K2Node_CallFunction}
  ]}
 SetTileSize {nodes [
   n1 = SetTileSize!() -> (then, TileSize:double=~) {is=title K2Node_FunctionEntry}
   > SetWidthOverride(
    self=Get TileSizeBox() -> (TileSizeBox:Size Box Object Reference)
    InWidthOverride="To Float (Integer)"(InInt=int * int(A:double=@n1, B=275))
   )
  ]}
 SetTileText {nodes [
   n1 = SetTileText!() -> (then, Title:String, Description:String, Tag:String) {is=title K2Node_FunctionEntry}
   n2 = Get Title() -> (Title:Common Text Object Reference)
   > SetVisibility(self=@n2, InVisibility=Select(Option 0=SelfHitTestInvisible, Option 1=Collapsed, Index=IsEmpty(InString=@n1)))
   > Set Text(Text="To Text (String)"(InString=@n1), self=@n2)
   n3 = Get Description() -> (Description:String)
   n4 = Get Description() -> (Description:Rich Text Block Object Reference)
   > SetVisibility(self=@n4, InVisibility=Select(Option 0=SelfHitTestInvisible, Option 1=Collapsed, Index=IsEmpty(InString=@n3)))
   > Set Text(Text="To Text (String)"(InString=@n3), self:Rich Text Block Object Reference=@n4)
   n5 = Get Tag() -> (Tag:String)
   n6 = Get WBP_UIKit_Highlight_Base() -> (WBP_UIKit_Highlight_Base:WBP UIKit Highlight Base Object Reference)
   > SetVisibility(self=@n6, InVisibility=Select(Option 0=SelfHitTestInvisible, Option 1=Collapsed, Index=IsEmpty(InString=@n5)))
   > SetText!(execute, self:WBP UIKit Highlight Base Object Reference=@n6, Text:Text="To Text (String)"(InString=@n5)) -> (then)
  ]}
 UpdateVisuals {nodes [
   n1 = UpdateVisuals!() -> (then, ShouldFocus:bool=~) {is=title K2Node_FunctionEntry}
   > RequestTransition!(
    execute
    self:UIKit Block Object Reference=Get Block_Outline()
    State:EUIKitBlockVisualState Enum=Select(Option 0:EUIKitBlockVisualState Enum=Unfocused, Option 1:EUIKitBlockVisualState Enum=Focused, Index=@n1) -> (ReturnValue:EUIKitBlockVisualState Enum)
   ) -> (then)
   > PlayAnimation(
    InAnimation=Get Selected()
    PlayMode=Select(Option 0:EUMGSequencePlayMode Enum=Reverse, Option 1:EUMGSequencePlayMode Enum, Index=@n1) -> (ReturnValue:EUMGSequencePlayMode Enum)
   )
  ]}
]
variables [
 CachedInputType {type=ECommonInputType, default=MouseAndKeyboard}
 IsActiveTile {type=boolean}
 TileSize {type=number, default=1}
]
widgets {
 info {inheritedWidgetCount=1, parentClass=/Script/FortniteUI.UIKitHoldableModularButton, rootWidgetClass=&SizeBox, widgetCount=13}
 widgets [
  TileSizeBox {
   widgetClassPath=&SizeBox
   bIsVariable=true
   widget=TileSizeBox
   properties {bOverride_HeightOverride=true, bOverride_WidthOverride=true, heightOverride=400, widthOverride=275}
  }
  Overlay_39 {widgetClassPath=&Overlay, parent=TileSizeBox, widget=Overlay_39, slot {class=SizeBoxSlot}}
  Overlay_41 {
   widgetClassPath=&Overlay
   parent=Overlay_39
   widget=Overlay_41
   slot {class=OverlaySlot, properties {horizontalAlignment=HAlign_Fill, verticalAlignment=VAlign_Bottom}}
  }
  BG {
   widgetClassPath=&Image
   bIsVariable=true
   parent=Overlay_41
   widget=BG
   properties {brush {
     bIsDynamicallyLoaded=false
     drawAs=Image
     imageType=FullColor
     mirroring=NoMirror
     resourceName=None
     resourceObject=/Crew/UI/Materials/CrewBenefits/Tile/MI_UI_Benefit_Tile_Car
     tiling=NoTile
     imageSize=<32 329>
     margin=<ltrb 0 0 0 0>
     outlineSettings {
      bUseBrushTransparency=false
      roundingType=HalfHeightRadius
      width=0
      color {colorUseRule=UseColor_Specified, specifiedColor=<rgba 0 0 0 0>}
      cornerRadii=<0 0 0 0>
     }
     tintColor {colorUseRule=UseColor_Specified, specifiedColor=<rgba 1 1 1 1>}
     uVRegion {bIsValid=false, max=<0 0>, min=<0 0>}
    }}
   slot {class=OverlaySlot, properties {horizontalAlignment=HAlign_Fill, verticalAlignment=VAlign_Fill}}
  }
  Block_Outline {
   widgetClassPath=/FortUILibrary/UIKit/Buttons/Blocks/WBP_UIKit_Block_Outline
   bIsVariable=true
   parent=Overlay_41
   widget=Block_Outline
   properties {bottomLeftCornerRadius=30, bottomRightCornerRadius=30, radiusInPixels=true, topLeftCornerRadius=30, topRightCornerRadius=30}
   slot {class=OverlaySlot, properties {horizontalAlignment=HAlign_Fill, verticalAlignment=VAlign_Fill}}
  }
  CMS_Image {
   widgetClassPath=&FortLazyImage
   bIsVariable=true
   parent=Overlay_39
   widget=CMS_Image
   properties {brush {
     bIsDynamicallyLoaded=false
     drawAs=Image
     imageType=FullColor
     mirroring=NoMirror
     resourceName=None
     resourceObject=/Crew/UI/Materials/CrewBenefits/Tile/MI_UI_Benefit_Tile_Character
     tiling=NoTile
     imageSize=<32 374>
     margin=<ltrb 0 0 0 0>
     outlineSettings {
      bUseBrushTransparency=false
      roundingType=HalfHeightRadius
      width=0
      color {colorUseRule=UseColor_Specified, specifiedColor=<rgba 0 0 0 0>}
      cornerRadii=<0 0 0 0>
     }
     tintColor {colorUseRule=UseColor_Specified, specifiedColor=<rgba 1 1 1 1>}
     uVRegion {bIsValid=false, max=<0 0>, min=<0 0>}
    }}
   slot {class=OverlaySlot, properties {horizontalAlignment=HAlign_Fill, verticalAlignment=VAlign_Center}}
  }
  VerticalBox_0 {
   widgetClassPath=&VerticalBox
   parent=Overlay_39
   widget=VerticalBox_0
   slot {class=OverlaySlot, properties {horizontalAlignment=HAlign_Fill, verticalAlignment=VAlign_Bottom, padding=<ltrb 16 16 16 0>}}
  }
  Title {
   widgetClassPath=&CommonTextBlock
   bIsVariable=true
   parent=VerticalBox_0
   widget=Title
   properties {
    autoWrapText=true
    text=Benefit Title
    font {
     bForceMonospaced=false
     bMaterialIsStencil=false
     fontMaterial=null
     fontObject=&HeadingNow
     letterSpacing=0
     monospacedWidth=1
     size=20.25
     skewAmount=0
     typefaceFontName=75 Medium
     outlineSettings {
      bApplyOutlineToDropShadows=false
      bMiteredCorners=false
      bSeparateFillAlpha=false
      outlineMaterial=null
      outlineSize=0
      outlineColor=<rgba 0 0 0 1>
     }
    }
    shadowOffset=<0 0>
   }
   slot {class=VerticalBoxSlot}
  }
  WBP_UIKit_Highlight_Base {
   widgetClassPath=/FortUILibrary/Systems/Highlight/WBP_UIKit_Highlight_Base
   bIsVariable=true
   parent=VerticalBox_0
   widget=WBP_UIKit_Highlight_Base
   properties {highlightSize=Compact, highlightText=Only in the Bundle, highlightType=Attention}
   slot {class=VerticalBoxSlot, properties {horizontalAlignment=HAlign_Left, verticalAlignment=VAlign_Top, padding=<ltrb 0 8 0 0>}}
  }
  DescriptionSize {
   widgetClassPath=&SizeBox
   parent=VerticalBox_0
   widget=DescriptionSize
   properties {bOverride_MaxDesiredHeight=true, clipping=ClipToBounds}
   slot {class=VerticalBoxSlot, properties {padding=<ltrb 0 0 0 16>}}
  }
  Description {
   widgetClassPath=/Script/UMG.RichTextBlock
   bIsVariable=true
   parent=DescriptionSize
   widget=Description
   properties {
    autoWrapText=true
    bOverrideDefaultStyle=true
    renderOpacity=0
    text="Get 800 V-Bucks !"
    defaultTextStyleOverride {
     overflowPolicy=Clip
     transformPolicy=None
     colorAndOpacity {colorUseRule=UseColor_Specified, specifiedColor=<rgba 1 1 1 1>}
     font {
      bForceMonospaced=false
      bMaterialIsStencil=false
      fontMaterial=null
      fontObject=&HeadingNow
      letterSpacing=0
      monospacedWidth=1
      size=14.25
      skewAmount=0
      typefaceFontName=74 Regular
      outlineSettings {
       bApplyOutlineToDropShadows=false
       bMiteredCorners=false
       bSeparateFillAlpha=false
       outlineMaterial=null
       outlineSize=0
       outlineColor=<rgba 0 0 0 1>
      }
     }
     highlightColor {colorUseRule=UseColor_Specified, specifiedColor=<rgba 0 0 0 1>}
     highlightShape {
      bIsDynamicallyLoaded=false
      drawAs=Image
      imageType=NoImage
      mirroring=NoMirror
      resourceName=None
      resourceObject=null
      tiling=NoTile
      imageSize=<32 32>
      margin=<ltrb 0 0 0 0>
      outlineSettings {
       bUseBrushTransparency=false
       roundingType=HalfHeightRadius
       width=0
       color {colorUseRule=UseColor_Specified, specifiedColor=<rgba 0 0 0 0>}
       cornerRadii=<0 0 0 0>
      }
      tintColor {colorUseRule=UseColor_Specified, specifiedColor=<rgba 1 1 1 1>}
      uVRegion {bIsValid=false, max=<0 0>, min=<0 0>}
     }
     selectedBackgroundColor {colorUseRule=UseColor_Foreground, specifiedColor=<rgba 1 0 1 1>}
     shadowColorAndOpacity=<rgba 0 0 0 1>
     shadowOffset=<0 0>
     strikeBrush {
      bIsDynamicallyLoaded=false
      drawAs=Image
      imageType=NoImage
      mirroring=NoMirror
      resourceName=None
      resourceObject=null
      tiling=NoTile
      imageSize=<32 32>
      margin=<ltrb 0 0 0 0>
      outlineSettings {
       bUseBrushTransparency=false
       roundingType=HalfHeightRadius
       width=0
       color {colorUseRule=UseColor_Specified, specifiedColor=<rgba 0 0 0 0>}
       cornerRadii=<0 0 0 0>
      }
      tintColor {colorUseRule=UseColor_Specified, specifiedColor=<rgba 1 1 1 1>}
      uVRegion {bIsValid=false, max=<0 0>, min=<0 0>}
     }
     underlineBrush {
      bIsDynamicallyLoaded=false
      drawAs=Image
      imageType=NoImage
      mirroring=NoMirror
      resourceName=None
      resourceObject=null
      tiling=NoTile
      imageSize=<32 32>
      margin=<ltrb 0 0 0 0>
      outlineSettings {
       bUseBrushTransparency=false
       roundingType=HalfHeightRadius
       width=0
       color {colorUseRule=UseColor_Specified, specifiedColor=<rgba 0 0 0 0>}
       cornerRadii=<0 0 0 0>
      }
      tintColor {colorUseRule=UseColor_Specified, specifiedColor=<rgba 1 1 1 1>}
      uVRegion {bIsValid=false, max=<0 0>, min=<0 0>}
     }
    }
   }
   slot {class=SizeBoxSlot, properties {padding=<ltrb 0 5 0 0>}}
  }
  WBP_UIKit_StatusIndicator {
   widgetClassPath=/FortUILibrary/UIKit/StatusIndicator/WBP_UIKit_StatusIndicator
   bIsVariable=true
   parent=Overlay_39
   widget=WBP_UIKit_StatusIndicator
   properties {icon=/FortUILibrary/Textures/MaterialInstances/MI_UIKit_Icon_Check}
   slot {class=OverlaySlot, properties {horizontalAlignment=HAlign_Right, padding=<ltrb 0 80 10 0>}}
  }
  InputActionWidget {widgetClassPath=&CommonActionWidget, bInherited=true, widget=null}
 ]
}
