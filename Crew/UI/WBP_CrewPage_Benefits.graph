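# Activatable Crew page screen showing current subscription benefits as tiles, with legal text and a Back button that logs analytics and fires CloseBenefits.
# On activation, SetUpBenefits fills two rows from CrewVM's alternate groups, sized by a hardcoded TestingSize array, or from a data asset when Crew_TestData is set.
# SetUpBenefitsGrid, uncalled here, offers a grid layout from passed-in groups.
# Compact UI restyles sizing and padding; setup focuses the first tile.
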WidgetBlueprint /Crew/UI/WBP_CrewPage_Benefits
graphs [
 CloseBenefits {nodes [CloseBenefits() {is=title K2Node_FunctionEntry}]}
 CreateBenefitTileWidget {nodes [
   n1 = Inputs!() -> (execute, CrewBenefitVM:Crew Benefit VM Object Reference)
   > n2 = Create Widget!(
    execute:Exec
    Class:User Widget Class Reference=&/WBP_CrewBenefit_Tile.WBP_CrewBenefit_Tile_C
    OwningPlayer:Player Controller Object Reference=GetOwningPlayer()
    CrewBenefitVM:Crew Benefit VM Object Reference=@n1
    Style:Common Button Style Class Reference
    bApplyAlphaOnDisable:bool=True
    bLocked:bool=False
    bSelectable:bool=False
    bShouldSelectUponReceivingFocus:bool=False
    bInteractableWhenSelected:bool=False
    bToggleable:bool=False
    bTriggerClickedAfterSelection:bool=False
    bDisplayInputActionWhenNotInteractable:bool=True
    bHideInputActionWithKeyboard:bool=False
    bShouldUseFallbackDefaultInputAction:bool=True
    bRequiresHold:bool=False
    HoldData:Common UIHold Data Class Reference
    ClickMethod:EButtonClickMethod Enum=DownAndUp
    InputPriority:int=0
    TriggeringInputAction:Data Table Row Handle Structure
   ) -> (then:Exec, ReturnValue:WBP Crew Benefit Tile Object Reference)
   > Outputs!(then:Exec, CrewBenefitTile:WBP Crew Benefit Tile Object Reference=@n2)
  ]}
 EventGraph {nodes [
   n1 = "On Clicked (BackButton)"!() -> (OutputDelegate, then, Button:Common Button Base Object Reference) {is=title K2Node_ComponentBoundEvent}
   > n2 = GetPurchaseTraceId!(execute) -> (then, ReturnValue:Flow Analytics Additional Info Structure)
   > Fire Event Element Interacted(Element=@n1.Button, AdditionalInfo=@n2)
   > Call Close Benefits!(execute:Exec, self:WBP Crew Page Benefits Object Reference) -> (then:Exec) {is=title K2Node_CallDelegate}
   Event OnInitialized()
   > SetCompactUIStyle() {is=title K2Node_CallFunction}
   Event On Activated()
   > SetUpBenefits!(execute, self, isTest:bool=GetConfigDataBoolWithDefault!(EntryName:String=Crew_TestData, bDefaultValue:bool=false) -> (ReturnValue:bool)) -> (then)
  ]}
 GetLegalText {nodes [
   n1 = GetLegalText!() -> (then, Disclamer:Text, PaymentLegal:Text) {is=title K2Node_FunctionEntry}
   > Return Node!(
    execute
    LegalText:Text=Format Text!(Format="LOCTEXT(\"{Disclaimer} {PaymentLegal}\")", Disclaimer:Text=@n1, PaymentLegal:Text=@n1.PaymentLegal) -> (Result)
   )
  ]}
 SetCompactUIStyle {nodes [
   SetCompactUIStyle() {is=title K2Node_FunctionEntry}
   > Branch(Condition=IsUsingCompactUI())
   > n1 = Sequence!(execute) -> (then_0, then_1, then_2:Exec)
   n2 = Get SizeBox_BenefitsRows() -> (SizeBox_BenefitsRows:Size Box Object Reference)
   n3 = Get SizeBox_BenefitsGrid() -> (SizeBox_BenefitsGrid:Size Box Object Reference)
   ClearHeightOverride(execute=@n1.then_1, self:Size Box Object Reference=[@n2, @n3])
   > ClearWidthOverride(self:Size Box Object Reference=[@n2, @n3])
   > SetSize!(
    execute
    self:Vertical Box Slot Object Reference=SlotAsVerticalBoxSlot(Widget=@n2)
    InSize_Value:float=1
    InSize_SizeRule:ESlateSizeRule Enum=Fill
   ) -> (then)
   > SetSize(self=SlotAsVerticalBoxSlot(Widget=@n3), InSize_Value=1, InSize_SizeRule=Fill)
   SetPadding!(
    execute=@n1.then_2
    self:Vertical Box Slot Object Reference=SlotAsVerticalBoxSlot(Widget=Get HBox_LegalActionBar() -> (HBox_LegalActionBar:Horizontal Box Object Reference))
    InPadding_Left:float=0
    InPadding_Top:float=36
    InPadding_Right:float=0
    InPadding_Bottom:float=0
   ) -> (then)
   n4 = Get SB_Legal() -> (SB_Legal:Size Box Object Reference)
   > ClearWidthOverride(self=@n4)
   n5 = SlotAsHorizontalBoxSlot!(Widget:Widget Object Reference=@n4) -> (ReturnValue:Horizontal Box Slot Object Reference)
   > SetHorizontalAlignment(self:Horizontal Box Slot Object Reference=@n5, InHorizontalAlignment=HAlign_Fill)
   > SetPadding(self:Horizontal Box Slot Object Reference=@n5, InPadding_Left=0, InPadding_Top=0, InPadding_Right=24, InPadding_Bottom=0)
   n6 = SlotAsSafeBoxSlot!(Widget:Widget Object Reference=Get VB_Content() -> (VB_Content:Vertical Box Object Reference)) -> (ReturnValue:Safe Zone Slot Object Reference)
   Set!(
    execute=@n1.then_0
    Padding_Left:float=48
    Padding_Top:float=24
    Padding_Right:float=48
    Padding_Bottom:float=24
    self:Safe Zone Slot Object Reference=@n6
   ) -> (then, Output_Get:Margin Structure)
   > Set HAlign!(execute, HAlign:EHorizontalAlignment Enum=HAlign_Fill, self:Safe Zone Slot Object Reference=@n6) -> (then, Output_Get:EHorizontalAlignment Enum=HAlign_Fill)
   > Set VAlign!(execute, VAlign:EVerticalAlignment Enum=VAlign_Fill, self:Safe Zone Slot Object Reference=@n6) -> (then, Output_Get:EVerticalAlignment Enum=VAlign_Fill)
   > SetPadding(
    self:Vertical Box Slot Object Reference=SlotAsVerticalBoxSlot(Widget=Get Text_Title() -> (Text_Title:Text Block Object Reference))
    InPadding_Left=0
    InPadding_Top=0
    InPadding_Right=0
    InPadding_Bottom=24
   )
  ]}
 SetSeasonPass {nodes [
   n1 = SetSeasonPass!() -> (then, SeasonPassVM:Season Pass VM Object Reference) {is=title K2Node_FunctionEntry}
   > Set CurrSeasonPass(CurrSeasonPass:Season Pass VM Object Reference=@n1) -> (Output_Get:Season Pass VM Object Reference)
  ]}
 SetTile {nodes [
   n1 = SetTile!() -> (
    then
    BenefitTile:WBP Crew Benefit Tile Object Reference
    Texture:Texture Object Reference
    Top Color:Color Structure
    Bot Color:Color Structure
    Title:String
    Description:String
    Tag:String
    TileSize:double=~
   ) {is=title K2Node_FunctionEntry}
   > SetColorTop!(execute, self:WBP Crew Benefit Tile Object Reference=@n1, Color:Color Structure=@n1.Top Color) -> (then)
   > SetColorBot!(execute, self:WBP Crew Benefit Tile Object Reference=@n1, Color:Color Structure=@n1.Bot Color) -> (then)
   > SetIMGTexture!(execute, self:WBP Crew Benefit Tile Object Reference=@n1, Texture:Texture Object Reference=@n1.Texture) -> (then)
   > SetTileText!(execute, self:WBP Crew Benefit Tile Object Reference=@n1, Title:String=@n1.Title, Description:String=@n1.Description, Tag:String=@n1.Tag) -> (then)
   > SetTileSize!(execute, self:WBP Crew Benefit Tile Object Reference=@n1, TileSize:double=@n1.TileSize) -> (then)
   > Return Node!(execute, isMultiple:bool="Equal (Float)"(A="% (Float)"!(A:double=@n1.TileSize, B:double=1) -> (ReturnValue:double)))
  ]}
 SetTileFocus {nodes [
   SetTileFocus() {is=title K2Node_FunctionEntry}
   n1 = GetChildAt!(self:Panel Widget Object Reference=Get BenefitRow1() -> (BenefitRow1:Horizontal Box Object Reference), Index:int=0) -> (ReturnValue:Widget Object Reference)
   > n2 = Branch(Condition=IsNullItem!(ItemObject:Object Reference=@n1) -> (Is Null Item:bool))
   SetFocus(execute=@n2.else, self=@n1)
  ]}
 SetUpBenefits {nodes [
   SetUpBenefits!() -> (then, isTest:bool=~) {is=title K2Node_FunctionEntry}
   > SetVisibility(self=Get SizeBox_BenefitsRows(), InVisibility=SelfHitTestInvisible)
   > SetVisibility(self=Get SizeBox_BenefitsGrid(), InVisibility=Collapsed)
   > n1 = Sequence!(execute) -> (then_0, then_1)
   > n2 = Branch(Condition=Get isTest() -> (isTest:bool))
   > ClearChildren(self=Get BenefitRow1())
   n3 = Get DA_PC() -> (DA_PC:DA Crew Benefits Object Reference)
   > n4 = For Loop(LastIndex=int - int(A=Get TopAmmount!(self:DA Crew Benefits Object Reference=@n3) -> (TopAmmount:int=0)))
   ClearChildren(execute=@n4.Completed, self=Get BenefitRow2() -> (BenefitRow2:Horizontal Box Object Reference))
   n5 = Get DA_PC()
   n6 = Get Benefit Crew Data!(self:DA Crew Benefits Object Reference=@n5) -> (Benefit Crew Data:Array of S Crew B Tile Data Structures)
   > n7 = For Loop(FirstIndex=Get TopAmmount(self=@n5) -> (TopAmmount=0), LastIndex=int - int(A=Length(TargetArray:Array of S Crew B Tile Data Structures=@n6)))
   > n8 = Create Widget(
    Class=&WBP_CrewBenefit_Tile_C
    OwningPlayer=GetOwningPlayer()
    bApplyAlphaOnDisable=True
    bLocked=False
    bSelectable=False
    bShouldSelectUponReceivingFocus=False
    bInteractableWhenSelected=False
    bToggleable=False
    bTriggerClickedAfterSelection=False
    bDisplayInputActionWhenNotInteractable=True
    bHideInputActionWithKeyboard=False
    bShouldUseFallbackDefaultInputAction=True
    bRequiresHold=False
    ClickMethod=DownAndUp
    InputPriority=0
   )
   n9 = "Get (a ref)"!(Array:Array of S Crew B Tile Data Structures=@n6, Dimension 1=@n7) -> (
    Output_TileTexture_26_#1:Texture Object Reference
    Output_TopColor_24_#2:Color Structure
    Output_BotColor_25_#3:Color Structure
    Output_TileSize_21_#4:double=~
    Output_TileTitle_16_#5:String
    Output_Description_18_#6:String
    Output_Tag_20_#7:String
   )
   > n10 = SetTile!(
    execute
    self
    BenefitTile:WBP Crew Benefit Tile Object Reference=@n8
    Texture:Texture Object Reference=@n9
    Top Color:Color Structure=@n9.Output_TopColor_24_#2
    Bot Color:Color Structure=@n9.Output_BotColor_25_#3
    Title:String=@n9.Output_TileTitle_16_#5
    Description:String=@n9.Output_Description_18_#6
    Tag:String=@n9.Output_Tag_20_#7
    TileSize:double=@n9.Output_TileSize_21_#4
   ) -> (then, isMultiple:bool) {is=title K2Node_CallFunction}
   > n11 = AddChildToHorizontalBox!(execute, self:Horizontal Box Object Reference=Get BenefitRow2(), Content:Widget Object Reference=@n8) -> (then, ReturnValue:Horizontal Box Slot Object Reference)
   > SetSize(
    self:Horizontal Box Slot Object Reference=@n11
    InSize_Value=1
    InSize_SizeRule=Select(Option 0:ESlateSizeRule Enum=Fill, Option 1:ESlateSizeRule Enum=Automatic, Index=@n10) -> (ReturnValue:ESlateSizeRule Enum)
   )
   n12 = Create Widget(
    execute=@n4.LoopBody
    Class=&WBP_CrewBenefit_Tile_C
    OwningPlayer=GetOwningPlayer()
    bApplyAlphaOnDisable=True
    bLocked=False
    bSelectable=False
    bShouldSelectUponReceivingFocus=False
    bInteractableWhenSelected=False
    bToggleable=False
    bTriggerClickedAfterSelection=False
    bDisplayInputActionWhenNotInteractable=True
    bHideInputActionWithKeyboard=False
    bShouldUseFallbackDefaultInputAction=True
    bRequiresHold=False
    ClickMethod=DownAndUp
    InputPriority=0
   )
   n13 = "Get (a ref)"(Array=Get Benefit Crew Data(self=@n3), Dimension 1=@n4) -> (Output_TileSize_21_#4=~)
   > SetTile(
    BenefitTile=@n12
    Texture=@n13
    Top Color=@n13.Output_TopColor_24_#2
    Bot Color=@n13.Output_BotColor_25_#3
    Title=@n13.Output_TileTitle_16_#5
    Description=@n13.Output_Description_18_#6
    Tag=@n13.Output_Tag_20_#7
    TileSize=@n13.Output_TileSize_21_#4
   )
   > AddChildToHorizontalBox(self=Get BenefitRow1(), Content=@n12)
   n14 = For Each Loop(
    Exec=@n2.else
    Array:Array of Crew Benefits Group VM Object References=Get AltBenefitsGroups!(self:Crew VM Object Reference=Get CrewVM() -> (CrewVM:Crew VM Object Reference)) -> (AltBenefitsGroups:Array of Crew Benefits Group VM Object References)
   ) -> (Array Element:Crew Benefits Group VM Object Reference)
   > n15 = For Each Loop(Array:Array of Crew Benefit VM Object References=Get Benefits!(self:Crew Benefits Group VM Object Reference=@n14) -> (Benefits:Array of Crew Benefit VM Object References)) -> (Array Element:Crew Benefit VM Object Reference)
   n16 = Get BenefitRow1()
   ClearChildren(execute=@n15.Completed, self=@n16)
   > n17 = For Loop(LastIndex=3)
   > n18 = Create Widget(
    Class=&WBP_CrewBenefit_Tile_C
    OwningPlayer=GetOwningPlayer()
    CrewBenefitVM="Get (a copy)"(
     Array:Array of Crew Benefit VM Object References=Get AllBenefits() -> (AllBenefits:Array of Crew Benefit VM Object References)
     Dimension 1=@n17
    ) -> (Output:Crew Benefit VM Object Reference)
    bApplyAlphaOnDisable=True
    bLocked=False
    bSelectable=False
    bShouldSelectUponReceivingFocus=False
    bInteractableWhenSelected=False
    bToggleable=False
    bTriggerClickedAfterSelection=False
    bDisplayInputActionWhenNotInteractable=True
    bHideInputActionWithKeyboard=False
    bShouldUseFallbackDefaultInputAction=True
    bRequiresHold=False
    ClickMethod=DownAndUp
    InputPriority=0
   )
   > AddChildToHorizontalBox(self=@n16, Content=@n18)
   > SetTileSize(
    self=@n18
    TileSize="Get (a ref)"!(Array:"Array of Float (double-precision)s"=Get TestingSize() -> (TestingSize:"Array of Float (double-precision)s"), Dimension 1=@n17) -> (Output:"Float (double-precision) (by ref)")
   )
   n19 = Get BenefitRow2()
   ClearChildren(execute=@n17.Completed, self=@n19)
   > n20 = For Loop(FirstIndex=4, LastIndex=7)
   > n21 = Create Widget(
    Class=&WBP_CrewBenefit_Tile_C
    OwningPlayer=GetOwningPlayer()
    CrewBenefitVM="Get (a copy)"(Array=Get AllBenefits(), Dimension 1=@n20)
    bApplyAlphaOnDisable=True
    bLocked=False
    bSelectable=False
    bShouldSelectUponReceivingFocus=False
    bInteractableWhenSelected=False
    bToggleable=False
    bTriggerClickedAfterSelection=False
    bDisplayInputActionWhenNotInteractable=True
    bHideInputActionWithKeyboard=False
    bShouldUseFallbackDefaultInputAction=True
    bRequiresHold=False
    ClickMethod=DownAndUp
    InputPriority=0
   )
   > AddChildToHorizontalBox(self=@n19, Content=@n21)
   > SetTileSize(self=@n21, TileSize="Get (a ref)"(Array=Get TestingSize(), Dimension 1=@n20))
   Add(
    execute=@n15.LoopBody
    TargetArray:Array of Crew Benefit VM Object References=Get AllBenefits()
    NewItem:"Crew Benefit VM Object Reference (by ref)"=@n15
   ) {is=title K2Node_CallArrayFunction}
   SetTileFocus(execute=@n1.then_1) {is=title K2Node_CallFunction}
  ]}
 SetUpBenefitsGrid {nodes [
   n1 = For Each Loop(
    Exec=@n6.then_0
    Array:Array of Crew Benefits Group VM Object References=Get BenefitGroups() -> (BenefitGroups:Array of Crew Benefits Group VM Object References)
   ) -> (Array Element:Crew Benefits Group VM Object Reference)
   > n2 = For Each Loop(Array:Array of Crew Benefit VM Object References=Get Benefits(self=@n1)) -> (Array Element:Crew Benefit VM Object Reference)
   n3 = "integer > integer"(A=@n2.Array Index, B=3)
   > n4 = Create Benefit Tile Widget!(execute:Exec, CrewBenefitVM:Crew Benefit VM Object Reference=@n2) -> (then:Exec, CrewBenefitTile:WBP Crew Benefit Tile Object Reference) {is=title K2Node_MacroInstance}
   > n5 = AddChildToGrid!(
    execute
    self:Grid Panel Object Reference=Get GridPanelBenefits() -> (GridPanelBenefits:Grid Panel Object Reference)
    Content:Widget Object Reference=@n4
    InRow:int=SelectInt(A=1, B=0, bPickA=@n3)
    InColumn:int=SelectInt(A=int - int(A=@n2.Array Index, B=3), B=@n2.Array Index, bPickA=@n3)
   ) -> (then, ReturnValue:Grid Slot Object Reference)
   SetColumnSpan!(execute, self:Grid Slot Object Reference=@n5, InColumnSpan:int=0) -> (then)
   > SetRowSpan!(execute, self:Grid Slot Object Reference=@n5, InRowSpan:int=0) -> (then)
   SetUpBenefitsGrid!() -> (then, BenefitGroups:Array of Crew Benefits Group VM Object References) {is=title K2Node_FunctionEntry}
   > ClearChildren(self=Get GridPanelBenefits())
   > SetVisibility(self=Get SizeBox_BenefitsRows(), InVisibility=Collapsed)
   > SetVisibility(self=Get SizeBox_BenefitsGrid(), InVisibility=SelfHitTestInvisible)
   > n6 = Sequence()
   SetTileFocus(execute=@n6.then_1)
  ]}
]
variables [
 CloseBenefits {}
 CurrSeasonPass {type=SeasonPassVM, default=null}
 DA_PC {type=DA_CrewBenefits_C, default=/Crew/UI/Data/DA_PC_CrewB}
 EmptyTile {type=WBP_CrewBenefit_Tile_C, default=null}
 TestingSize {type="array<number>", default [3,1,1,1,2,2,1,1]}
]
widgets {
 info {
  inheritedWidgetCount=0
  namedSlotCount=1
  parentClass=&CommonActivatableWidget
  rootWidgetClass=&Overlay
  widgetCount=27
 }
 widgets [
  Overlay_24 {widgetClassPath=&Overlay, widget=Overlay_24}
  BG {
   widgetClassPath=&Image
   bIsVariable=true
   parent=Overlay_24
   widget=BG
   properties {
    visibility=SelfHitTestInvisible
    brush {
     bIsDynamicallyLoaded=false
     drawAs=Image
     imageType=FullColor
     mirroring=NoMirror
     resourceName=None
     resourceObject=/Crew/UI/Materials/CrewBenefits/M_UI_Crew_Unlocking_BG_Benefits_Instance
     tiling=NoTile
     imageSize=<1200 704>
     margin=<ltrb 0 0 0 0>
     outlineSettings {
      bUseBrushTransparency=false
      roundingType=HalfHeightRadius
      width=0
      color {colorUseRule=UseColor_Specified, specifiedColor=<rgba 0 0 0 0>}
      cornerRadii=<0 0 0 0>
     }
     tintColor {colorUseRule=UseColor_Specified, specifiedColor=<rgba 1 1 1 1>}
     uVRegion {bIsValid=false, max=<0 0>, min=<0 0>}
    }
   }
   slot {class=OverlaySlot, properties {horizontalAlignment=HAlign_Fill, verticalAlignment=VAlign_Fill}}
  }
  SafeZone {
   widgetClassPath=&SafeZone
   bIsVariable=true
   parent=Overlay_24
   widget=SafeZone
   slot {class=OverlaySlot, properties {horizontalAlignment=HAlign_Fill, verticalAlignment=VAlign_Fill, padding=<ltrb 0 15 0 0>}}
  }
  VB_Content {
   widgetClassPath=&VerticalBox
   bIsVariable=true
   parent=SafeZone
   widget=VB_Content
   slot {class=SafeZoneSlot, properties {hAlign=HAlign_Center, vAlign=VAlign_Center}}
  }
  Text_Title {
   widgetClassPath=&UEFN_TextBlock
   bIsVariable=true
   parent=VB_Content
   widget=Text_Title
   properties {
    text=CURRENT CREW BENEFITS
    font {
     bForceMonospaced=false
     bMaterialIsStencil=false
     fontMaterial=null
     fontObject=&HeadingNow
     letterSpacing=0
     monospacedWidth=1
     size=32.25
     skewAmount=0
     typefaceFontName=86 Bold Italic
     outlineSettings {
      bApplyOutlineToDropShadows=false
      bMiteredCorners=false
      bSeparateFillAlpha=false
      outlineMaterial=null
      outlineSize=0
      outlineColor=<rgba 0 0 0 1>
     }
    }
   }
   slot {class=VerticalBoxSlot, properties {horizontalAlignment=HAlign_Left, verticalAlignment=VAlign_Top, padding=<ltrb 0 0 0 37>}}
  }
  SizeBox_BenefitsRows {
   widgetClassPath=&SizeBox
   bIsVariable=true
   parent=VB_Content
   widget=SizeBox_BenefitsRows
   properties {bOverride_HeightOverride=true, bOverride_WidthOverride=true, heightOverride=750, visibility=Collapsed, widthOverride=1755}
   slot {class=VerticalBoxSlot}
  }
  ContainedinScreen {widgetClassPath=&ScaleBox, parent=SizeBox_BenefitsRows, widget=ContainedinScreen, slot {class=SizeBoxSlot}}
  BenefitRows {widgetClassPath=&VerticalBox, parent=ContainedinScreen, widget=BenefitRows, slot {class=ScaleBoxSlot}}
  BenefitRow1 {
   widgetClassPath=&HorizontalBox
   bIsVariable=true
   parent=BenefitRows
   widget=BenefitRow1
   slot {class=VerticalBoxSlot, properties {padding=<ltrb 0 0 0 10>}}
  }
  WBP_CrewBenefit_Tile {
   widgetClassPath=&/WBP_CrewBenefit_Tile
   bIsVariable=true
   parent=BenefitRow1
   widget=WBP_CrewBenefit_Tile
   properties {tileSize=2}
   slot {class=HorizontalBoxSlot, properties {padding=<ltrb 15 0 15 0>, size {sizeRule=Automatic, value=1.45}}}
  }
  WBP_CrewBenefit_Tile_2 {
   widgetClassPath=&WBP_CrewBenefit_Tile
   bIsVariable=true
   parent=BenefitRow1
   widget=WBP_CrewBenefit_Tile_2
   properties {tileSize=1.3}
   slot {class=HorizontalBoxSlot, properties {padding=<ltrb 15 0 15 0>, size {sizeRule=Fill, value=1}}}
  }
  WBP_CrewBenefit_Tile_3 {
   widgetClassPath=&WBP_CrewBenefit_Tile
   bIsVariable=true
   parent=BenefitRow1
   widget=WBP_CrewBenefit_Tile_3
   properties {tileSize=1.3}
   slot {class=HorizontalBoxSlot, properties {padding=<ltrb 15 0 15 0>, size {sizeRule=Fill, value=1}}}
  }
  WBP_CrewBenefit_Tile_4 {
   widgetClassPath=&WBP_CrewBenefit_Tile
   bIsVariable=true
   parent=BenefitRow1
   widget=WBP_CrewBenefit_Tile_4
   properties {tileSize=1.3}
   slot {class=HorizontalBoxSlot, properties {padding=<ltrb 15 0 15 0>, size {sizeRule=Fill, value=1}}}
  }
  BenefitRow2 {
   widgetClassPath=&HorizontalBox
   bIsVariable=true
   parent=BenefitRows
   widget=BenefitRow2
   properties {navigation=&self:WidgetTree.BenefitRow2.WidgetNavigation_0}
   slot {class=VerticalBoxSlot, properties {padding=<ltrb 0 10 0 0>}}
  }
  WBP_CrewBenefit_Tile_1 {
   widgetClassPath=&WBP_CrewBenefit_Tile
   bIsVariable=true
   parent=BenefitRow2
   widget=WBP_CrewBenefit_Tile_1
   properties {tileSize=2}
   slot {class=HorizontalBoxSlot, properties {padding=<ltrb 15 0 15 0>}}
  }
  WBP_CrewBenefit_Tile_7 {
   widgetClassPath=&WBP_CrewBenefit_Tile
   bIsVariable=true
   parent=BenefitRow2
   widget=WBP_CrewBenefit_Tile_7
   properties {tileSize=2}
   slot {class=HorizontalBoxSlot, properties {padding=<ltrb 15 0 15 0>}}
  }
  WBP_CrewBenefit_Tile_8 {
   widgetClassPath=&WBP_CrewBenefit_Tile
   bIsVariable=true
   parent=BenefitRow2
   widget=WBP_CrewBenefit_Tile_8
   slot {class=HorizontalBoxSlot, properties {padding=<ltrb 15 0 15 0>}}
  }
  WBP_CrewBenefit_Tile_9 {
   widgetClassPath=&WBP_CrewBenefit_Tile
   bIsVariable=true
   parent=BenefitRow2
   widget=WBP_CrewBenefit_Tile_9
   slot {class=HorizontalBoxSlot, properties {padding=<ltrb 15 0 15 0>}}
  }
  SizeBox_BenefitsGrid {
   widgetClassPath=&SizeBox
   bIsVariable=true
   parent=VB_Content
   widget=SizeBox_BenefitsGrid
   properties {bOverride_HeightOverride=true, bOverride_WidthOverride=true, heightOverride=750, visibility=Collapsed, widthOverride=1755}
   slot {class=VerticalBoxSlot}
  }
  GridPanelBenefits {widgetClassPath=&GridPanel, bIsVariable=true, parent=SizeBox_BenefitsGrid, widget=GridPanelBenefits, slot {class=SizeBoxSlot}}
  HBox_LegalActionBar {
   widgetClassPath=&HorizontalBox
   bIsVariable=true
   parent=VB_Content
   widget=HBox_LegalActionBar
   slot {class=VerticalBoxSlot, properties {verticalAlignment=VAlign_Bottom, padding=<ltrb 0 50 80 0>}}
  }
  SB_Legal {
   widgetClassPath=&SizeBox
   bIsVariable=true
   parent=HBox_LegalActionBar
   widget=SB_Legal
   properties {bOverride_WidthOverride=true, widthOverride=830}
   slot {class=HorizontalBoxSlot, properties {horizontalAlignment=HAlign_Left, verticalAlignment=VAlign_Bottom, size {sizeRule=Fill, value=1}}}
  }
  Text_Legal {
   widgetClassPath=&CommonTextBlock
   bIsVariable=true
   parent=SB_Legal
   widget=Text_Legal
   properties {
    autoWrapText=true
    mobileFontSizeMultiplier=1.4
    scrollOrientation=Orient_Vertical
    text="[PH] Auto-renews each month until cancelled. Cancel anytime. Territorial restrictions apply. Ability to claim Premium Rewards during active subscription only. Select \"Subscription Terms\" for more info."
    colorAndOpacity {colorUseRule=UseColor_Specified, specifiedColor=<rgba 0.5776 0.5776 0.5776 1>}
    font {
     bForceMonospaced=false
     bMaterialIsStencil=false
     fontMaterial=null
     fontObject=&NotoSans
     letterSpacing=0
     monospacedWidth=1
     size=24
     skewAmount=0
     typefaceFontName=NotoSans-Regular
     outlineSettings {
      bApplyOutlineToDropShadows=false
      bMiteredCorners=false
      bSeparateFillAlpha=false
      outlineMaterial=null
      outlineSize=0
      outlineColor=<rgba 0 0 0 1>
     }
    }
    shadowOffset=<0 0>
   }
   slot {class=SizeBoxSlot}
  }
  WBP_ActionBarContainer {
   widgetClassPath=/Game/UI/Frontend/BottomBar/WBP_ActionBarContainer
   bIsVariable=true
   parent=HBox_LegalActionBar
   widget=WBP_ActionBarContainer
   slot {class=HorizontalBoxSlot, properties {horizontalAlignment=HAlign_Right, verticalAlignment=VAlign_Bottom}}
  }
  HorizontalBox {widgetClassPath=&HorizontalBox, namedSlotHost=WBP_ActionBarContainer, widget=HorizontalBox}
  WBP_AthenaBottomBar_WithNotifications {
   widgetClassPath=/Game/UI/Frontend/BottomBar/WBP_AthenaBottomBar_WithNotifications
   bIsVariable=true
   parent=HorizontalBox
   widget=WBP_AthenaBottomBar_WithNotifications
   slot {class=HorizontalBoxSlot, properties {padding=<ltrb 0 0 16 0>}}
  }
  BackButton {
   widgetClassPath=/FortUILibrary/UIKit/Buttons/WBP_UIKit_Button_Quiet
   bIsVariable=true
   parent=HorizontalBox
   widget=BackButton
   properties {text=Back, triggeringInputAction {dataTable=&InputActionDataTable, rowName=GenericBack}}
   slot {class=HorizontalBoxSlot}
  }
 ]
}
