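# Activatable menu screen for Crew subscription info, hosting the upsell landing, benefits and progressive cosmetic pages in one switcher over a background.
# PageSelected sets the visible page, hides the background for progressive content, and collapses the switcher for the reset state used on init and deactivation.
# Child events drive navigation: closing benefits exits if it is root, otherwise returns to upsell.
# A deprecated benefits screen relays CrewPurchased; the tab group is disabled and collapsed.
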WidgetBlueprint /Crew/UI/CrewInfoContainer_B
graphs [
 BP_GetDesiredInputConfig {nodes [
   Get Desired Input Config() {is=title K2Node_FunctionEntry}
   > Return Node!(
    execute
    ReturnValue_bIgnoreMoveInput:bool=False
    ReturnValue_bIgnoreLookInput:bool=False
    ReturnValue_InputMode:ECommonInputMode Enum=Menu
    ReturnValue_MouseCaptureMode:EMouseCaptureMode Enum=NoCapture
    ReturnValue_MouseLockMode:EMouseLockMode Enum=DoNotLock
    ReturnValue_bHideCursorDuringViewportCapture:bool=True
   )
  ]}
 CrewPurchased {nodes [CrewPurchased() {is=title K2Node_FunctionEntry}]}
 EventGraph {nodes [
   n1 = SetActiveWidgetIndex(self=Get PageSelectSwitch() -> (PageSelectSwitch:Common Activatable Widget Switcher Object Reference), Index=0)
   SetActiveWidgetIndex(
    execute=[@n1.then, @n2.then]
    self:Common Visibility Switcher Object Reference=Get CrewBenefitSwitchDepricated() -> (CrewBenefitSwitchDepricated:Common Visibility Switcher Object Reference)
    Index=Select(
     Option 0:int=0
     Option 1:int=1
     Index=GetConfigDataBoolWithDefault!(EntryName:String=New_CrewBenefits, bDefaultValue:bool=false) -> (ReturnValue:bool)
    ) -> (ReturnValue:int)
   )
   n2 = SetActiveWidgetIndex(self:Widget Switcher Object Reference=Get PageSelectSwitch(), Index=SelectInt(A=0, B=1, bPickA="Equal (Name)"(B=0)))
   "Close Screen (ProgressiveContentContainer)"() {is=title K2Node_ComponentBoundEvent}
   > DeactivateWidget()
   "Crew Purchased (CrewBenefitsScreen)"() {is=title K2Node_ComponentBoundEvent}
   > Call Crew Purchased!(execute:Exec, self:Crew Info Container B Object Reference) -> (then:Exec) {is=title K2Node_CallDelegate}
   "See Benefits (WBP_CrewLandingPage_Upsell)"() {is=title K2Node_ComponentBoundEvent}
   > PageSelected!(execute, self, PageToGo:E_CrewContainer_Page Enum=NewEnumerator1, IsBenefitsRoot:bool=false) -> (then)
   n3 = "Close Page (WBP_CrewLandingPage_Upsell)"() {is=title K2Node_ComponentBoundEvent}
   DeactivateWidget(execute=[@n3.then, @n4.then])
   "Close Benefits (WBP_CrewPage_Benefits)"() {is=title K2Node_ComponentBoundEvent}
   > n4 = Branch(Condition=Get Is Benefits Root() -> (Is Benefits Root:bool))
   PageSelected(execute=[@n5.then, @n4.else], PageToGo=NewEnumerator2, IsBenefitsRoot=false)
   n5 = "Up Sell Screen (ProgressiveContentContainer)"() {is=title K2Node_ComponentBoundEvent}
   Event On Deactivated()
   > PageSelected(PageToGo=NewEnumerator3, IsBenefitsRoot=false)
   Event OnInitialized()
   > SetCompactUIStyle() {is=title K2Node_CallFunction}
   > PageSelected(PageToGo=NewEnumerator3, IsBenefitsRoot=false)
   "Close Screen (CrewBenefitsScreen)"() {is=title K2Node_ComponentBoundEvent}
   "On UIKit Tab Selected (WBP_UIKit_TabGroup)"!() -> (OutputDelegate, then, TabID:Name=None, TabVM:MVVM Base Viewmodel Object Reference) {is=title K2Node_ComponentBoundEvent}
   Event On Activated()
  ]}
 GetFrontendFlowInterruptConfig {nodes [
   GetFrontendFlowInterruptConfig() {is=title K2Node_FunctionEntry}
   > Return Node!(
    execute
    ReturnValue:Frontend Flow Interrup Config Structure=Make Frontend Flow Interrup Config!(bCanFrontendFlowInterrupt:bool=True, bCanInterruptByEmergentOnly:bool=true) -> (FrontendFlowInterrupConfig:Frontend Flow Interrup Config Structure) {is=title K2Node_MakeStruct}
   )
  ]}
 GetPrimaryContentSetup {nodes [
   GetPrimaryContentSetup() {is=title K2Node_FunctionEntry}
   > Return Node(ReturnValue:Primary Content Setup Structure=Make Primary Content Setup!(
     bShowBottomBar:bool=false
     bShowChatWidget:bool=false
     bEnableNewTextChat:bool=False
     bShowTopBar:bool=false
     bTabSupportsInterface:bool=False
    ) -> (PrimaryContentSetup:Primary Content Setup Structure) {is=title K2Node_MakeStruct})
  ]}
 PageSelected {nodes [
   n1 = PageSelected!() -> (then, PageToGo:E_CrewContainer_Page Enum=NewEnumerator0, IsBenefitsRoot:bool=~) {is=title K2Node_FunctionEntry}
   > SetVisibility(
    self=Get PageSelectSwitch()
    InVisibility=Select(
     Option 0:ESlateVisibility Enum=SelfHitTestInvisible
     Option 1:ESlateVisibility Enum=Collapsed
     Index="Equal (Enum)"(A:E_CrewContainer_Page Enum=@n1, B:E_CrewContainer_Page Enum=NewEnumerator3)
    ) -> (ReturnValue:ESlateVisibility Enum)
   )
   > Set Is Benefits Root(Is Benefits Root:bool=AND Boolean(A=@n1.IsBenefitsRoot, B="Equal (Enum)"(A=NewEnumerator1, B=@n1))) -> (Output_Get:bool)
   > Set CurrentPage(CurrentPage:E_CrewContainer_Page Enum=@n1) -> (Output_Get:E_CrewContainer_Page Enum=NewEnumerator0)
   > SetActiveWidgetIndex(
    self=Get PageSelectSwitch()
    Index=Select!(NewEnumerator0:int=2, NewEnumerator1:int=1, NewEnumerator2:int=0, NewEnumerator3:int=0, Index:E_CrewContainer_Page Enum=@n1) -> (ReturnValue:int)
   )
   > Set Visibility(
    Visibility=Select!(
     Option 0:ESlateVisibility Enum=SelfHitTestInvisible
     Option 1:ESlateVisibility Enum=Collapsed
     Index:bool="Equal (Enum)"(A=@n1, B=NewEnumerator0)
    ) -> (ReturnValue:ESlateVisibility Enum)
    self=Get BG() -> (BG:Image Object Reference)
   )
  ]}
 SetCompactUIStyle {nodes [
   SetCompactUIStyle() {is=title K2Node_FunctionEntry}
   > Branch(Condition=IsMobileGame())
   > Set!(
    execute
    MinButtonSize_X:int=0
    MinButtonSize_Y:int=80
    self:WBP UIKit Tab Group Base Object Reference=Get WBP_UIKit_TabGroup() -> (WBP_UIKit_TabGroup:WBP UIKit Tab Group Object Reference)
   ) -> (then, Output_Get:Int Point Structure)
   > SetPadding!(
    execute
    self:Overlay Slot Object Reference=SlotAsOverlaySlot(Widget=Get SafeZone_TabGroup() -> (SafeZone_TabGroup:Safe Zone Object Reference))
    InPadding_Left:float=0
    InPadding_Top:float=24
    InPadding_Right:float=0
    InPadding_Bottom:float=0
   ) -> (then)
  ]}
]
variables [
 CrewPurchased {}
 CurrentPage {type=E_CrewContainer_Page, default=CrewStyles}
 Is Benefits Root {type=boolean}
]
widgets {
 info {inheritedWidgetCount=0, parentClass=&CommonActivatableWidget, rootWidgetClass=&Overlay, widgetCount=10}
 widgets [
  Overlay_0 {widgetClassPath=&Overlay, widget=Overlay_0}
  BG {
   widgetClassPath=&Image
   bIsVariable=true
   parent=Overlay_0
   widget=BG
   properties {
    visibility=SelfHitTestInvisible
    brush {
     bIsDynamicallyLoaded=false
     drawAs=Image
     imageType=FullColor
     mirroring=NoMirror
     resourceName=None
     resourceObject=/Crew/UI/Materials/CrewBenefits/M_UI_Crew_Unlocking_BG_Benefits_Instance
     tiling=NoTile
     imageSize=<1200 704>
     margin=<ltrb 0 0 0 0>
     outlineSettings {
      bUseBrushTransparency=false
      roundingType=HalfHeightRadius
      width=0
      color {colorUseRule=UseColor_Specified, specifiedColor=<rgba 0 0 0 0>}
      cornerRadii=<0 0 0 0>
     }
     tintColor {colorUseRule=UseColor_Specified, specifiedColor=<rgba 1 1 1 1>}
     uVRegion {bIsValid=false, max=<0 0>, min=<0 0>}
    }
   }
   slot {class=OverlaySlot, properties {horizontalAlignment=HAlign_Fill, verticalAlignment=VAlign_Fill}}
  }
  PageSelectSwitch {
   widgetClassPath=&CommonActivatableWidgetSwitcher
   bIsVariable=true
   parent=Overlay_0
   widget=PageSelectSwitch
   slot {class=OverlaySlot, properties {horizontalAlignment=HAlign_Fill, verticalAlignment=VAlign_Fill}}
  }
  WBP_CrewLandingPage_Upsell {
   widgetClassPath=&/WBP_CrewLandingPage_Upsell
   bIsVariable=true
   parent=PageSelectSwitch
   widget=WBP_CrewLandingPage_Upsell
   slot {class=WidgetSwitcherSlot}
  }
  WBP_CrewPage_Benefits {
   widgetClassPath=&/WBP_CrewPage_Benefits
   bIsVariable=true
   parent=PageSelectSwitch
   widget=WBP_CrewPage_Benefits
   properties {emptyTile=&self:WidgetTree.WBP_CrewPage_Benefits.WBP_CrewBenefit_Tile_C_0}
   slot {class=WidgetSwitcherSlot}
  }
  ProgressiveContentContainer {
   widgetClassPath=/Crew/UI/ProgressiveCosmetic/ProgressiveContentContainer
   bIsVariable=true
   parent=PageSelectSwitch
   widget=ProgressiveContentContainer
   slot {class=WidgetSwitcherSlot}
  }
  CrewBenefitSwitchDepricated {
   widgetClassPath=&CommonVisibilitySwitcher
   bIsVariable=true
   parent=PageSelectSwitch
   widget=CrewBenefitSwitchDepricated
   slot {class=WidgetSwitcherSlot}
  }
  CrewBenefitsScreen {
   widgetClassPath=&/CrewBenefitsScreen
   bIsVariable=true
   parent=CrewBenefitSwitchDepricated
   widget=CrewBenefitsScreen
   slot {class=CommonVisibilitySwitcherSlot}
  }
  SafeZone_TabGroup {
   widgetClassPath=&SafeZone
   bIsVariable=true
   parent=Overlay_0
   widget=SafeZone_TabGroup
   properties {padBottom=false, padLeft=false, padRight=false}
   slot {class=OverlaySlot, properties {horizontalAlignment=HAlign_Center, padding=<ltrb 0 42 0 0>}}
  }
  WBP_UIKit_TabGroup {
   widgetClassPath=/FortUILibrary/UIKit/Tabs/WBP_UIKit_TabGroup
   bIsVariable=true
   parent=SafeZone_TabGroup
   widget=WBP_UIKit_TabGroup
   properties {
    bIsEnabled=false
    backplateBrightness=Dark
    backplateCornerRadius=Tab
    initialSelectedTabId=0
    tabSpacing=0
    useInitialTabs=true
    useMinButtonSize=false
    visibility=Collapsed
    buttonPadding=<ltrb 8 8 8 8>
    initialTabsData [
     {bEnableTabId=false, tabId=None, tabViewModel=&self:WidgetTree.WBP_UIKit_TabGroup.UIKitTabGroupButtonViewModel_0}
     {bEnableTabId=false, tabId=None, tabViewModel=&self:WidgetTree.WBP_UIKit_TabGroup.UIKitTabGroupButtonViewModel_1}
    ]
    tabGroupPadding=<ltrb 0 0 0 0>
   }
   slot {class=SafeZoneSlot}
  }
 ]
}
